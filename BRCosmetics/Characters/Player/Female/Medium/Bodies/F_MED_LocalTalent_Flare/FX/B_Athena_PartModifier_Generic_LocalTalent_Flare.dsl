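; Part modifier for the LocalTalent Flare body cosmetic that spawns reactive Niagara effects at the head and left thigh.
; In gameplay, a positive stat change spawns them unless the head is hidden, with a one-time delayed repeat via DoOnce.
; In the frontend, a looping timer respawns them while a ping-pong timeline with hold delays fades a fake-audio parameter on materials and idle VFX.
; The timer is cleared on destroy.
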(event Custom|CE_Loop_Frontend
  (bind _returnvalue (CustomCharacter|GetSkeletalMeshForPartOnPlayerPawn))
  (bind _returnvalue_2 (PartModifier|SpawnSystem|SpawnSystemAttachedPartModifier (Variables|Default|GetNiagaraSystem) _returnvalue "head" "2.500000,0.500000,12.000000" "90.000000,0.000000,0.000000")
    (:Success
      (Effects|Components|ParticleSystem|SetBoolParameter _returnvalue_2 "IsLeft" true)
      (PartModifier|SpawnSystem|SpawnSystemAttachedPartModifier (Variables|Default|GetNiagaraSystem) _returnvalue "head" "3.000000,0.500000,-12.000000" "-90.000000,0.000000,0.000000"
        (:Success
          (bind _returnvalue_3 (PartModifier|SpawnSystem|SpawnSystemAttachedPartModifier (Variables|Default|GetNiagaraSystem_02) _returnvalue "thigh_l")
            (:Success
              (Effects|Components|ParticleSystem|SetBoolParameter _returnvalue_3 "Hide light" true))))))))

(event EventOnPostResIn
  (|CallParentFunction)
  (if (Reactive|Frontend|ShouldDisplayFrontendReactivity 0 "Body")
    (Utilities|FlowControl|Sequence
      (:then_0
        (Variables|Default|SetBeginLoop true :!label _label)
        (bind (_direction _in) (Timeline|Timeline PlayFromStart)
          (:Update
            (CallFunction|OnFakeLoopTimelineUpdate _in))
          (:Finished
            (if (Variables|Default|GetBeginLoop)
              (Variables|Default|SetBeginLoop false)
              (Utilities|FlowControl|Delay (Variables|Default|GetFakeAmplitudeOnLength))
              (Timeline|Timeline ReverseFromEnd)
              (else
                (Utilities|FlowControl|Delay (Variables|Default|GetFakeAmplitudeOffLength))
                (goto _label))))))
      (:then_1
        (bind _returnvalue (Utilities|Time|SetTimerbyEvent (delegate Custom|CE_Loop_Frontend) 3.0 true))
        (Variables|Default|SetTimer _returnvalue)))))

(event EventOnStatChanged
  (params (State%20Name :type Name :default "None") (Stat%20Value :type Integer))
  (bind _returnvalue (CustomCharacter|GetSkeletalMeshForPartOnPlayerPawn))
  (|CallParentFunction State%20Name Stat%20Value)
  (if (> Stat%20Value 0)
    (if (not (Variables|HiddenPartCheck|GetbIsHeadHidden))
      (bind _returnvalue_2 (PartModifier|SpawnSystem|SpawnSystemAttachedPartModifier (Variables|Default|GetNiagaraSystem) _returnvalue "head" "2.500000,0.500000,12.000000" "90.000000,0.000000,0.000000" :!label _label)
        (:Success
          (Effects|Components|ParticleSystem|SetBoolParameter _returnvalue_2 "IsLeft" true)
          (PartModifier|SpawnSystem|SpawnSystemAttachedPartModifier (Variables|Default|GetNiagaraSystem) _returnvalue "head" "3.000000,0.500000,-12.000000" "-90.000000,0.000000,0.000000"
            (:Success
              (bind _returnvalue_3 (PartModifier|SpawnSystem|SpawnSystemAttachedPartModifier (Variables|Default|GetNiagaraSystem_02) _returnvalue "thigh_l" "2.000000,0.000000,4.000000" "90.000000,0.000000,0.000000")
                (:Success
                  (Transformation|SetRelativeScale3D _returnvalue_3 "1.300000,1.300000,1.300000")
                  (Effects|Components|ParticleSystem|SetBoolParameter _returnvalue_3 "Hide light" true)
                  (Utilities|FlowControl|RetriggerableDelay 1.3)
                  (Utilities|FlowControl|DoOnce false
                    (:Completed
                      (goto _label))))))))))))

(event EventDestroyed
  (|CallParentFunction)
  (if (CustomCosmeticModifier|IsinFrontEnd)
    (Utilities|Time|ClearandInvalidateTimerbyHandle (Variables|Default|GetTimer))))

(event Custom|OnFakeLoopTimelineUpdate
  (params (TransitionValue :type Float%20%28single-precision%29))
  (Rendering|Material|SetScalarParameterValueonMaterials (CustomCharacter|GetSkeletalMeshForPartOnPlayerPawn) "EnableFrontendFakeAudioLoop" TransitionValue) ; Enable Fake Audio in Frontend for Materials
  (Effects|Components|ParticleSystem|SetFloatParameter (CallFunction|GetIdleVFXComponent) "EnableFrontendFakeAudioLoop" TransitionValue))

(fn ConstructionScript
  (|CallParentFunction))
