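# Niagara burst effect at the head of the LocalTalent Flare character, layering star meshes and glows, rays, small spinning sprites and a brief colored light.
# Each emitter spawns short-lived burst particles shaped by size, alpha and material-parameter curves.
# User.VariantSwitch selects palettes through array lookups and curve lerps, bisHeadHidden drives particle VisibilityTags, and Hide light suppresses the light.
# Front-end modifiers dim some star layers, and time of day scales sprite alpha.
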NiagaraSystem /BRCosmetics/Characters/Player/Female/Medium/Bodies/F_MED_LocalTalent_Flare/FX/NS_Character_LocalTalent_Flare_HeadBurst
dependencies {
 usedDataInterfaces [
  &NiagaraDataInterfaceCurve
  &NiagaraDataInterfaceArrayColor
  &NiagaraDataInterfaceArrayFloat
  &NiagaraDataInterfaceArrayFloat3
  &NiagaraDataInterfaceColorCurve
 ]
 usedDynamicInputs [&AddVectorToPosition]
 usedModules [&SystemState]
}
emitters {
 Light {
  emitterUpdateScript [
   &EmitterState(
    Inactive Response=::NewEnumerator1
    Life Cycle Mode=::NewEnumerator1
    Loop Behavior=::NewEnumerator1
    Loop Delay=0.3
    Loop Duration=1
    UseLoopDelay=true
   )
   &SpawnBurst_Instantaneous(
    Spawn Count=&MakeCustomIntFromBool(Bool=$User.Hide light, True INT=0, False INT=1)
    Use Spawn Probability=false
    UseLoopCountLimit=false
   )
  ]
  particleSpawnScript [&InitializeParticle(
    Lifetime=0.5
    Mesh Scale=<0.3 -1 -1>
    Mesh Scale Mode=::NewEnumerator4
    Position Offset=<10 -20 0>
    Sprite Rotation Angle=&RandomRangeFloat(Minimum=0, Maximum=1)
    Sprite Rotation Mode=::NewEnumerator1
    Sprite Size Mode=::NewEnumerator3
    UsePositionOffset=true
    Write Material Random=true
   )]
  particleUpdateScript [
   &ParticleState()
   &FN_Light_Attributes(
    Color=&MakeLinearColorFromVectorAndFloat(
     "Vector (RGB)"=/Niagara/DynamicInputs/Math/LerpMultipleVectors(
      Alpha:NiagaraFloat=$User.VariantSwitch
      V 1:Vector3f=&MakeVectorFromLinearColorRGB(Color=&ColorFromCurve(ColorCurve=DataInterface() {propertyValues {
          AlphaCurve {keys []}
          BlueCurve {keys [
            <at 0.003 0>
            {arriveTangent=1.2415, interpMode=RCIM_Cubic, leaveTangent=1.2415, time=0.2077, value=0.002}
            {arriveTangent=1.2415, interpMode=RCIM_Cubic, leaveTangent=1.2415, time=0.4825, value=0}
            <at 0.8084 1>
            <at 0.9954 0>
           ]}
          GreenCurve {keys [
            <at 0.003 0>
            {arriveTangent=0.5353, interpMode=RCIM_Cubic, leaveTangent=0.5353, time=0.2077, value=1}
            {arriveTangent=0.5353, interpMode=RCIM_Cubic, leaveTangent=0.5353, time=0.4825, value=1}
            <at 0.8084 0.4312>
            <at 0.9954 0>
           ]}
          RedCurve {keys [<at 0.003 0>,<at 0.2077 0.9978 RCIM_Cubic>,<at 0.4825 0.0389 RCIM_Cubic>,<at 0.8084 0>,<at 0.9954 0>]}
         }}))
      V 2:Vector3f=&MakeVectorFromLinearColorRGB(Color=&ColorFromCurve(ColorCurve=DataInterface() {propertyValues {
          AlphaCurve {keys []}
          BlueCurve {keys [
            <at 0.003 0>
            {arriveTangent=1.2415, interpMode=RCIM_Cubic, leaveTangent=1.2415, time=0.2077, value=0.002}
            {arriveTangent=1.2415, interpMode=RCIM_Cubic, leaveTangent=1.2415, time=0.4825, value=0}
            <at 0.8084 1>
            <at 0.9954 0>
           ]}
          GreenCurve {keys [
            <at 0.003 0>
            {arriveTangent=0.5353, interpMode=RCIM_Cubic, leaveTangent=0.5353, time=0.2077, value=1}
            {arriveTangent=0.5353, interpMode=RCIM_Cubic, leaveTangent=0.5353, time=0.4825, value=1}
            <at 0.8084 0.4312>
            <at 0.9954 0>
           ]}
          RedCurve {keys [<at 0.003 0>,<at 0.2077 0.9978 RCIM_Cubic>,<at 0.4825 0.0389 RCIM_Cubic>,<at 0.8084 0>,<at 0.9954 0>]}
         }}))
      V 3:Vector3f=&MakeVectorFromLinearColorRGB(Color=&ColorFromCurve(ColorCurve=DataInterface() {propertyValues {
          AlphaCurve {keys []}
          BlueCurve {keys [
            <at 0 0>
            <at 0.1383 0.5138>
            <at 0.1966 0.5392>
            <at 0.3441 1>
            <at 0.438 1>
            {arriveTangent=2.8924, leaveTangent=2.8924, time=0.5752, value=0.886}
            <at 0.7062 0.886>
            <at 0.8281 0.5361>
            <at 0.9308 0.4659>
            <at 1 0>
           ]}
          GreenCurve {keys [<at 0 0>,<at 0.1383 0.3199>,<at 0.1966 0.2344>,<at 0.3441 0.2797>,<at 0.438 0.2596>,<at 0.5752 1>,<at 0.7062 1>,<at 0.8281 0.1904>,<at 0.9308 0.2029>,<at 1 0>]}
          RedCurve {keys [
            <at 0 0>
            <at 0.1383 1>
            <at 0.1966 1>
            <at 0.3441 0.4759>
            <at 0.438 0.4265>
            {arriveTangent=-2.6514, leaveTangent=-2.6514, time=0.5752, value=0}
            <at 0.7062 0>
            <at 0.8281 1>
            <at 0.9308 1>
            <at 1 0>
           ]}
         }}))
      V 4:Vector3f=&MakeVectorFromLinearColorRGB(Color=&ColorFromCurve(ColorCurve=DataInterface() {propertyValues {
          AlphaCurve {keys []}
          BlueCurve {keys [
            <at 0.003 0>
            {arriveTangent=1.9217, interpMode=RCIM_Cubic, leaveTangent=1.9217, time=0.2522, value=0}
            <at 0.5234 1 RCIM_Cubic>
            <at 0.8433 1>
            <at 0.9954 0>
           ]}
          GreenCurve {keys [
            <at 0.003 0>
            {arriveTangent=0.1391, interpMode=RCIM_Cubic, leaveTangent=0.1391, time=0.2522, value=1}
            {arriveTangent=-1.3454, interpMode=RCIM_Cubic, leaveTangent=-1.3454, time=0.5234, value=0.0724}
            <at 0.8433 0.4312>
            <at 0.9954 0>
           ]}
          RedCurve {keys [
            <at 0.003 0>
            {arriveTangent=0.6979, interpMode=RCIM_Cubic, leaveTangent=0.6979, time=0.2522, value=0.0389}
            {arriveTangent=-0.0523, interpMode=RCIM_Cubic, leaveTangent=-0.0523, time=0.5234, value=0.3632}
            <at 0.8433 0>
            <at 0.9954 0>
           ]}
         }}))
      V 5:Vector3f=&MakeVectorFromLinearColorRGB(Color=&ColorFromCurve(ColorCurve=DataInterface() {propertyValues {
          AlphaCurve {keys []}
          BlueCurve {keys [
            <at 0.003 0>
            {arriveTangent=0.0094, interpMode=RCIM_Cubic, leaveTangent=0.0094, time=0.2522, value=0.0044}
            {arriveTangent=-0.0074, interpMode=RCIM_Cubic, leaveTangent=-0.0074, time=0.5234, value=0.0049}
            <at 0.8433 0>
            <at 0.9954 0>
           ]}
          GreenCurve {keys [
            <at 0.003 0>
            {arriveTangent=1.4601, interpMode=RCIM_Cubic, leaveTangent=1.4601, time=0.2522, value=0.1628}
            {arriveTangent=0.0319, interpMode=RCIM_Cubic, leaveTangent=0.0319, time=0.5234, value=0.7598}
            <at 0.8433 0.1816>
            <at 0.9954 0>
           ]}
          RedCurve {keys [
            <at 0.003 0>
            {arriveTangent=1.9217, interpMode=RCIM_Cubic, leaveTangent=1.9217, time=0.2522, value=1}
            <at 0.5234 1 RCIM_Cubic>
            <at 0.8433 1>
            <at 0.9954 0>
           ]}
         }}))
      Vector Count:ENiagara_MultipleLerpCount=::NewEnumerator2
     )
     "Float (Alpha)"=1
    )
    Create Color Attribute=true
    Create Exponent Attribute=true
    Create Position Attribute=true
    Create Radius Attribute=true
    Exponent=1.5
    Position=<20 -30 0>
    Radius=75
    Scale Color=20
   )
   /SetVariables(Particles.VisibilityTag=&MakeCustomIntFromBool(Bool=$User.bisHeadHidden, True INT=1, False INT=0)) {isSetParameters=true}
   &FN_Light_Attributes(
    Color=&ColorFromCurve(ColorCurve=DataInterface() {propertyValues {
       AlphaCurve {keys []}
       BlueCurve {keys [
         <at 0.003 0>
         {arriveTangent=1.2415, interpMode=RCIM_Cubic, leaveTangent=1.2415, time=0.2077, value=0.002}
         {arriveTangent=1.2415, interpMode=RCIM_Cubic, leaveTangent=1.2415, time=0.4825, value=0}
         <at 0.8084 1>
         <at 0.9954 0>
        ]}
       GreenCurve {keys [
         <at 0.003 0>
         {arriveTangent=0.5353, interpMode=RCIM_Cubic, leaveTangent=0.5353, time=0.2077, value=1}
         {arriveTangent=0.5353, interpMode=RCIM_Cubic, leaveTangent=0.5353, time=0.4825, value=1}
         <at 0.8084 0.4312>
         <at 0.9954 0>
        ]}
       RedCurve {keys [<at 0.003 0>,<at 0.2077 0.9978 RCIM_Cubic>,<at 0.4825 0.0389 RCIM_Cubic>,<at 0.8084 0>,<at 0.9954 0>]}
      }})
    Create Color Attribute=true
    Create Exponent Attribute=true
    Create Position Attribute=true
    Create Radius Attribute=true
    Exponent=1.5
    Position=<20 -30 0>
    Radius=75
    Scale Color=20
   ) {enabled=false, module=FN_Light_Attributes001}
  ]
  rendererClasses [&NiagaraLightRendererProperties, &NiagaraSpriteRendererProperties]
  renderers [
   {rendererClass=&NiagaraLightRendererProperties, rendererIndex=0}
   {rendererClass=&NiagaraSpriteRendererProperties, rendererIndex=1}
  ]
 }
 Ray_Line {
  emitterUpdateScript [
   &EmitterState(Inactive Response=::NewEnumerator1, Life Cycle Mode=::NewEnumerator1, Loop Behavior=::NewEnumerator1, Loop Duration=1, UseLoopDelay=false)
   &SpawnBurst_Instantaneous(Spawn Count=2, Use Spawn Probability=true, UseLoopCountLimit=false)
  ]
  particleSpawnScript [
   &PartitionParticles()
   &InitializeParticle(
    Lifetime=1.5
    Mesh Scale=<0.6 -1.1 -1.1>
    Mesh Scale Mode=::NewEnumerator4
    Position Offset=<-3 0 0>
    Sprite Rotation Angle=&RandomRangeFloat(Minimum=0, Maximum=1)
    Sprite Rotation Mode=::NewEnumerator1
    UsePositionOffset=true
    Write Material Random=true
   )
   &InitialMeshOrientation(
    Mesh Orientation Mode=::NewEnumerator2
    Rotation=&SelectVectorFromArray(
     Direct Array Index=$Particles.PartitionParticles.Partition
     Vector Selection Array=DataInterface() {propertyValues {FloatData [<0 0 0>, <0.5 0 0>]}}
    )
    Use Rotation Vector=true
   )
  ]
  particleUpdateScript [
   &ParticleState()
   &UpdateMeshOrientation()
   &ScaleMeshSize(Scale Factor=&VectorFromFloat(Value=&FloatFromCurve(
      FloatCurve=DataInterface() {propertyValues {Curve {keys [
          {arriveTangent=2.3969, interpMode=RCIM_Cubic, leaveTangent=2.3969, tangentMode=RCTM_User, time=0, value=0}
          {arriveTangent=0.0007, interpMode=RCIM_Cubic, leaveTangent=0.0007, tangentMode=RCTM_User, time=0.2, value=0.35}
          {arriveTangent=-0.4123, interpMode=RCIM_Cubic, leaveTangent=-0.4123, tangentMode=RCTM_User, time=1, value=0.1}
         ]}}}
      Scale Curve=&RandomRangeFloat(Minimum=0.65, Maximum=0.65)
     )))
   &ScaleColor(
    Scale Alpha=&FloatFromCurve(
     FloatCurve=DataInterface() {propertyValues {Curve {keys [
         <at 0 0 RCIM_Cubic>
         {arriveTangent=2.8571, interpMode=RCIM_Cubic, leaveTangent=2.8571, time=0.4, value=1}
         {arriveTangent=0.0225, interpMode=RCIM_Cubic, leaveTangent=0.0225, tangentMode=RCTM_User, time=0.7, value=2}
         <at 1 0 RCIM_Cubic>
        ]}}}
     Scale Curve=10
    )
    Scale RGB=&Multiply_VectorByFloat(
     Vector=&MakeVectorFromLinearColorRGB(Color=&SelectLinearColorFromArray(
       Array Sampling Mode=::NewEnumerator1
       Direct Array Index=&MakeIntFromFloat(Float=$User.VariantSwitch)
       Color Selection Array=DataInterface() {propertyValues {ColorData [<rgba 1 1 1 1>, <rgba 1 1 1 1>, <rgba 2 0.4688 0.9755 1>, <rgba 0.4285 0.1645 1 1>, <rgba 2 1.0856 0.1443 1>]}}
       Random Seed=?
      ))
     Float=10
    )
    ScaleA=true
    ScaleRGB=true
   )
   /SetVariables(Particles.VisibilityTag=&MakeCustomIntFromBool(Bool=$User.bisHeadHidden, True INT=1, False INT=0)) {isSetParameters=true}
  ]
  rendererClasses [&NiagaraMeshRendererProperties]
  renderers [{rendererClass=&NiagaraMeshRendererProperties, rendererIndex=0}]
 }
 Rays {
  emitterUpdateScript [
   &EmitterState(
    Inactive Response=::NewEnumerator1
    Life Cycle Mode=::NewEnumerator1
    Loop Behavior=::NewEnumerator1
    Loop Delay=0.3
    Loop Duration=1
    UseLoopDelay=true
   )
   &SpawnBurst_Instantaneous(Spawn Count=2, Use Spawn Probability=false, UseLoopCountLimit=false)
  ]
  particleSpawnScript [
   &InitializeParticle(
    Color=<rgba 1 1 1 0.5>
    Lifetime=&RandomRangeFloat(Minimum=0.5, Maximum=0.6)
    Mesh Scale=<0.3 -1 -1>
    Mesh Scale Mode=::NewEnumerator4
    Sprite Rotation Angle=&RandomRangeFloat(Minimum=0, Maximum=1)
    Sprite Rotation Mode=::NewEnumerator1
    UsePositionOffset=true
    Write Material Random=true
   )
   &Color(Color=&MakeLinearColorFromVectorAndFloat("Vector (RGB)"=<0.1 0.002 0.12>, "Float (Alpha)"=&ClampFloat(Float=1)))
   &DynamicMaterialParameters(
    Index 0 Param 1=&RandomRangeFloat(Minimum=0, Maximum=10)
    Index 0 Param 2=&RandomRangeFloat(Minimum=2.5, Maximum=4.5)
    Index 0 Param 3=&RandomRangeFloat(Minimum=0.8, Maximum=1.6)
    Module.Param0WriteEnabled=true
    Module.Param1WriteEnabled=true
    Module.Param2WriteEnabled=true
    Module.Param3WriteEnabled=false
   )
  ]
  particleUpdateScript [
   &ParticleState()
   &ScaleMeshSize(Scale Factor=&VectorFromFloat(Value=&FloatFromCurve(
      FloatCurve=DataInterface() {propertyValues {Curve {keys [
          {arriveTangent=15.9108, interpMode=RCIM_Cubic, time=0, value=0.0062}
          {arriveTangent=1.323, interpMode=RCIM_Cubic, leaveTangent=1.323, time=0.2, value=0.6}
          {arriveTangent=-0.0562, interpMode=RCIM_Cubic, leaveTangent=-0.0562, tangentMode=RCTM_User, time=0.6, value=0.8}
          {arriveTangent=-0.9004, interpMode=RCIM_Cubic, leaveTangent=-0.9004, tangentMode=RCTM_User, time=1, value=0.4}
         ]}}}
      Scale Curve=&RandomRangeFloat(Minimum=0.5, Maximum=0.7)
     )))
   &ScaleColor(
    Scale Alpha=&FloatFromCurve(FloatCurve=DataInterface() {propertyValues {Curve {keys [
         <at 0 0 RCIM_Cubic>
         {arriveTangent=-2.0099, interpMode=RCIM_Cubic, leaveTangent=-2.0099, tangentMode=RCTM_User, time=0.3, value=1}
         <at 1 0 RCIM_Cubic>
        ]}}})
    Scale RGB=&MakeVectorFromLinearColorRGB(Color=&SelectLinearColorFromArray(
      Array Sampling Mode=::NewEnumerator1
      Direct Array Index=&MakeIntFromFloat(Float=$User.VariantSwitch)
      Color Selection Array=DataInterface() {propertyValues {ColorData [<rgba 1 1 1 1>, <rgba 1 1 1 1>, <rgba 1 1 1 1>, <rgba 1 1 1 1>, <rgba 1 1 1 1>]}}
      Random Seed=?
     ))
    ScaleA=true
    ScaleRGB=true
   )
   &DynamicMaterialParameters(
    Index 0 Param 4=1
    Module.Param0WriteEnabled=false
    Module.Param1WriteEnabled=false
    Module.Param2WriteEnabled=false
    Module.Param3WriteEnabled=false
   ) {enabled=false, module=DynamicMaterialParameters001}
   /SetVariables(Particles.VisibilityTag=&MakeCustomIntFromBool(Bool=$User.bisHeadHidden, True INT=1, False INT=0)) {isSetParameters=true}
  ]
  rendererClasses [&NiagaraMeshRendererProperties]
  renderers [{rendererClass=&NiagaraMeshRendererProperties, rendererIndex=0}]
 }
 Sprites_group {
  emitterUpdateScript [
   &EmitterState(
    Enable Distance Culling=true
    Inactive Response=::NewEnumerator1
    Life Cycle Mode=::NewEnumerator1
    Loop Behavior=::NewEnumerator1
    Loop Delay=&RandomRangeFloat(Minimum=0, Maximum=1)
    Loop Duration=1
    Max Distance Response=::KillImmediately
    Module.bUseMaxDistance=true
    Module.bUseMinDistance=false
    Scalability Mode=::NewEnumerator1
    UseLoopDelay=false
   )
   &SpawnBurst_Instantaneous(Spawn Count=6, Spawn Time=0.3, Use Spawn Probability=false, UseLoopCountLimit=false)
  ]
  particleSpawnScript [
   &InitializeParticle(
    Color=&MakeCustomLinearColorFromBool(
     Bool=&RandomBool()
     True Color=&SelectLinearColorFromArray(
      Array Sampling Mode=::NewEnumerator1
      Direct Array Index=&MakeIntFromFloat(Float=$User.VariantSwitch)
      Color Selection Array=DataInterface() {propertyValues {ColorData [<rgba 0.7658 1 0 1>, <rgba 0.7658 1 0 1>, <rgba 1.5 0.3132 0.7232 1>, <rgba 0.269 0.1119 0.8594 1>, <rgba 1.5 0.3323 0.0012 1>]}}
      Random Seed=?
     )
     False Color=&MakeCustomLinearColorFromBool(
      Bool=&RandomBool()
      True Color=&SelectLinearColorFromArray(
       Array Sampling Mode=::NewEnumerator1
       Direct Array Index=&MakeIntFromFloat(Float=$User.VariantSwitch)
       Color Selection Array=DataInterface() {propertyValues {ColorData [<rgba 0 0.7188 0.2139 1>, <rgba 0 0.7188 0.2139 1>, <rgba 0 0.6164 0.6979 1>, <rgba 0.269 0.1119 0.8594 1>, <rgba 1.5 0.3323 0.0012 1>]}}
       Random Seed=?
      )
      False Color=&SelectLinearColorFromArray(
       Array Sampling Mode=::NewEnumerator1
       Direct Array Index=&MakeIntFromFloat(Float=$User.VariantSwitch)
       Color Selection Array=DataInterface() {propertyValues {ColorData [<rgba 0 0.6164 0.6979 1>, <rgba 0 0.6164 0.6979 1>, <rgba 1.5 0.3132 0.7232 1>, <rgba 0 0.7188 0.2139 1>, <rgba 1 0.15 0.2521 1>]}}
       Random Seed=?
      )
     )
    )
    Lifetime Max=0.5
    Lifetime Min=0.4
    Lifetime Mode=::NewEnumerator1
    Sprite Rotation Mode=::NewEnumerator2
    Sprite Size Mode=::NewEnumerator1
    Uniform Sprite Size Min=6
    UsePositionOffset=false
    Write Material Random=true
   )
   &ShapeLocation(
    Cylinder Height=10
    Cylinder Radius=6
    MaskSpawnGroup=false
    Override Seed=false
    Rotation Mode=::NewEnumerator1
    Shape Origin=&VectorFromFloat(Value=0)
    Shape Primitive=::NewEnumerator1
    Surface Only=true
    Surface Only Band Thickness=0.1
    Use Endcaps In Surface Only Mode=false
    UseOwnerScale=true
   )
   &AddVelocity(
    Module.Use Falloff Distance=false
    Origin Offset=<-5 0 0>
    Override Seed=false
    Velocity Mode=::NewEnumerator1
    Velocity Speed=&RandomRangeFloat(Minimum=20, Maximum=35)
   )
   /SetVariables(Particles.VisibilityTag=&RandomRangeInt(Minimum=0, Maximum=3)) {isSetParameters=true}
   /SetVariables(Particles.Color_triangle:LinearColor=&MakeCustomLinearColorFromBool(
     Bool=&RandomBool()
     True Color=<rgba 0.7658 1 0 1>
     False Color=&MakeCustomLinearColorFromBool(Bool=&RandomBool(), True Color=<rgba 0 0.7188 0.1369 1>, False Color=<rgba 0 0.6348 0.7188 1>)
    )) {isSetParameters=true}
   /SetVariables(Particles.Color_triangleglow:LinearColor=&MultiplyLinearColor(
     Linear Color=$Particles.Color_triangle
     Scale Factor=&MakeVector4_FromLinearColor(LinearColor=/Niagara/DynamicInputs/TypeConversions/MakeLinearColorFromFloat(Value:NiagaraFloat=0.5))
    )) {isSetParameters=true}
  ]
  particleUpdateScript [
   &ParticleState()
   &SpriteRotationRate(Rotation Rate=&RandomRangeFloat(Minimum=-200, Maximum=200))
   &ScaleSpriteSize(Scale Factor=&Vector2DFromFloat(Value=&FloatFromCurve(FloatCurve=DataInterface() {propertyValues {Curve {keys [
          {interpMode=RCIM_Cubic, tangentMode=RCTM_User, time=0, value=0}
          {interpMode=RCIM_Cubic, tangentMode=RCTM_User, time=0.4, value=1}
          {tangentMode=RCTM_User, time=1, value=0}
         ]}}})))
   /SetVariables(Particles.Size002:Vector2f=&Multiply_Vector2D(A=$Particles.SpriteSize, B=&Vector2DFromFloat(Value=1.5))) {isSetParameters=true}
   /SetVariables(Particles.Size_Triangle:Vector2f=&Multiply_Vector2D(A=$Particles.SpriteSize, B=&Vector2DFromFloat(Value=0.8))) {isSetParameters=true}
   &Drag(Drag=&RandomRangeFloat(Minimum=0.5, Maximum=1), Use Linear Drag=true, Use Rotational Drag=true)
   &SolveForcesAndVelocity(Clamp Velocity=false, Limit Acceleration=false)
   &ScaleColor(
    Scale Alpha=&Multiply_Float(
     A=&FNTimeOfDayScalar(Night Scale=1, Morning Scale=0.3, Evening Scale=0.3)
     B=&FloatFromCurve(FloatCurve=DataInterface() {propertyValues {Curve {keys [
          {interpMode=RCIM_Cubic, tangentMode=RCTM_User, time=0.1, value=0}
          {interpMode=RCIM_Cubic, tangentMode=RCTM_User, time=0.3, value=0.8}
          {tangentMode=RCTM_User, time=0.8, value=0}
         ]}}})
    )
    ScaleA=true
    ScaleRGB=true
   )
   /SetVariables(Particles.Color002=&MultiplyLinearColor(Linear Color=<rgba 1 1 1 1>, Scale Factor=&MakeVector4_FromLinearColor(LinearColor=$Particles.Color))) {isSetParameters=true}
   /SetVariables(Particles.VisibilityTag=&MakeCustomIntFromBool(Bool=$User.bisHeadHidden, True INT=1, False INT=0)) {isSetParameters=true}
  ]
  rendererClasses [&NiagaraSpriteRendererProperties]
  renderers [
   {rendererClass=&NiagaraSpriteRendererProperties}
   {rendererClass=&NiagaraSpriteRendererProperties, rendererIndex=1}
   {rendererClass=&NiagaraSpriteRendererProperties, rendererIndex=2}
   {rendererClass=&NiagaraSpriteRendererProperties, rendererIndex=3}
  ]
 }
 Star_Glow {
  emitterUpdateScript [
   &EmitterState(
    Inactive Response=::NewEnumerator1
    Life Cycle Mode=::NewEnumerator1
    Loop Behavior=::NewEnumerator1
    Loop Delay=0.3
    Loop Duration=1
    UseLoopDelay=true
   )
   &SpawnBurst_Instantaneous(Use Spawn Probability=false, UseLoopCountLimit=false)
  ]
  particleSpawnScript [
   &InitializeParticle(
    Color=<rgba 1.5 1.5 1.5 1>
    Lifetime=0.3
    Sprite Size Mode=::NewEnumerator3
    Uniform Sprite Size=23
    UsePositionOffset=false
    Write Material Random=true
   )
   &SpriteFacingAndAlignment(Sprite Alignment=<0 0 1>, WriteAlignment=true, WriteFacing=true)
   &AddVelocity(Velocity=<5 0 0>)
  ]
  particleUpdateScript [
   &ParticleState()
   &ScaleSpriteSize(
    Uniform Curve Scale=1.05
    Uniform Curve Sprite Scale=DataInterface() {propertyValues {Curve {keys [{tangentMode=RCTM_User, time=0, value=1}, {tangentMode=RCTM_User, time=1, value=2}]}}}
   )
   &ScaleColor(
    Scale Alpha=&FloatFromCurve(
     FloatCurve=DataInterface() {propertyValues {Curve {keys [
         <at 0 0 RCIM_Cubic>
         {arriveTangent=-2.0518, interpMode=RCIM_Cubic, leaveTangent=-2.0518, tangentMode=RCTM_User, time=0.2, value=1}
         <at 1 0 RCIM_Cubic>
        ]}}}
     Scale Curve=0.7
    )
    Scale RGB=&MakeVectorFromLinearColorRGB(Color=&SelectLinearColorFromArray(
      Array Sampling Mode=::NewEnumerator1
      Direct Array Index=&MakeIntFromFloat(Float=$User.VariantSwitch)
      Color Selection Array=DataInterface() {propertyValues {ColorData [<rgba 0.1395 0.8333 0.5546 1>, <rgba 0.6152 0.8125 0.2621 1>, <rgba 1 0.5104 0.6724 1>, <rgba 0.4512 0.4405 1 1>, <rgba 1.5 0.8587 0.0756 1>]}}
      Random Seed=?
     ))
    ScaleA=true
    ScaleRGB=true
   )
   &SolveForcesAndVelocity(Clamp Velocity=false, Limit Acceleration=false)
   /SetVariables(Particles.VisibilityTag=&MakeCustomIntFromBool(Bool=$User.bisHeadHidden, True INT=1, False INT=0)) {isSetParameters=true}
   &DynamicMaterialParameters(
    Index 0 Param 2=&SelectFloatFromArray(
     Direct Array Index=&MakeIntFromFloat(Float=$User.VariantSwitch)
     Float Selection Array=DataInterface() {propertyValues {FloatData [0,1,1]}}
    )
    Module.Index0Float0Write=true
    Module.Index0Float1Write=true
    Module.Index0Float2Write=false
    Module.Index0Float3Write=false
   )
  ]
  rendererClasses [&NiagaraSpriteRendererProperties]
  renderers [{rendererClass=&NiagaraSpriteRendererProperties}]
 }
 Star_Glowmesh {
  emitterUpdateScript [
   &EmitterState(
    Inactive Response=::NewEnumerator1
    Life Cycle Mode=::NewEnumerator1
    Loop Behavior=::NewEnumerator1
    Loop Delay=0.3
    Loop Duration=1
    UseLoopDelay=true
   )
   &SpawnBurst_Instantaneous(Use Spawn Probability=false, UseLoopCountLimit=false)
  ]
  particleSpawnScript [
   &InitializeParticle(
    Color=&SelectLinearColorFromArray(
     Array Sampling Mode=::NewEnumerator1
     Direct Array Index=&MakeIntFromFloat(Float=$User.VariantSwitch)
     Color Selection Array=DataInterface() {propertyValues {ColorData [<rgba 1 1 1 1>, <rgba 1 1 1 1>, <rgba 1 1 1 1>, <rgba 0.026 1 0.0466 1>, <rgba 1.5 0.236 0.0637 1>]}}
     Random Seed=?
    )
    Lifetime=0.5
    Mesh Scale=<4.8 2.6 2.6>
    Mesh Scale Mode=::NewEnumerator4
    Sprite Size Mode=::NewEnumerator3
    Uniform Sprite Size=23
    UsePositionOffset=true
    Write Material Random=true
   )
   &InitialMeshOrientation(Mesh Orientation Mode=::NewEnumerator2, Use Rotation Vector=true)
  ]
  particleUpdateScript [
   &ParticleState()
   &ScaleMeshSize(Scale Factor=&MakeVector(
     X=&FloatFromCurve(FloatCurve=DataInterface() {propertyValues {Curve {keys [
          {arriveTangent=2.3556, interpMode=RCIM_Cubic, leaveTangent=2.3556, tangentMode=RCTM_User, time=0, value=0.3}
          {arriveTangent=0.0629, interpMode=RCIM_Cubic, leaveTangent=0.0629, tangentMode=RCTM_User, time=1, value=1}
         ]}}})
     Y=1
     Z=1
    ))
   &ScaleColor(
    Scale Alpha=&FloatFromCurve(
     FloatCurve=DataInterface() {propertyValues {Curve {keys [
         <at 0 0 RCIM_Cubic>
         {arriveTangent=-2.0518, interpMode=RCIM_Cubic, leaveTangent=-2.0518, tangentMode=RCTM_User, time=0.2, value=1}
         <at 1 0 RCIM_Cubic>
        ]}}}
     Scale Curve=&FN_FrontEndFloatModifier(Gameplay=1, Front End=0.7)
    )
    Scale RGB=&MakeVectorFromLinearColorRGB(Color=&MakeLinearColorFromVectorAndFloat(
      "Vector (RGB)"=&LerpMultipleVectors(
       Alpha:NiagaraFloat=$User.VariantSwitch
       V 1:Vector3f=&MakeVectorFromLinearColorRGB(Color=&ColorFromCurve(ColorCurve=DataInterface() {propertyValues {
           AlphaCurve {keys []}
           BlueCurve {keys [<at 0.201 0.002 RCIM_Cubic>, {arriveTangent=1.6431, interpMode=RCIM_Cubic, leaveTangent=1.6431, time=0.535, value=0}, <at 0.8084 1>]}
           GreenCurve {keys [<at 0.201 1 RCIM_Cubic>, {arriveTangent=-0.9364, interpMode=RCIM_Cubic, leaveTangent=-0.9364, time=0.535, value=1}, <at 0.8084 0.4312>]}
           RedCurve {keys [<at 0.201 0.9978 RCIM_Cubic>, {arriveTangent=-1.6426, interpMode=RCIM_Cubic, leaveTangent=-1.6426, time=0.535, value=0.0389}, <at 0.8084 0>]}
          }}))
       V 2:Vector3f=&MakeVectorFromLinearColorRGB(Color=&ColorFromCurve(ColorCurve=DataInterface() {propertyValues {
           AlphaCurve {keys []}
           BlueCurve {keys [<at 0.201 0.002 RCIM_Cubic>, {arriveTangent=1.6431, interpMode=RCIM_Cubic, leaveTangent=1.6431, time=0.535, value=0}, <at 0.8084 1>]}
           GreenCurve {keys [<at 0.201 1 RCIM_Cubic>, {arriveTangent=-0.9364, interpMode=RCIM_Cubic, leaveTangent=-0.9364, time=0.535, value=1}, <at 0.8084 0.4312>]}
           RedCurve {keys [<at 0.201 0.9978 RCIM_Cubic>, {arriveTangent=-1.6426, interpMode=RCIM_Cubic, leaveTangent=-1.6426, time=0.535, value=0.0389}, <at 0.8084 0>]}
          }}))
       V 3:Vector3f=&MakeVectorFromLinearColorRGB(Color=&ColorFromCurve(ColorCurve=DataInterface() {propertyValues {
           AlphaCurve {keys []}
           BlueCurve {keys [<at 0.5583 0.3724 RCIM_Cubic>,<at 0.8456 1 RCIM_Cubic>]}
           GreenCurve {keys [<at 0.5583 0.1771 RCIM_Cubic>,<at 0.8456 0.2514 RCIM_Cubic>]}
           RedCurve {keys [<at 0.5583 1 RCIM_Cubic>,<at 0.8456 0.0379 RCIM_Cubic>]}
          }}))
       V 4:Vector3f=&MakeVectorFromLinearColorRGB(Color=&ColorFromCurve(ColorCurve=DataInterface() {propertyValues {
           AlphaCurve {keys []}
           BlueCurve {keys [<at 0.3319 0.0908 RCIM_Cubic>,<at 0.5708 1 RCIM_Cubic>]}
           GreenCurve {keys [<at 0.3319 1 RCIM_Cubic>,<at 0.5708 0.0514 RCIM_Cubic>]}
           RedCurve {keys [<at 0.3319 0.0959 RCIM_Cubic>,<at 0.5708 0.2423 RCIM_Cubic>]}
          }}))
       V 5:Vector3f=&MakeVectorFromLinearColorRGB(Color=&ColorFromCurve(ColorCurve=DataInterface() {propertyValues {
           AlphaCurve {keys []}
           BlueCurve {keys [<at 0.3319 0.0537 RCIM_Cubic>,<at 0.5708 0.0348 RCIM_Cubic>]}
           GreenCurve {keys [<at 0.3319 0.6142 RCIM_Cubic>,<at 0.5708 0.1305 RCIM_Cubic>]}
           RedCurve {keys [<at 0.3319 1 RCIM_Cubic>,<at 0.5708 1 RCIM_Cubic>]}
          }}))
       Vector Count:ENiagara_MultipleLerpCount=::NewEnumerator2
      )
      "Float (Alpha)"=0.6
     ))
    ScaleA=true
    ScaleRGB=true
   )
   &SolveForcesAndVelocity(Clamp Velocity=false, Limit Acceleration=false)
   /SetVariables(Particles.VisibilityTag=&MakeCustomIntFromBool(Bool=$User.bisHeadHidden, True INT=1, False INT=0)) {isSetParameters=true}
  ]
  rendererClasses [&NiagaraMeshRendererProperties]
  renderers [{rendererClass=&NiagaraMeshRendererProperties, rendererIndex=0}]
 }
 Star_Main {
  emitterUpdateScript [
   &EmitterState(
    Inactive Response=::NewEnumerator1
    Life Cycle Mode=::NewEnumerator1
    Loop Behavior=::NewEnumerator1
    Loop Delay=0.3
    Loop Duration=1
    UseLoopDelay=true
   )
   &SpawnBurst_Instantaneous(Use Spawn Probability=true, UseLoopCountLimit=false)
  ]
  particleSpawnScript [
   &InitializeParticle(
    Color=<rgba 1.5 1.5 1.5 2>
    Lifetime=0.55
    Mesh Scale=&VectorFromFloat(Value=0.48)
    Mesh Scale Mode=::NewEnumerator4
    Sprite Size Mode=::NewEnumerator3
    Uniform Sprite Size=30
    UsePositionOffset=true
    Write Material Random=true
   )
   &AddVelocity(Velocity=<10 0 0>)
   &InitialMeshOrientation(Mesh Orientation Mode=::NewEnumerator2, Rotation=<0 0 0.25>, Use Rotation Vector=true)
   &SpriteFacingAndAlignment(Sprite Alignment=<0 0 1>, WriteAlignment=true, WriteFacing=true)
  ]
  particleUpdateScript [
   &ParticleState()
   &ScaleMeshSize(Scale Factor=&VectorFromFloat(Value=&FloatFromCurve(
      FloatCurve=DataInterface() {propertyValues {Curve {keys [
          {arriveTangent=1.2339, interpMode=RCIM_Cubic, leaveTangent=1.2339, tangentMode=RCTM_User, time=0, value=0.6}
          {arriveTangent=0.2683, interpMode=RCIM_Cubic, leaveTangent=0.2683, tangentMode=RCTM_User, time=1, value=1.5}
         ]}}}
      Scale Curve=0.6
     )))
   &ScaleSpriteSize(
    Uniform Curve Scale=1.05
    Uniform Curve Sprite Scale=DataInterface() {propertyValues {Curve {keys [
        {arriveTangent=0.8538, interpMode=RCIM_Cubic, leaveTangent=0.8538, tangentMode=RCTM_User, time=0, value=0.6}
        {arriveTangent=0.3287, interpMode=RCIM_Cubic, leaveTangent=0.3287, tangentMode=RCTM_User, time=1, value=1.2}
       ]}}}
   )
   &ScaleColor(
    Scale Alpha=&FloatFromCurve(
     FloatCurve=DataInterface() {propertyValues {Curve {keys [
         {arriveTangent=5.3825, interpMode=RCIM_Cubic, leaveTangent=5.3825, tangentMode=RCTM_User, time=0, value=0}
         {arriveTangent=6.6821, tangentMode=RCTM_User, time=0.15, value=2}
         {arriveTangent=-1.5609, interpMode=RCIM_Cubic, leaveTangent=-1.5609, tangentMode=RCTM_User, time=0.3, value=0.5}
         {arriveTangent=-0.2298, interpMode=RCIM_Cubic, leaveTangent=-0.2298, tangentMode=RCTM_User, time=1, value=0}
        ]}}}
     Scale Curve=&FN_FrontEndFloatModifier(Gameplay=1, Front End=0.8)
    )
    Scale RGB=&MakeVectorFromLinearColorRGB(Color=&SelectLinearColorFromArray(
      Array Sampling Mode=::NewEnumerator1
      Direct Array Index=&MakeIntFromFloat(Float=$User.VariantSwitch)
      Color Selection Array=DataInterface() {propertyValues {ColorData [<rgba 0.4078 1.5 0.0415 1>, <rgba 0.4078 1.5 0.0415 1>, <rgba 1 0.2344 0.4877 1>, <rgba 0.303 0.1771 1 1>, <rgba 1.5 0.236 0.0637 1>]}}
      Random Seed=?
     ))
    ScaleA=true
    ScaleRGB=true
   )
   &SolveForcesAndVelocity(Clamp Velocity=false, Limit Acceleration=false)
   &DynamicMaterialParameters(
    Index 0 Param 1=&FloatFromCurve(FloatCurve=DataInterface() {propertyValues {Curve {keys [{arriveTangent=0.2079, time=0.1, value=0}, {arriveTangent=0.7869, leaveTangent=0.7869, time=1, value=1}]}}})
    Index 0 Param 2=&SelectFloatFromArray(
     Direct Array Index=&MakeIntFromFloat(Float=$User.VariantSwitch)
     Float Selection Array=DataInterface() {propertyValues {FloatData [0,0,1,1,1]}}
    )
    Index 0 Param 3=&FloatFromCurve(FloatCurve=DataInterface() {propertyValues {Curve {keys [<at 0 5>,<at 1 1.5>]}}})
    Index 0 Param 4=&FloatFromCurve(FloatCurve=DataInterface() {propertyValues {Curve {keys [<at 0 5>,<at 1 5>]}}})
    Module.Index0Param0Write=true
    Module.Index0Param1Write=true
    Module.Index0Param2Write=false
    Module.Index0Param3Write=false
   )
   /SetVariables(Particles.Dynamic002:Vector4f=&MakeVector4(
     X=&FloatFromCurve(FloatCurve=DataInterface() {propertyValues {Curve {keys [
          {arriveTangent=2.0325, interpMode=RCIM_Cubic, leaveTangent=2.0325, tangentMode=RCTM_User, time=0.4, value=0}
          {arriveTangent=0.0327, interpMode=RCIM_Cubic, leaveTangent=0.0327, tangentMode=RCTM_User, time=0.7, value=1.5}
         ]}}})
     Y=&SelectFloatFromArray(
      Direct Array Index=&MakeIntFromFloat(Float=$User.VariantSwitch)
      Float Selection Array=DataInterface() {propertyValues {FloatData [0,0,1]}}
     )
     Z=0
     W=0
    )) {isSetParameters=true}
   /SetVariables(Particles.VisibilityTag=&MakeCustomIntFromBool(Bool=$User.bisHeadHidden, True INT=1, False INT=0)) {isSetParameters=true}
  ]
  rendererClasses [&NiagaraMeshRendererProperties, &NiagaraSpriteRendererProperties]
  renderers [
   {rendererClass=&NiagaraMeshRendererProperties, rendererIndex=0}
   {rendererClass=&NiagaraSpriteRendererProperties, rendererIndex=1}
  ]
 }
 Star_mesh {
  emitterUpdateScript [
   &EmitterState(Inactive Response=::NewEnumerator1, Life Cycle Mode=::NewEnumerator1, Loop Behavior=::NewEnumerator1, Loop Duration=1, UseLoopDelay=false)
   &SpawnBurst_Instantaneous(Spawn Time=0.3, Use Spawn Probability=false, UseLoopCountLimit=false)
  ]
  particleSpawnScript [
   &InitializeParticle(
    Color Mode=::NewEnumerator0
    Lifetime=0.6
    Mesh Scale=<10 2.5 2.5>
    Mesh Scale Mode=::NewEnumerator4
    UsePositionOffset=true
    Write Material Random=true
   )
   &InitialMeshOrientation(Mesh Orientation Mode=::NewEnumerator2, Rotation=<0 10 0>, Use Rotation Vector=true)
  ]
  particleUpdateScript [
   &ParticleState()
   /SetVariables(Particles.Size003:Vector3f=&Multiply_Vector(
     A=<0.5 0.5 0.5>
     B=&VectorFromFloat(Value=&FloatFromCurve(
       FloatCurve=DataInterface() {propertyValues {Curve {keys [{arriveTangent=1.8728, interpMode=RCIM_Cubic, leaveTangent=1.8728, tangentMode=RCTM_User, time=0, value=1}, <at 1 2 RCIM_Cubic>]}}}
       Scale Curve=0.95
      ))
    )) {isSetParameters=true}
   /SetVariables(Particles.Color003:LinearColor=&MultiplyLinearColor(
     Linear Color=&SelectLinearColorFromArray(
      Array Sampling Mode=::NewEnumerator1
      Direct Array Index=&MakeIntFromFloat(Float=$User.VariantSwitch)
      Color Selection Array=DataInterface() {propertyValues {ColorData [<rgba 1.2 0.7326 0.1114 1>, <rgba 1.2 0.7326 0.1114 1>, <rgba 0.1393 0.4126 1.2 1>, <rgba 0.6401 0.0671 1.5 1>, <rgba 1.2 0.5275 0.0668 1>]}}
      Random Seed=?
     )
     Scale Factor=&MakeVector4_FromFloat(Float=&FloatFromCurve(
       FloatCurve=DataInterface() {propertyValues {Curve {keys [
           {arriveTangent=0.9513, interpMode=RCIM_Cubic, leaveTangent=0.9513, tangentMode=RCTM_User, time=0, value=0}
           {arriveTangent=10.1951, interpMode=RCIM_Cubic, leaveTangent=-5.7722, tangentMode=RCTM_Break, time=0.1, value=1.5}
           {arriveTangent=-1.8916, interpMode=RCIM_Cubic, leaveTangent=-1.8916, tangentMode=RCTM_User, time=0.3, value=0.575}
           {arriveTangent=-0.2689, interpMode=RCIM_Cubic, leaveTangent=-0.2689, tangentMode=RCTM_User, time=1, value=0}
          ]}}}
       Scale Curve=&FN_FrontEndFloatModifier(Gameplay=1, Front End=0.75)
      ))
    )) {isSetParameters=true}
   &DynamicMaterialParameters(
    Index 0 Param 1=&FloatFromCurve(FloatCurve=DataInterface() {propertyValues {Curve {keys [
         {arriveTangent=-4.0585, interpMode=RCIM_Cubic, leaveTangent=-4.0585, tangentMode=RCTM_User, time=0, value=1}
         {arriveTangent=-1.0113, interpMode=RCIM_Cubic, leaveTangent=-1.0113, tangentMode=RCTM_User, time=1, value=-1.5}
        ]}}})
    Module.Index0Param0Write=true
    Module.Index0Param1Write=true
    Module.Index0Param2Write=true
    Module.Index0Param3Write=true
   )
   /SetVariables(Particles.VisibilityTag=&MakeCustomIntFromBool(Bool=$User.bisHeadHidden, True INT=1, False INT=0)) {isSetParameters=true}
  ]
  rendererClasses [&NiagaraMeshRendererProperties]
  renderers [{rendererClass=&NiagaraMeshRendererProperties, rendererIndex=0}]
 }
}
summary {
 systemName=NS_Character_LocalTalent_Flare_HeadBurst
 emitters [
  {emitterName=Star_Glow, rendererClasses [&NiagaraSpriteRendererProperties]}
  {emitterName=Rays, rendererClasses [&NiagaraMeshRendererProperties]}
  {emitterName=Ray_Line, rendererClasses [&NiagaraMeshRendererProperties]}
  {emitterName=Star_mesh, rendererClasses [&NiagaraMeshRendererProperties]}
  {emitterName=Light, rendererClasses [&NiagaraLightRendererProperties, &NiagaraSpriteRendererProperties]}
  {emitterName=Star_Main, rendererClasses [&NiagaraMeshRendererProperties, &NiagaraSpriteRendererProperties]}
  {emitterName=Sprites_group, rendererClasses [&NiagaraSpriteRendererProperties]}
  {emitterName=Star_Glowmesh, rendererClasses [&NiagaraMeshRendererProperties]}
 ]
 userVariables [
  User.bIsFrontEnd {type=NiagaraBool}
  User.bisFrontEndPreview {type=NiagaraBool}
  User.bisHeadHidden {type=NiagaraBool}
  User.Hide light {type=NiagaraBool}
  User.IsLeft {type=NiagaraBool}
  User.VariantSwitch {type=NiagaraFloat, defaultValue=0, description=""}
 ]
}
