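# Idle cosmetic effect for the LocalTalent Flare character, adding glowing lights, sparkle sprites and mesh trails around the head.
# Emitters follow skeletal mesh sockets, using partition indices for paired left/right placement; another trail shifts via PocketTrail_Location.
# Trails pass time-smoothed positions into a dynamic variable, and sprite spawning depends on owner speed.
# User.VariantSwitch selects palettes, while head-hidden, BigHeadMode and front-end preview inputs adjust visibility, scale and brightness.
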NiagaraSystem /BRCosmetics/Characters/Player/Female/Medium/Bodies/F_MED_LocalTalent_Flare/FX/NS_Character_LocalTalent_Flare_Idle
dependencies {
 usedDataInterfaces [
  &NiagaraDataInterfacePlatformSet
  &NiagaraDataInterfaceArrayFloat3
  &NiagaraDataInterfaceArrayColor
  &NiagaraDataInterfaceCurve
  &NiagaraDataInterfaceArrayFloat
  &NiagaraDataInterfaceColorCurve
  &NiagaraDataInterfaceSkeletalMesh
 ]
 usedDynamicInputs [&SimulationPosition]
 usedModules [
  &self:SystemSpawnScript.SystemScriptSource.SystemScriptGraph.NiagaraNodeAssignment_0.SetVariables_#1
  &SystemState
 ]
}
emitters {
 Head_Light {
  emitterUpdateScript [
   &EmitterState(
    Inactive Response=::NewEnumerator1
    Life Cycle Mode=::NewEnumerator1
    Loop Behavior=::NewEnumerator1
    Loop Duration Mode=::NewEnumerator1
    UseLoopDelay=false
   )
   &SpawnBurst_Instantaneous(Spawn Count=2, Use Spawn Probability=false, UseLoopCountLimit=false)
  ]
  particleSpawnScript [
   &PartitionParticles()
   &InitializeParticle(Color Mode=::NewEnumerator0, Lifetime=1.5, Sprite Size Mode=::NewEnumerator3, UsePositionOffset=false, Write Material Random=true)
  ]
  particleUpdateScript [
   &ParticleState(
    Kill Particles When Lifetime Has Elapsed=false
    Let Infinitely Lived Particles Die When Emitter Deactivates=true
    Loop Particles Lifetime=true
   )
   &SkeletalMeshLocation(
    Bone Sampling Mode=::NewEnumerator4
    MaskSpawnGroup=false
    Sampled Position Offset=&Multiply_VectorByFloat(
     Vector=&SelectVectorFromArray(
      Direct Array Index=$Particles.PartitionParticles.Partition
      Vector Selection Array=DataInterface() {propertyValues {FloatData [<3 0 15>, <3 0 -15>]}}
     )
     Float=&Lerp_Float(A=1, B=2, Alpha=$NPC.FortniteNPC.BigHeadMode)
    )
    Skeletal Mesh=$System.Skeletal Mesh Reader
   )
   &FN_Light_Attributes(
    Color=&MakeLinearColorFromVectorAndFloat(
     "Vector (RGB)"=/Niagara/DynamicInputs/Math/LerpMultipleVectors(
      Alpha:NiagaraFloat=$User.VariantSwitch
      V 1:Vector3f=&MakeVectorFromLinearColorRGB(Color=&ColorFromCurve(ColorCurve=DataInterface() {propertyValues {
          AlphaCurve {keys []}
          BlueCurve {keys [
            <at 0 0>
            <at 0.1383 0>
            <at 0.1966 0>
            <at 0.3441 0>
            <at 0.3987 0>
            {arriveTangent=2.8924, leaveTangent=2.8924, time=0.5463, value=0.886}
            <at 0.7348 0.886>
            <at 0.8948 0>
            <at 0.9308 0>
            <at 1 0>
           ]}
          GreenCurve {keys [<at 0 0>,<at 0.1383 1>,<at 0.1966 1>,<at 0.3441 1>,<at 0.3987 1>,<at 0.5463 1>,<at 0.7348 1>,<at 0.8948 1>,<at 0.9308 1>,<at 1 0>]}
          RedCurve {keys [
            <at 0 0>
            <at 0.1383 0.0226>
            <at 0.1966 0.0226>
            <at 0.3441 0.8122>
            <at 0.3987 0.8122>
            {arriveTangent=-2.6514, leaveTangent=-2.6514, time=0.5463, value=0}
            <at 0.7348 0>
            <at 0.8948 0.0115>
            <at 0.9308 0.0226>
            <at 1 0>
           ]}
         }}))
      V 2:Vector3f=&MakeVectorFromLinearColorRGB(Color=&ColorFromCurve(ColorCurve=DataInterface() {propertyValues {
          AlphaCurve {keys []}
          BlueCurve {keys [
            <at 0 0>
            <at 0.1383 0>
            <at 0.1966 0>
            <at 0.3441 0>
            <at 0.3987 0>
            {arriveTangent=2.8924, leaveTangent=2.8924, time=0.5463, value=0.886}
            <at 0.7348 0.886>
            <at 0.8948 0>
            <at 0.9308 0>
            <at 1 0>
           ]}
          GreenCurve {keys [<at 0 0>,<at 0.1383 1>,<at 0.1966 1>,<at 0.3441 1>,<at 0.3987 1>,<at 0.5463 1>,<at 0.7348 1>,<at 0.8948 1>,<at 0.9308 1>,<at 1 0>]}
          RedCurve {keys [
            <at 0 0>
            <at 0.1383 0.0226>
            <at 0.1966 0.0226>
            <at 0.3441 0.8122>
            <at 0.3987 0.8122>
            {arriveTangent=-2.6514, leaveTangent=-2.6514, time=0.5463, value=0}
            <at 0.7348 0>
            <at 0.8948 0.0115>
            <at 0.9308 0.0226>
            <at 1 0>
           ]}
         }}))
      V 3:Vector3f=&MakeVectorFromLinearColorRGB(Color=&ColorFromCurve(ColorCurve=DataInterface() {propertyValues {
          AlphaCurve {keys []}
          BlueCurve {keys [
            <at 0 0>
            <at 0.1383 0.5138>
            <at 0.1966 0.5392>
            <at 0.3441 1>
            <at 0.3987 1>
            {arriveTangent=2.8924, leaveTangent=2.8924, time=0.5463, value=0.886}
            <at 0.7348 0.886>
            <at 0.8948 0.5361>
            <at 0.9308 0.4659>
            <at 1 0>
           ]}
          GreenCurve {keys [<at 0 0>,<at 0.1383 0.3199>,<at 0.1966 0.2344>,<at 0.3441 0.2797>,<at 0.3987 0.2596>,<at 0.5463 1>,<at 0.7348 1>,<at 0.8948 0.1904>,<at 0.9308 0.2029>,<at 1 0>]}
          RedCurve {keys [
            <at 0 0>
            <at 0.1383 1>
            <at 0.1966 1>
            <at 0.3441 0.4759>
            <at 0.3987 0.4265>
            {arriveTangent=-2.6514, leaveTangent=-2.6514, time=0.5463, value=0}
            <at 0.7348 0>
            <at 0.8948 1>
            <at 0.9308 1>
            <at 1 0>
           ]}
         }}))
      V 4:Vector3f=&MakeVectorFromLinearColorRGB(Color=&ColorFromCurve(ColorCurve=DataInterface() {propertyValues {
          AlphaCurve {keys []}
          BlueCurve {keys [
            <at 0 0>
            <at 0.1383 0.1292>
            <at 0.1966 0.0966>
            <at 0.305 1>
            <at 0.3987 1>
            {arriveTangent=2.8924, leaveTangent=2.8924, time=0.5463, value=0.886}
            <at 0.6582 0.886>
            <at 0.7417 1>
            <at 0.9308 1>
            <at 1 0>
           ]}
          GreenCurve {keys [<at 0 0>,<at 0.1383 1>,<at 0.1966 1>,<at 0.305 0.1547>,<at 0.3987 0.1203>,<at 0.5463 1>,<at 0.6582 1>,<at 0.7417 0.1995>,<at 0.9308 0.1535>,<at 1 0>]}
          RedCurve {keys [
            <at 0 0>
            <at 0.1383 0.1116>
            <at 0.1966 0.1526>
            <at 0.305 0.4236>
            <at 0.3987 0.4356>
            {arriveTangent=-2.6514, leaveTangent=-2.6514, time=0.5463, value=0}
            <at 0.6582 0>
            <at 0.7417 0.4264>
            <at 0.9308 0.3741>
            <at 1 0>
           ]}
         }}))
      V 5:Vector3f=&MakeVectorFromLinearColorRGB(Color=&ColorFromCurve(ColorCurve=DataInterface() {propertyValues {
          AlphaCurve {keys []}
          BlueCurve {keys [
            <at 0 0>
            <at 0.1383 0.0254>
            <at 0.1966 0>
            <at 0.305 0.0582>
            <at 0.3987 0.0598>
            {arriveTangent=2.8924, leaveTangent=2.8924, time=0.5463, value=0.0659}
            <at 0.7953 0.0226>
            <at 0.8773 1>
            <at 0.9308 1>
            <at 1 0>
           ]}
          GreenCurve {keys [<at 0 0>,<at 0.1383 0.3206>,<at 0.1966 0.1186>,<at 0.305 0.7173>,<at 0.3987 0.8732>,<at 0.5463 0.3982>,<at 0.7953 0.1375>,<at 0.8773 0.1995>,<at 0.9308 0.1535>,<at 1 0>]}
          RedCurve {keys [
            <at 0 0>
            <at 0.1383 1>
            <at 0.1966 1>
            <at 0.305 1>
            <at 0.3987 1>
            {arriveTangent=-2.6514, leaveTangent=-2.6514, time=0.5463, value=1}
            <at 0.7953 1>
            <at 0.8773 0.4264>
            <at 0.9308 0.3741>
            <at 1 0>
           ]}
         }}))
      Vector Count:ENiagara_MultipleLerpCount=::NewEnumerator2
     )
     "Float (Alpha)"=1
    )
    Create Color Attribute=true
    Create Exponent Attribute=true
    Create Position Attribute=true
    Create Radius Attribute=true
    Enable Particle Light Enable=true
    Exponent=1.5
    Light Enabled:NiagaraBool=&Boolean_NotOperation(Bool=$User.bisHeadHidden)
    Position=&AddVectorToPosition(
     Position=$Particles.Position
     Vector=&TransformVector(
      Vector=&SelectVectorFromArray(
       Direct Array Index=$Particles.PartitionParticles.Partition
       Vector Selection Array=DataInterface() {propertyValues {FloatData [<0 0 0>, <0 0 0>]}}
      )
      Destination Space=::Simulation
     )
    )
    Radius=&Lerp_Float(A=15, B=30, Alpha=$NPC.FortniteNPC.BigHeadMode)
    Scale Color=10
   )
  ]
  rendererClasses [&NiagaraLightRendererProperties]
  renderers [{rendererClass=&NiagaraLightRendererProperties, rendererIndex=0}]
 }
 Head_Sprites {
  emitterUpdateScript [
   &EmitterState(
    Enable Distance Culling=true
    Inactive Response=::NewEnumerator1
    Life Cycle Mode=::NewEnumerator1
    Loop Delay=&RandomRangeFloat(Minimum=0, Maximum=1)
    Max Distance Response=::KillImmediately
    Module.bUseMaxDistance=true
    Module.bUseMinDistance=false
    Scalability Mode=::NewEnumerator1
    UseLoopDelay=false
   )
   &SpawnRate(
    Spawn Probability=&MakeCustomFloatFromBool(
     Bool=&SetBoolByFloatComparison(A=&VectorLength(Vector=$Engine.Owner.Velocity), B=1000)
     True Float=0
     False Float=1
    )
    SpawnRate=&MakeCustomFloatFromBool(Bool=$User.bisHeadHidden, True Float=0, False Float=12)
    Use Spawn Probability=true
   )
   &SpawnPerUnit(Module.Use Max Movement Threshold=true, Module.Use Movement Tolerance=false, Spawn Spacing=40, Use Spawn Probability=false) {enabled=false}
  ]
  particleSpawnScript [
   &PartitionParticles()
   &InitializeParticle(
    Color=&MakeCustomLinearColorFromBool(
     Bool=&RandomBool()
     True Color=&SelectLinearColorFromArray(
      Array Sampling Mode=::NewEnumerator1
      Direct Array Index=&MakeIntFromFloat(Float=$User.VariantSwitch)
      Color Selection Array=DataInterface() {propertyValues {ColorData [<rgba 0.7658 1 0 1>, <rgba 0.7658 1 0 1>, <rgba 1.5 0.3132 0.7232 1>, <rgba 0.2141 0.0573 1 1>, <rgba 1 0.3428 0.0513 1>]}}
      Random Seed=?
     )
     False Color=&MakeCustomLinearColorFromBool(Bool=&RandomBool(), True Color=<rgba 0 0.7188 0.2139 1>, False Color=<rgba 0 0.6164 0.6979 1>)
    )
    Lifetime Max=0.36
    Lifetime Min=0.35
    Lifetime Mode=::NewEnumerator1
    Sprite Rotation Mode=::NewEnumerator2
    Sprite Size Mode=::NewEnumerator1
    Uniform Sprite Size Max=5
    Uniform Sprite Size Min=4
    UsePositionOffset=false
    Write Material Random=true
   )
   &ShapeLocation(
    MaskSpawnGroup=false
    Override Seed=false
    Rotation Axis=<0 1 0>
    Rotation Mode=::NewEnumerator1
    Shape Origin=&VectorFromFloat(Value=0)
    Sphere Radius=5
    UseOwnerScale=true
   )
   /SetVariables(Particles.VisibilityTag=&RandomRangeInt(Minimum=0, Maximum=3)) {isSetParameters=true, module=SetVariables_#2}
   /SetVariables(Particles.Color_triangle:LinearColor=&MakeCustomLinearColorFromBool(
     Bool=&RandomBool()
     True Color=<rgba 0.7658 1 0 1>
     False Color=&MakeCustomLinearColorFromBool(Bool=&RandomBool(), True Color=<rgba 0 0.7188 0.1369 1>, False Color=<rgba 0 0.6348 0.7188 1>)
    )) {isSetParameters=true, module=SetVariables_#3}
   /SetVariables(Particles.Color_triangleglow:LinearColor=&MultiplyLinearColor(
     Linear Color=$Particles.Color_triangle
     Scale Factor=&MakeVector4_FromLinearColor(LinearColor=/Niagara/DynamicInputs/TypeConversions/MakeLinearColorFromFloat(Value:NiagaraFloat=0.5))
    )) {isSetParameters=true, module=SetVariables_#4}
  ]
  particleUpdateScript [
   &ParticleState()
   &SkeletalMeshLocation(
    Bone Sampling Mode=::NewEnumerator4
    MaskSpawnGroup=false
    Sampled Position Offset=&Multiply_VectorByFloat(
     Vector=&AddVectorToPosition(
      Position=$Particles.ShapeLocation.ShapePosition
      Vector=&SelectVectorFromArray(
       Direct Array Index=$Particles.PartitionParticles.Partition
       Vector Selection Array=DataInterface() {propertyValues {FloatData [<5 -8 -13>, <4 -8 15>]}}
      )
     )
     Float=&Lerp_Float(A=1, B=2, Alpha=$NPC.FortniteNPC.BigHeadMode)
    )
    Skeletal Mesh=$System.Skeletal Mesh Reader
   )
   &ScaleSpriteSize(Scale Factor=&Vector2DFromFloat(Value=&FloatFromCurve(
      FloatCurve=DataInterface() {propertyValues {Curve {keys [
          {interpMode=RCIM_Cubic, tangentMode=RCTM_User, time=0, value=0}
          {interpMode=RCIM_Cubic, tangentMode=RCTM_User, time=0.4, value=1}
          {tangentMode=RCTM_User, time=1, value=0}
         ]}}}
      Scale Curve=&Lerp_Float(A=1, B=2, Alpha=$NPC.FortniteNPC.BigHeadMode)
     )))
   /SetVariables(Particles.Size002:Vector2f=&Multiply_Vector2D(A=$Particles.SpriteSize, B=&Vector2DFromFloat(Value=1.3))) {isSetParameters=true, module=SetVariables_#5}
   /SetVariables(Particles.Size_Triangle:Vector2f=&Multiply_Vector2D(A=$Particles.SpriteSize, B=&Vector2DFromFloat(Value=0.6))) {isSetParameters=true, module=SetVariables_#6}
   &Drag(Use Linear Drag=true, Use Rotational Drag=true)
   &SolveForcesAndVelocity(Clamp Velocity=false, Limit Acceleration=false)
   &ScaleColor(
    Scale Alpha=&Multiply_Float(
     A=&FNTimeOfDayScalar(Night Scale=1, Morning Scale=0.3, Evening Scale=0.3)
     B=&FloatFromCurve(FloatCurve=DataInterface() {propertyValues {Curve {keys [
          {interpMode=RCIM_Cubic, tangentMode=RCTM_User, time=0.1, value=0}
          {interpMode=RCIM_Cubic, tangentMode=RCTM_User, time=0.3, value=0.8}
          {tangentMode=RCTM_User, time=0.8, value=0}
         ]}}})
    )
    Scale RGB=&MakeVectorFromLinearColorRGB(Color=&SelectLinearColorFromArray(
      Array Sampling Mode=::NewEnumerator1
      Direct Array Index=&MakeIntFromFloat(Float=$User.VariantSwitch)
      Color Selection Array=DataInterface() {propertyValues {ColorData [<rgba 1 1 1 1>, <rgba 0 0.6174 1 1>, <rgba 1 1 1 1>]}}
      Random Seed=?
     ))
    ScaleA=true
    ScaleRGB=true
   )
   /SetVariables(Particles.Color002=&MultiplyLinearColor(Linear Color=<rgba 1.2 1.2 1.2 1>, Scale Factor=&MakeVector4_FromLinearColor(LinearColor=$Particles.Color))) {isSetParameters=true, module=SetVariables_#7}
  ]
  rendererClasses [&NiagaraSpriteRendererProperties]
  renderers [
   {rendererClass=&NiagaraSpriteRendererProperties}
   {rendererClass=&NiagaraSpriteRendererProperties, rendererIndex=1}
   {rendererClass=&NiagaraSpriteRendererProperties, rendererIndex=2}
   {rendererClass=&NiagaraSpriteRendererProperties, rendererIndex=3}
  ]
 }
 Head_Trail {
  emitterUpdateScript [
   &EmitterState(
    Inactive Response=::NewEnumerator1
    Life Cycle Mode=::NewEnumerator1
    Loop Behavior=::NewEnumerator1
    Loop Duration Mode=::NewEnumerator1
    UseLoopDelay=false
   )
   &SpawnBurst_Instantaneous(Spawn Count=2, Use Spawn Probability=false, UseLoopCountLimit=false)
   /SetVariables(Emitter.LinearMaskLeft_Power:NiagaraFloat=&FN_PickFloatDependOnPlatform(TargetPlatformActive=1, TargetPlatformNoActive=2)) {isSetParameters=true}
  ]
  particleSpawnScript [
   &PartitionParticles()
   &InitializeParticle(Color=<rgba 1 1 1 5>, Lifetime=1.5, Mesh Scale Mode=::NewEnumerator4, UsePositionOffset=true, Write Material Random=true)
   &CameraOffset(Camera Offset Amount=&FN_PickFloatDependOnPlatform(TargetPlatformActive=0, TargetPlatformNoActive=5))
   &InitialMeshOrientation(
    Mesh Orientation Mode=::NewEnumerator2
    Rotation=&SelectVectorFromArray(
     Direct Array Index=$Particles.PartitionParticles.Partition
     Vector Selection Array=DataInterface() {propertyValues {FloatData [<0 0.25 0>, <0 0.25 0>]}}
    )
    Use Rotation Vector=true
   ) {module=InitialMeshOrientation001}
   &InitialMeshOrientation(
    Mesh Orientation Mode=::NewEnumerator2
    Rotation=&SelectVectorFromArray(
     Direct Array Index=$Particles.PartitionParticles.Partition
     Vector Selection Array=DataInterface() {propertyValues {FloatData [<0.1 0 0>, <-0.1 0 0.05>]}}
    )
    Use Rotation Vector=true
   ) {module=InitialMeshOrientation002}
   /SetVariables(Particles.Size_Orange:Vector3f=&SelectVectorFromArray(
     Direct Array Index=$Particles.PartitionParticles.Partition
     Vector Selection Array=DataInterface() {propertyValues {FloatData [<-0.3 0.2 0.6>, <0.3 0.2 0.65>]}}
    )) {isSetParameters=true, module=SetVariables_#8}
   /SetVariables(Particles.Size_D:Vector3f=&SelectVectorFromArray(
     Direct Array Index=$Particles.PartitionParticles.Partition
     Vector Selection Array=DataInterface() {propertyValues {FloatData [<-0.1 0.25 0.5>, <0.1 0.25 0.5>]}}
    )) {isSetParameters=true, module=SetVariables_#9}
   /SetVariables(Particles.Color_D:LinearColor=<rgba 0 1 0.277 1.1>) {isSetParameters=true, module=SetVariables_#10}
   /SetVariables(Particles.Size_Green:Vector3f=&SelectVectorFromArray(
     Direct Array Index=$Particles.PartitionParticles.Partition
     Vector Selection Array=DataInterface() {propertyValues {FloatData [<0.25 0.2 0.4>, <0.25 0.2 0.4>]}}
    )) {isSetParameters=true, module=SetVariables_#11}
  ]
  particleUpdateScript [
   &ParticleState(Kill Particles When Lifetime Has Elapsed=false, Let Infinitely Lived Particles Die When Emitter Deactivates=true)
   &ScaleColor(
    Scale Alpha=&MakeCustomFloatFromBool(
     Bool=$User.bisFrontEndPreview
     True Float=&Multiply_Float(
      A=0.2
      B=&FloatFromCurve(FloatCurve=DataInterface() {propertyValues {Curve {keys [<at -0.0012 0>,<at 0.3 0>,<at 1 1>]}}})
     )
     False Float=0.33
    )
    Scale RGB=&MakeVectorFromLinearColorRGB(Color=&SelectLinearColorFromArray(
      Array Sampling Mode=::NewEnumerator1
      Direct Array Index=&MakeIntFromFloat(Float=$User.VariantSwitch)
      Color Selection Array=DataInterface() {propertyValues {ColorData [<rgba 1 1 1 1>, <rgba 0.4 0.4 0.4 1>, <rgba 0.4 0.4 0.4 1>, <rgba 0.4 0.4 0.4 1>, <rgba 0.7552 0.7552 0.7552 1>]}}
      Random Seed=?
     ))
    ScaleA=true
    ScaleRGB=true
   )
   /SetVariables(
    Particles.Size_D:Vector3f=&Multiply_VectorByFloat(Vector=$Particles.Initial.Size_D, Float=&Lerp_Float(A=1, B=2, Alpha=$NPC.FortniteNPC.BigHeadMode))
    Particles.Size_Green:Vector3f=&Multiply_VectorByFloat(Vector=$Particles.Initial.Size_Green, Float=&Lerp_Float(A=1, B=2, Alpha=$NPC.FortniteNPC.BigHeadMode))
    Particles.Size_Orange:Vector3f=&Multiply_VectorByFloat(Vector=$Particles.Initial.Size_Orange, Float=&Lerp_Float(A=1, B=2, Alpha=$NPC.FortniteNPC.BigHeadMode))
   ) {isSetParameters=true}
   /SetVariables(Particles.ColorPoint:LinearColor=&MakeLinearColorFromVectorAndFloat(
     "Vector (RGB)"=&MakeVectorFromLinearColorRGB(Color=&SelectLinearColorFromArray(
       Array Sampling Mode=::NewEnumerator1
       Direct Array Index=&MakeIntFromFloat(Float=$User.VariantSwitch)
       Color Selection Array=DataInterface() {propertyValues {ColorData [<rgba 0.0598 1 0.6264 10>, <rgba 0.8854 1 0.1078 10>, <rgba 1.2471 0.7109 1.5 10>, <rgba 0.3333 0.8124 2 10>, <rgba 2 1.6462 0 10>]}}
       Random Seed=?
      ))
     "Float (Alpha)"=&MakeCustomFloatFromBool(
      Bool=$User.bisFrontEndPreview
      True Float=&Multiply_Float(
       A=4
       B=&FloatFromCurve(FloatCurve=DataInterface() {propertyValues {Curve {keys [<at -0.0012 0>,<at 0.3 0>,<at 1 1>]}}})
      )
      False Float=12
     )
    )) {isSetParameters=true, module=SetVariables_#12}
   /SetVariables(Particles.ColorLine_Green:LinearColor=&MakeLinearColorFromVectorAndFloat(
     "Vector (RGB)"=&MakeVectorFromLinearColorRGB(Color=&MakeCustomLinearColorFromBool(
       Bool=$User.bisFrontEndPreview
       True Color=&SelectLinearColorFromArray(
        Array Sampling Mode=::NewEnumerator1
        Direct Array Index=&MakeIntFromFloat(Float=$User.VariantSwitch)
        Color Selection Array=DataInterface() {propertyValues {ColorData [<rgba 0 0.26 0.2012 10>, <rgba 0.6354 0.448 0.0233 10>, <rgba 0.6354 0.5626 0.5845 15>, <rgba 1 1 1 15>, <rgba 1 1 1 16>]}}
        Random Seed=?
       )
       False Color=&SelectLinearColorFromArray(
        Array Sampling Mode=::NewEnumerator1
        Direct Array Index=&MakeIntFromFloat(Float=$User.VariantSwitch)
        Color Selection Array=DataInterface() {propertyValues {ColorData [<rgba 0 0.3385 0.262 15>, <rgba 0.6354 0.448 0.0233 15>, <rgba 0.6354 0.5626 0.5845 15>, <rgba 1 1 1 16>, <rgba 1 1 1 16>]}}
        Random Seed=?
       )
      ))
     "Float (Alpha)"=&MakeCustomFloatFromBool(
      Bool=$User.bisFrontEndPreview
      True Float=&Multiply_Float(
       A=14
       B=&FloatFromCurve(FloatCurve=DataInterface() {propertyValues {Curve {keys [<at -0.0012 0>,<at 0.3 0>,<at 1 1>]}}})
      )
      False Float=15
     )
    )) {isSetParameters=true, module=SetVariables_#13}
   /SetVariables(Particles.ColorLine_Yellow:LinearColor=&MakeLinearColorFromVectorAndFloat(
     "Vector (RGB)"=&MakeVectorFromLinearColorRGB(Color=&SelectLinearColorFromArray(
       Array Sampling Mode=::NewEnumerator1
       Direct Array Index=&MakeIntFromFloat(Float=$User.VariantSwitch)
       Color Selection Array=DataInterface() {propertyValues {ColorData [<rgba 0.7405 1 0 10>, <rgba 0 0.233 0.65 15>, <rgba 0.0102 0.4138 0.65 15>, <rgba 0 0.2091 0.5208 15>, <rgba 0.8 0.1306 0.2927 15>]}}
       Random Seed=?
      ))
     "Float (Alpha)"=&MakeCustomFloatFromBool(
      Bool=$User.bisFrontEndPreview
      True Float=&Multiply_Float(
       A=15
       B=&FloatFromCurve(FloatCurve=DataInterface() {propertyValues {Curve {keys [<at -0.0012 0>,<at 0.3 0>,<at 1 1>]}}})
      )
      False Float=15
     )
    )) {isSetParameters=true, module=SetVariables_#14}
   /SetVariables(Particles.ColorLine_D_Green:LinearColor=&MakeLinearColorFromVectorAndFloat(
     "Vector (RGB)"=&MakeVectorFromLinearColorRGB(Color=&SelectLinearColorFromArray(
       Array Sampling Mode=::NewEnumerator1
       Direct Array Index=&MakeIntFromFloat(Float=$User.VariantSwitch)
       Color Selection Array=DataInterface() {propertyValues {ColorData [<rgba 0 0.5 0.4641 10>, <rgba 0.5 0.4582 0.0043 8>, <rgba 0.0354 0.4377 0.5 8>, <rgba 0.0042 0.2083 0.1568 8>, <rgba 0.8 0.0625 0.0396 8>]}}
       Random Seed=?
      ))
     "Float (Alpha)"=&MakeCustomFloatFromBool(
      Bool=$User.bisFrontEndPreview
      True Float=&Multiply_Float(
       A=8
       B=&FloatFromCurve(FloatCurve=DataInterface() {propertyValues {Curve {keys [<at -0.0012 0>,<at 0.3 0>,<at 1 1>]}}})
      )
      False Float=10
     )
    )) {isSetParameters=true, module=SetVariables_#15}
   /SetVariables(Particles.ColorLine_D_Yellow:LinearColor=&MakeLinearColorFromVectorAndFloat(
     "Vector (RGB)"=&MakeVectorFromLinearColorRGB(Color=&SelectLinearColorFromArray(
       Array Sampling Mode=::NewEnumerator1
       Direct Array Index=&MakeIntFromFloat(Float=$User.VariantSwitch)
       Color Selection Array=DataInterface() {propertyValues {ColorData [
          <rgba 0.1481 0.875 0.0763 5>
          <rgba 0 0.3924 0.8 10>
          <rgba 0.0137 0.1265 0.3125 10>
          <rgba 0.1481 0.875 0.0763 5>
          <rgba 0.4446 0.1198 0.8 5>
         ]}}
       Random Seed=?
      ))
     "Float (Alpha)"=&MakeCustomFloatFromBool(
      Bool=$User.bisFrontEndPreview
      True Float=&Multiply_Float(
       A=5
       B=&FloatFromCurve(FloatCurve=DataInterface() {propertyValues {Curve {keys [<at -0.0012 0>,<at 0.3 0>,<at 1 1>]}}})
      )
      False Float=5
     )
    )) {isSetParameters=true}
   /SetVariables(Particles.Color_soft:LinearColor=&MakeLinearColorFromVectorAndFloat(
     "Vector (RGB)"=&MakeVectorFromLinearColorRGB(Color=&SelectLinearColorFromArray(
       Array Sampling Mode=::NewEnumerator1
       Direct Array Index=&MakeIntFromFloat(Float=$User.VariantSwitch)
       Color Selection Array=DataInterface() {propertyValues {ColorData [<rgba 0 0.8958 0.057 5>, <rgba 0.2466 0.8958 0.048 5>, <rgba 0.1458 0.5293 0.8958 5>, <rgba 0.2666 0.0765 1 5>, <rgba 1 0.2954 0.044 5>]}}
       Random Seed=?
      ))
     "Float (Alpha)"=&MakeCustomFloatFromBool(
      Bool=$User.bisFrontEndPreview
      True Float=&Multiply_Float(
       A=5
       B=&FloatFromCurve(FloatCurve=DataInterface() {propertyValues {Curve {keys [<at -0.0012 0>,<at 0.3 0>,<at 1 1>]}}})
      )
      False Float=5
     )
    )) {isSetParameters=true}
   &SkeletalMeshLocation(
    Bone Sampling Mode=::NewEnumerator4
    MaskSpawnGroup=false
    Sampled Position Offset=&Multiply_VectorByFloat(
     Vector=&SelectVectorFromArray(
      Direct Array Index=$Particles.PartitionParticles.Partition
      Vector Selection Array=DataInterface() {propertyValues {FloatData [<2 0 -10>, <2 0 11>]}}
     )
     Float=&Lerp_Float(A=1, B=2, Alpha=$NPC.FortniteNPC.BigHeadMode)
    )
    Skeletal Mesh=$System.Skeletal Mesh Reader
   )
   &DynamicMaterialParameters(
    Index 0 Param 1=&RandomRangeFloat(Minimum=-0.6, Maximum=-1.2)
    Index 0 Param 2=&SelectFloatFromArray(Direct Array Index=$Particles.PartitionParticles.Partition, Float Selection Array=DataInterface() {propertyValues {FloatData [0,0.5]}})
    Index 0 Param 3=&Lerp_Float(A=-50, B=-100, Alpha=$NPC.FortniteNPC.BigHeadMode)
    Index 0 Param 4=&Lerp_Float(A=50, B=100, Alpha=$NPC.FortniteNPC.BigHeadMode)
    Index 2 Param 1=1
    Index 2 Param 3=1
    Index 2 Param 4=&SelectFloatFromArray(
     Direct Array Index=&MakeIntFromFloat(Float=$User.VariantSwitch)
     Float Selection Array=DataInterface() {propertyValues {FloatData [0,1,2,3,4]}}
    )
    Module.Index0Param0Write=true
    Module.Index0Param1Write=true
    Module.Index0Param2Write=true
    Module.Index0Param3Write=true
    Module.Index2Param0Write=true
    Module.Index2Param1Write=true
    Module.Index2Param2Write=true
    Module.Index2Param3Write=true
    Write Parameter Index 2=true
   )
   &InterpolateOverTime(Rate Of Change=10, Target Vector Value:NiagaraPosition=$Particles.Position, Variable Type=::NewEnumerator2)
   /SetVariables(Particles.Interpolate Dynamic=&MakeFloat4FromVectorAndFloat("Vector (RGB)"=$Particles.InterpolateOverTime.InterpolateOverTime_V3_.Moving Average, "Float (Alpha)"=0.5)) {isSetParameters=true, module=SetVariables_#16}
   /SetVariables(Particles.VisibilityTag=&MakeCustomIntFromBool(Bool=$User.bisHeadHidden, True INT=1, False INT=0)) {isSetParameters=true}
  ]
  rendererClasses [&NiagaraMeshRendererProperties]
  renderers [
   {rendererClass=&NiagaraMeshRendererProperties, rendererIndex=0}
   {rendererClass=&NiagaraMeshRendererProperties, rendererIndex=1}
   {rendererClass=&NiagaraMeshRendererProperties, rendererIndex=2}
   {rendererClass=&NiagaraMeshRendererProperties, rendererIndex=3}
   {rendererClass=&NiagaraMeshRendererProperties, rendererIndex=4}
   {rendererClass=&NiagaraMeshRendererProperties, rendererIndex=5}
   {rendererClass=&NiagaraMeshRendererProperties, rendererIndex=6}
  ]
 }
 Head_Trail001 {
  emitterUpdateScript [
   &EmitterState(
    Inactive Response=::NewEnumerator1
    Life Cycle Mode=::NewEnumerator1
    Loop Behavior=::NewEnumerator1
    Loop Duration Mode=::NewEnumerator1
    UseLoopDelay=false
   )
   &SpawnBurst_Instantaneous(Use Spawn Probability=false, UseLoopCountLimit=false)
   /SetVariables(Emitter.LinearMaskLeft_Power:NiagaraFloat=&FN_PickFloatDependOnPlatform(TargetPlatformActive=1, TargetPlatformNoActive=3)) {isSetParameters=true}
  ]
  particleSpawnScript [
   &InitializeParticle(Color=<rgba 0.65 0.65 0.65 5>, Lifetime=1.5, Mesh Scale Mode=::NewEnumerator4, UsePositionOffset=true, Write Material Random=true)
   &InitialMeshOrientation(Mesh Orientation Mode=::NewEnumerator2, Rotation=<0 -0.25 0>, Use Rotation Vector=true)
   &InitialMeshOrientation(
    Mesh Orientation Mode=::NewEnumerator2
    Rotation=&MakeCustomVectorFromBool(Bool=$User.bisFrontEndPreview, True Vector=<0.1 0 0.08>, False Vector=<0.15 0 0.08>)
    Use Rotation Vector=true
   ) {module=InitialMeshOrientation001}
   /SetVariables(Particles.Color_D:LinearColor=<rgba 0 1 0.277 1.1>) {isSetParameters=true, module=SetVariables_#10}
  ]
  particleUpdateScript [
   &ParticleState(Kill Particles When Lifetime Has Elapsed=false, Let Infinitely Lived Particles Die When Emitter Deactivates=true)
   &ScaleColor(
    Scale Alpha=&MakeCustomFloatFromBool(
     Bool=$User.bisFrontEndPreview
     True Float=&Multiply_Float(
      A=0.3
      B=&FloatFromCurve(FloatCurve=DataInterface() {propertyValues {Curve {keys [<at -0.0012 0>,<at 0.3 0>,<at 1 1>]}}})
     )
     False Float=0.5
    )
    Scale RGB=&MakeVectorFromLinearColorRGB(Color=&SelectLinearColorFromArray(
      Array Sampling Mode=::NewEnumerator1
      Direct Array Index=&MakeIntFromFloat(Float=$User.VariantSwitch)
      Color Selection Array=DataInterface() {propertyValues {ColorData [<rgba 1 1 1 1>, <rgba 0.5 0.4536 0.025 1>, <rgba 0.6562 0.6562 0.6562 1>, <rgba 0.6562 0.6562 0.6562 1>, <rgba 0.7552 0.7552 0.7552 1>]}}
      Random Seed=?
     ))
    ScaleA=true
    ScaleRGB=true
   )
   /SetVariables(Particles.ColorPoint:LinearColor=&MakeLinearColorFromVectorAndFloat(
     "Vector (RGB)"=&MakeVectorFromLinearColorRGB(Color=&SelectLinearColorFromArray(
       Array Sampling Mode=::NewEnumerator1
       Direct Array Index=&MakeIntFromFloat(Float=$User.VariantSwitch)
       Color Selection Array=DataInterface() {propertyValues {ColorData [<rgba 0.2216 1 0.0882 10>, <rgba 0.8854 1 0.1078 10>, <rgba 0.9977 0.5688 1.2 10>, <rgba 0.3333 0.8124 2 10>, <rgba 2 1.6462 0 10>]}}
       Random Seed=?
      ))
     "Float (Alpha)"=&MakeCustomFloatFromBool(
      Bool=$User.bisFrontEndPreview
      True Float=&Multiply_Float(
       A=5
       B=&FloatFromCurve(FloatCurve=DataInterface() {propertyValues {Curve {keys [<at -0.0012 0>,<at 0.3 0>,<at 1 1>]}}})
      )
      False Float=10
     )
    )) {isSetParameters=true, module=SetVariables_#12}
   /SetVariables(Particles.ColorLine_Green:LinearColor=&MakeLinearColorFromVectorAndFloat(
     "Vector (RGB)"=&MakeVectorFromLinearColorRGB(Color=&MakeCustomLinearColorFromBool(
       Bool=$User.bisFrontEndPreview
       True Color=&SelectLinearColorFromArray(
        Array Sampling Mode=::NewEnumerator1
        Direct Array Index=&MakeIntFromFloat(Float=$User.VariantSwitch)
        Color Selection Array=DataInterface() {propertyValues {ColorData [<rgba 0 0.25 0.1756 15>, <rgba 0.76 1 0.037 5>, <rgba 0.6354 0.5626 0.5845 15>, <rgba 1 1 1 16>, <rgba 1 1 1 16>]}}
        Random Seed=?
       )
       False Color=&SelectLinearColorFromArray(
        Array Sampling Mode=::NewEnumerator1
        Direct Array Index=&MakeIntFromFloat(Float=$User.VariantSwitch)
        Color Selection Array=DataInterface() {propertyValues {ColorData [<rgba 0 0.3385 0.262 15>, <rgba 0.9167 0.5842 0.0587 10>, <rgba 0.6354 0.5626 0.5845 15>, <rgba 1 1 1 16>, <rgba 1 1 1 16>]}}
        Random Seed=?
       )
      ))
     "Float (Alpha)"=&MakeCustomFloatFromBool(
      Bool=$User.bisFrontEndPreview
      True Float=&Multiply_Float(
       A=13
       B=&FloatFromCurve(FloatCurve=DataInterface() {propertyValues {Curve {keys [<at -0.0012 0>,<at 0.3 0>,<at 1 1>]}}})
      )
      False Float=15
     )
    )) {isSetParameters=true, module=SetVariables_#13}
   /SetVariables(Particles.ColorLine_Yellow:LinearColor=&MakeLinearColorFromVectorAndFloat(
     "Vector (RGB)"=&MakeVectorFromLinearColorRGB(Color=&SelectLinearColorFromArray(
       Array Sampling Mode=::NewEnumerator1
       Direct Array Index=&MakeIntFromFloat(Float=$User.VariantSwitch)
       Color Selection Array=DataInterface() {propertyValues {ColorData [<rgba 0.7405 1 0 10>, <rgba 0 0.233 0.65 15>, <rgba 0.0102 0.4138 0.65 15>, <rgba 0 0.2091 0.5208 15>, <rgba 0.8 0.1306 0.2927 15>]}}
       Random Seed=?
      ))
     "Float (Alpha)"=&MakeCustomFloatFromBool(
      Bool=$User.bisFrontEndPreview
      True Float=&Multiply_Float(
       A=15
       B=&FloatFromCurve(FloatCurve=DataInterface() {propertyValues {Curve {keys [<at -0.0012 0>,<at 0.3 0>,<at 1 1>]}}})
      )
      False Float=15
     )
    )) {isSetParameters=true, module=SetVariables_#14}
   /SetVariables(Particles.ColorLine_D_Green:LinearColor=&MakeLinearColorFromVectorAndFloat(
     "Vector (RGB)"=&MakeVectorFromLinearColorRGB(Color=&SelectLinearColorFromArray(
       Array Sampling Mode=::NewEnumerator1
       Direct Array Index=&MakeIntFromFloat(Float=$User.VariantSwitch)
       Color Selection Array=DataInterface() {propertyValues {ColorData [
          <rgba 0 0.3777 0.75 10>
          <rgba 0.6 0.4617 0.0145 8>
          <rgba 0.5677 0.1449 0.2784 8>
          <rgba 0.0042 0.2083 0.1568 8>
          <rgba 0.8 0.0625 0.0396 8>
         ]}}
       Random Seed=?
      ))
     "Float (Alpha)"=&MakeCustomFloatFromBool(
      Bool=$User.bisFrontEndPreview
      True Float=&Multiply_Float(
       A=4
       B=&FloatFromCurve(FloatCurve=DataInterface() {propertyValues {Curve {keys [<at -0.0012 0>,<at 0.3 0>,<at 1 1>]}}})
      )
      False Float=10
     )
    )) {isSetParameters=true, module=SetVariables_#15}
   /SetVariables(Particles.ColorLine_D_Yellow:LinearColor=&MakeLinearColorFromVectorAndFloat(
     "Vector (RGB)"=&MakeVectorFromLinearColorRGB(Color=&SelectLinearColorFromArray(
       Array Sampling Mode=::NewEnumerator1
       Direct Array Index=&MakeIntFromFloat(Float=$User.VariantSwitch)
       Color Selection Array=DataInterface() {propertyValues {ColorData [<rgba 0.1481 0.875 0.0763 5>, <rgba 0 0.3003 0.8 15>, <rgba 0 0.2252 0.6 15>, <rgba 0.1481 0.875 0.0763 5>, <rgba 0.4446 0.1198 0.8 5>]}}
       Random Seed=?
      ))
     "Float (Alpha)"=&MakeCustomFloatFromBool(
      Bool=$User.bisFrontEndPreview
      True Float=&Multiply_Float(
       A=10
       B=&FloatFromCurve(FloatCurve=DataInterface() {propertyValues {Curve {keys [<at -0.0012 0>,<at 0.3 0>,<at 1 1>]}}})
      )
      False Float=12
     )
    )) {isSetParameters=true}
   /SetVariables(Particles.Color_soft:LinearColor=&MakeLinearColorFromVectorAndFloat(
     "Vector (RGB)"=&MakeVectorFromLinearColorRGB(Color=&SelectLinearColorFromArray(
       Array Sampling Mode=::NewEnumerator1
       Direct Array Index=&MakeIntFromFloat(Float=$User.VariantSwitch)
       Color Selection Array=DataInterface() {propertyValues {ColorData [<rgba 0 0.8958 0.057 5>, <rgba 0.2466 0.8958 0.048 5>, <rgba 0.1458 0.5293 0.8958 5>, <rgba 0.2666 0.0765 1 5>, <rgba 1 0.2954 0.044 5>]}}
       Random Seed=?
      ))
     "Float (Alpha)"=&MakeCustomFloatFromBool(
      Bool=$User.bisFrontEndPreview
      True Float=&Multiply_Float(
       A=5
       B=&FloatFromCurve(FloatCurve=DataInterface() {propertyValues {Curve {keys [<at -0.0012 0>,<at 0.3 0>,<at 1 1>]}}})
      )
      False Float=5
     )
    )) {isSetParameters=true}
   &SkeletalMeshLocation(
    Bone / Socket Index=1
    Bone Sampling Mode=::NewEnumerator4
    MaskSpawnGroup=false
    Sampled Position Offset=&Lerp_Vector(A=<-22 -3 -8>, B=<-3 1 -11>, Alpha=$User.PocketTrail_Location)
    Skeletal Mesh=$System.Skeletal Mesh Reader
   )
   /SetVariables(Particles.Size_Orange:Vector3f=<0.55 0.2 -0.65>) {isSetParameters=true, module=SetVariables_#8}
   /SetVariables(Particles.Size_D:Vector3f=<0.1 0.3 -0.6>) {isSetParameters=true, module=SetVariables_#9}
   /SetVariables(Particles.Size_Green:Vector3f=<0.3 0.3 0.5>) {isSetParameters=true, module=SetVariables_#11}
   &DynamicMaterialParameters(
    Index 0 Param 1=&RandomRangeFloat(Minimum=-0.6, Maximum=-1.2)
    Index 0 Param 2=&SelectFloatFromArray(Direct Array Index=$Particles.PartitionParticles.Partition, Float Selection Array=DataInterface() {propertyValues {FloatData [0,0.5]}})
    Index 0 Param 3=&Lerp_Float(A=-50, B=-100, Alpha=$NPC.FortniteNPC.BigHeadMode)
    Index 0 Param 4=&Lerp_Float(A=50, B=100, Alpha=$NPC.FortniteNPC.BigHeadMode)
    Index 2 Param 1=1
    Index 2 Param 3=1
    Index 2 Param 4=&SelectFloatFromArray(
     Direct Array Index=&MakeIntFromFloat(Float=$User.VariantSwitch)
     Float Selection Array=DataInterface() {propertyValues {FloatData [0,1,2,3,4]}}
    )
    Module.Index0Param0Write=true
    Module.Index0Param1Write=true
    Module.Index0Param2Write=true
    Module.Index0Param3Write=true
    Module.Index2Param0Write=true
    Module.Index2Param1Write=true
    Module.Index2Param2Write=true
    Module.Index2Param3Write=true
    Write Parameter Index 2=true
   )
   &InterpolateOverTime(Rate Of Change=10, Target Vector Value:NiagaraPosition=$Particles.Position, Variable Type=::NewEnumerator2)
   /SetVariables(Particles.Interpolate Dynamic=&MakeFloat4FromVectorAndFloat("Vector (RGB)"=$Particles.InterpolateOverTime.InterpolateOverTime_V3_.Moving Average, "Float (Alpha)"=0.5)) {isSetParameters=true, module=SetVariables_#16}
  ]
  rendererClasses [&NiagaraMeshRendererProperties]
  renderers [
   {rendererClass=&NiagaraMeshRendererProperties, rendererIndex=0}
   {rendererClass=&NiagaraMeshRendererProperties, rendererIndex=1}
   {rendererClass=&NiagaraMeshRendererProperties, rendererIndex=2}
   {rendererClass=&NiagaraMeshRendererProperties, rendererIndex=3}
   {rendererClass=&NiagaraMeshRendererProperties, rendererIndex=4}
   {rendererClass=&NiagaraMeshRendererProperties, rendererIndex=5}
   {rendererClass=&NiagaraMeshRendererProperties, rendererIndex=6}
  ]
 }
 Head_Trail_soft {
  emitterUpdateScript [
   &EmitterState(
    Inactive Response=::NewEnumerator1
    Life Cycle Mode=::NewEnumerator1
    Loop Behavior=::NewEnumerator1
    Loop Duration Mode=::NewEnumerator1
    Reset Age On Awaken=true
    Scalability Mode=::NewEnumerator1
    UseLoopDelay=false
   )
   &SpawnBurst_Instantaneous(Spawn Count=2, Use Spawn Probability=false)
  ]
  particleSpawnScript [
   &PartitionParticles()
   &InitializeParticle(
    Color=<rgba 1 1 1 1.5>
    Mass Mode=::NewEnumerator1
    Mesh Scale=<1.2 0.1 1>
    Mesh Scale Mode=::NewEnumerator4
    Sprite Size Mode=::NewEnumerator4
    UsePositionOffset=false
    Write Material Random=true
   ) {module=InitializeParticle001}
   &DynamicMaterialParameters(
    Index 0 Param 1=0.5
    Index 0 Param 2=1
    Index 0 Param 3=15
    Module.Param0WriteEnabled=true
    Module.Param1WriteEnabled=true
    Module.Param2WriteEnabled=true
    Module.Param3WriteEnabled=true
   )
   &InitialMeshOrientation(
    Mesh Orientation Mode=::NewEnumerator2
    Rotation=&SelectVectorFromArray(
     Direct Array Index=$Particles.PartitionParticles.Partition
     Vector Selection Array=DataInterface() {propertyValues {FloatData [<0.28 0 0>, <0.2 0 0>]}}
    )
    Use Rotation Vector=true
   ) {module=InitialMeshOrientation001}
   &InitialMeshOrientation(
    Mesh Orientation Mode=::NewEnumerator2
    Rotation=&SelectVectorFromArray(
     Direct Array Index=$Particles.PartitionParticles.Partition
     Vector Selection Array=DataInterface() {propertyValues {FloatData [<0.05 0 0>, <-0.05 0 0.02>]}}
    )
    Use Rotation Vector=true
   )
  ]
  particleUpdateScript [
   &ParticleState(Kill Particles When Lifetime Has Elapsed=false, Let Infinitely Lived Particles Die When Emitter Deactivates=true)
   &SkeletalMeshLocation(
    Bone Sampling Mode=::NewEnumerator4
    "Interpolate Mesh Sampling (Update)"=true
    MaskSpawnGroup=false
    Sampled Position Offset=&Multiply_VectorByFloat(
     Vector=&SelectVectorFromArray(
      Direct Array Index=$Particles.PartitionParticles.Partition
      Vector Selection Array=DataInterface() {propertyValues {FloatData [<2.5 0 -10>, <2.5 0 11>]}}
     )
     Float=&Lerp_Float(A=1, B=2, Alpha=$NPC.FortniteNPC.BigHeadMode)
    )
    Skeletal Mesh=$System.Skeletal Mesh Reader
   ) {module=SkeletalMeshLocation001}
   &ScaleMeshSize(Scale Factor=&VectorFromFloat(Value=&Lerp_Float(A=1, B=2, Alpha=$NPC.FortniteNPC.BigHeadMode)))
   &InterpolateOverTime(Rate Of Change=7, Target Vector Value:NiagaraPosition=$Particles.Position, Variable Type=::NewEnumerator2)
   /SetVariables(Particles.Interpolate Dynamic=&MakeFloat4FromVectorAndFloat("Vector (RGB)"=$Output.InterpolateOverTime.Moving Average, "Float (Alpha)"=1)) {isSetParameters=true, module=SetVariables_#17}
   &SolveForcesAndVelocity(Clamp Velocity=false, Limit Acceleration=false)
   &ScaleColor(
    Scale Alpha=&MakeCustomFloatFromBool(
     Bool=$User.bisFrontEndPreview
     True Float=&Multiply_Float(
      A=0.02
      B=&FloatFromCurve(FloatCurve=DataInterface() {propertyValues {Curve {keys [<at -0.0012 0>,<at 0.3 0>,<at 1 1>]}}})
     )
     False Float=0.05
    )
    Scale RGB=&MakeVectorFromLinearColorRGB(Color=&SelectLinearColorFromArray(
      Array Sampling Mode=::NewEnumerator1
      Direct Array Index=&MakeIntFromFloat(Float=$User.VariantSwitch)
      Color Selection Array=DataInterface() {propertyValues {ColorData [
         <rgba 0.0146 0.4531 0.3526 1.5>
         <rgba 0.4372 0.6146 0.0709 8>
         <rgba 0.6146 0.2259 0.3453 8>
         <rgba 0.3481 0.1522 0.6146 8>
         <rgba 0.6146 0.4855 0.032 8>
        ]}}
      Random Seed=?
     ))
    ScaleA=true
    ScaleRGB=true
   )
   /SetVariables(Particles.VisibilityTag=&MakeCustomIntFromBool(Bool=$User.bisHeadHidden, True INT=1, False INT=0)) {isSetParameters=true}
  ]
  rendererClasses [&NiagaraMeshRendererProperties]
  renderers [{rendererClass=&NiagaraMeshRendererProperties, rendererIndex=0}]
 }
 Light001 {
  emitterUpdateScript [
   &EmitterState(
    Inactive Response=::NewEnumerator1
    Life Cycle Mode=::NewEnumerator1
    Loop Behavior=::NewEnumerator1
    Loop Duration Mode=::NewEnumerator1
    UseLoopDelay=false
   )
   &SpawnBurst_Instantaneous(Use Spawn Probability=false, UseLoopCountLimit=false)
  ]
  particleSpawnScript [&InitializeParticle(
    Color Mode=::NewEnumerator0
    Lifetime=1.5
    Sprite Size Mode=::NewEnumerator3
    Uniform Sprite Size=15
    UsePositionOffset=false
    Write Material Random=true
   )]
  particleUpdateScript [
   &ParticleState(
    Kill Particles When Lifetime Has Elapsed=false
    Let Infinitely Lived Particles Die When Emitter Deactivates=true
    Loop Particles Lifetime=true
   )
   &SkeletalMeshLocation(
    Bone / Socket Index=1
    Bone Sampling Mode=::NewEnumerator4
    MaskSpawnGroup=false
    Sampled Position Offset=&Lerp_Vector(A=<-22 1 -15>, B=<-3 8 -15>, Alpha=$User.PocketTrail_Location)
    Skeletal Mesh=$System.Skeletal Mesh Reader
   )
   &FN_Light_Attributes(
    Color=&MakeLinearColorFromVectorAndFloat(
     "Vector (RGB)"=&LerpMultipleVectors(
      Alpha:NiagaraFloat=$User.VariantSwitch
      V 1:Vector3f=&MakeVectorFromLinearColorRGB(Color=&ColorFromCurve(ColorCurve=DataInterface() {propertyValues {
          AlphaCurve {keys []}
          BlueCurve {keys [
            <at 0 0>
            <at 0.1383 0>
            <at 0.1966 0>
            <at 0.3441 0>
            <at 0.3987 0>
            {arriveTangent=2.8924, leaveTangent=2.8924, time=0.5463, value=0.886}
            <at 0.7348 0.886>
            <at 0.8948 0>
            <at 0.9308 0>
            <at 1 0>
           ]}
          GreenCurve {keys [<at 0 0>,<at 0.1383 1>,<at 0.1966 1>,<at 0.3441 1>,<at 0.3987 1>,<at 0.5463 1>,<at 0.7348 1>,<at 0.8948 1>,<at 0.9308 1>,<at 1 0>]}
          RedCurve {keys [
            <at 0 0>
            <at 0.1383 0.0226>
            <at 0.1966 0.0226>
            <at 0.3441 0.8122>
            <at 0.3987 0.8122>
            {arriveTangent=-2.6514, leaveTangent=-2.6514, time=0.5463, value=0}
            <at 0.7348 0>
            <at 0.8948 0.0115>
            <at 0.9308 0.0226>
            <at 1 0>
           ]}
         }}))
      V 2:Vector3f=&MakeVectorFromLinearColorRGB(Color=&ColorFromCurve(ColorCurve=DataInterface() {propertyValues {
          AlphaCurve {keys []}
          BlueCurve {keys [
            <at 0 0>
            <at 0.1383 0>
            <at 0.1966 0>
            <at 0.3441 0>
            <at 0.3987 0>
            {arriveTangent=2.8924, leaveTangent=2.8924, time=0.5463, value=0.886}
            <at 0.7348 0.886>
            <at 0.8948 0>
            <at 0.9308 0>
            <at 1 0>
           ]}
          GreenCurve {keys [<at 0 0>,<at 0.1383 1>,<at 0.1966 1>,<at 0.3441 1>,<at 0.3987 1>,<at 0.5463 1>,<at 0.7348 1>,<at 0.8948 1>,<at 0.9308 1>,<at 1 0>]}
          RedCurve {keys [
            <at 0 0>
            <at 0.1383 0.0226>
            <at 0.1966 0.0226>
            <at 0.3441 0.8122>
            <at 0.3987 0.8122>
            {arriveTangent=-2.6514, leaveTangent=-2.6514, time=0.5463, value=0}
            <at 0.7348 0>
            <at 0.8948 0.0115>
            <at 0.9308 0.0226>
            <at 1 0>
           ]}
         }}))
      V 3:Vector3f=&MakeVectorFromLinearColorRGB(Color=&ColorFromCurve(ColorCurve=DataInterface() {propertyValues {
          AlphaCurve {keys []}
          BlueCurve {keys [
            <at 0 0>
            <at 0.1383 0.5138>
            <at 0.1966 0.5392>
            <at 0.3441 1>
            <at 0.3987 1>
            {arriveTangent=2.8924, leaveTangent=2.8924, time=0.5463, value=0.886}
            <at 0.7348 0.886>
            <at 0.8948 0.5361>
            <at 0.9308 0.4659>
            <at 1 0>
           ]}
          GreenCurve {keys [<at 0 0>,<at 0.1383 0.3199>,<at 0.1966 0.2344>,<at 0.3441 0.2797>,<at 0.3987 0.2596>,<at 0.5463 1>,<at 0.7348 1>,<at 0.8948 0.1904>,<at 0.9308 0.2029>,<at 1 0>]}
          RedCurve {keys [
            <at 0 0>
            <at 0.1383 1>
            <at 0.1966 1>
            <at 0.3441 0.4759>
            <at 0.3987 0.4265>
            {arriveTangent=-2.6514, leaveTangent=-2.6514, time=0.5463, value=0}
            <at 0.7348 0>
            <at 0.8948 1>
            <at 0.9308 1>
            <at 1 0>
           ]}
         }}))
      V 4:Vector3f=&MakeVectorFromLinearColorRGB(Color=&ColorFromCurve(ColorCurve=DataInterface() {propertyValues {
          AlphaCurve {keys []}
          BlueCurve {keys [
            <at 0 0>
            <at 0.1383 0.1292>
            <at 0.1966 0.0966>
            <at 0.305 1>
            <at 0.3987 1>
            {arriveTangent=2.8924, leaveTangent=2.8924, time=0.5463, value=0.886}
            <at 0.6582 0.886>
            <at 0.7417 1>
            <at 0.9308 1>
            <at 1 0>
           ]}
          GreenCurve {keys [<at 0 0>,<at 0.1383 1>,<at 0.1966 1>,<at 0.305 0.1547>,<at 0.3987 0.1203>,<at 0.5463 1>,<at 0.6582 1>,<at 0.7417 0.1995>,<at 0.9308 0.1535>,<at 1 0>]}
          RedCurve {keys [
            <at 0 0>
            <at 0.1383 0.1116>
            <at 0.1966 0.1526>
            <at 0.305 0.4236>
            <at 0.3987 0.4356>
            {arriveTangent=-2.6514, leaveTangent=-2.6514, time=0.5463, value=0}
            <at 0.6582 0>
            <at 0.7417 0.4264>
            <at 0.9308 0.3741>
            <at 1 0>
           ]}
         }}))
      V 5:Vector3f=&MakeVectorFromLinearColorRGB(Color=&ColorFromCurve(ColorCurve=DataInterface() {propertyValues {
          AlphaCurve {keys []}
          BlueCurve {keys [
            <at 0 0>
            <at 0.1383 0.0254>
            <at 0.1966 0>
            <at 0.305 0.0582>
            <at 0.3987 0.0598>
            {arriveTangent=2.8924, leaveTangent=2.8924, time=0.5463, value=0.0659}
            <at 0.7953 0.0226>
            <at 0.8773 1>
            <at 0.9308 1>
            <at 1 0>
           ]}
          GreenCurve {keys [<at 0 0>,<at 0.1383 0.3206>,<at 0.1966 0.1186>,<at 0.305 0.7173>,<at 0.3987 0.8732>,<at 0.5463 0.3982>,<at 0.7953 0.1375>,<at 0.8773 0.1995>,<at 0.9308 0.1535>,<at 1 0>]}
          RedCurve {keys [
            <at 0 0>
            <at 0.1383 1>
            <at 0.1966 1>
            <at 0.305 1>
            <at 0.3987 1>
            {arriveTangent=-2.6514, leaveTangent=-2.6514, time=0.5463, value=1}
            <at 0.7953 1>
            <at 0.8773 0.4264>
            <at 0.9308 0.3741>
            <at 1 0>
           ]}
         }}))
      Vector Count:ENiagara_MultipleLerpCount=::NewEnumerator2
     )
     "Float (Alpha)"=1
    )
    Create Color Attribute=true
    Create Exponent Attribute=true
    Create Radius Attribute=true
    Exponent=2
    Radius=30
    Scale Color=15
   )
  ]
  rendererClasses [&NiagaraLightRendererProperties, &NiagaraSpriteRendererProperties]
  renderers [
   {rendererClass=&NiagaraLightRendererProperties, rendererIndex=0}
   {rendererClass=&NiagaraSpriteRendererProperties, rendererIndex=1}
  ]
 }
 Sprites001 {
  emitterUpdateScript [
   &EmitterState(
    Enable Distance Culling=true
    Inactive Response=::NewEnumerator1
    Life Cycle Mode=::NewEnumerator1
    Loop Delay=&RandomRangeFloat(Minimum=0, Maximum=1)
    Max Distance Response=::KillImmediately
    Module.bUseMaxDistance=true
    Module.bUseMinDistance=false
    Scalability Mode=::NewEnumerator1
    UseLoopDelay=false
   )
   &SpawnRate(
    Spawn Probability=&MakeCustomFloatFromBool(
     Bool=&SetBoolByFloatComparison(A=&VectorLength(Vector=$Engine.Owner.Velocity), B=1000)
     True Float=0
     False Float=1
    )
    SpawnRate=15
    Use Spawn Probability=true
   )
  ]
  particleSpawnScript [
   &InitializeParticle(
    Color=&MakeCustomLinearColorFromBool(
     Bool=&RandomBool()
     True Color=&SelectLinearColorFromArray(
      Array Sampling Mode=::NewEnumerator1
      Direct Array Index=&MakeIntFromFloat(Float=$User.VariantSwitch)
      Color Selection Array=DataInterface() {propertyValues {ColorData [<rgba 0.7658 1 0 1>, <rgba 0.7658 1 0 1>, <rgba 1.5 0.3132 0.7232 1>, <rgba 0.2141 0.0573 1 1>, <rgba 1 0.3428 0.0513 1>]}}
      Random Seed=?
     )
     False Color=&MakeCustomLinearColorFromBool(
      Bool=&RandomBool()
      True Color=&SelectLinearColorFromArray(
       Array Sampling Mode=::NewEnumerator1
       Direct Array Index=&MakeIntFromFloat(Float=$User.VariantSwitch)
       Color Selection Array=DataInterface() {propertyValues {ColorData [<rgba 0 0.7188 0.2139 1>, <rgba 0 0.7188 0.2139 1>, <rgba 1.5 0.3132 0.7232 1>, <rgba 0.2141 0.0573 1 1>, <rgba 1 0.3428 0.0513 1>]}}
       Random Seed=?
      )
      False Color=&SelectLinearColorFromArray(
       Array Sampling Mode=::NewEnumerator1
       Direct Array Index=&MakeIntFromFloat(Float=$User.VariantSwitch)
       Color Selection Array=DataInterface() {propertyValues {ColorData [<rgba 0 0.6164 0.6979 1>, <rgba 0 0.6164 0.6979 1>, <rgba 1.5 0.3132 0.7232 1>, <rgba 0.2141 0.0573 1 1>, <rgba 1 0.3428 0.0513 1>]}}
       Random Seed=?
      )
     )
    )
    Lifetime Max=0.45
    Lifetime Min=0.4
    Lifetime Mode=::NewEnumerator1
    Sprite Rotation Mode=::NewEnumerator2
    Sprite Size Mode=::NewEnumerator1
    Uniform Sprite Size Max=8
    Uniform Sprite Size Min=4
    UsePositionOffset=false
    Write Material Random=true
   )
   &ShapeLocation(
    Cylinder Height=12
    Cylinder Radius=8
    MaskSpawnGroup=false
    Override Seed=false
    Rotation Axis=<0 1 0>
    Rotation Mode=::NewEnumerator1
    Shape Origin=&VectorFromFloat(Value=0)
    Shape Primitive=::NewEnumerator1
    Surface Only=false
    UseOwnerScale=true
   )
   /SetVariables(Particles.VisibilityTag=&RandomRangeInt(Minimum=0, Maximum=3)) {isSetParameters=true, module=SetVariables_#2}
   /SetVariables(Particles.Color_triangle:LinearColor=&MakeCustomLinearColorFromBool(
     Bool=&RandomBool()
     True Color=<rgba 0.7658 1 0 1>
     False Color=&MakeCustomLinearColorFromBool(Bool=&RandomBool(), True Color=<rgba 0 0.7188 0.1369 1>, False Color=<rgba 0 0.6348 0.7188 1>)
    )) {isSetParameters=true, module=SetVariables_#3}
   /SetVariables(Particles.Color_triangleglow:LinearColor=&MultiplyLinearColor(
     Linear Color=$Particles.Color_triangle
     Scale Factor=&MakeVector4_FromLinearColor(LinearColor=&MakeLinearColorFromFloat(Value:NiagaraFloat=0.5))
    )) {isSetParameters=true, module=SetVariables_#4}
  ]
  particleUpdateScript [
   &ParticleState()
   &SkeletalMeshLocation(
    Bone / Socket Index=1
    Bone Sampling Mode=::NewEnumerator4
    MaskSpawnGroup=false
    Position Sampling=::NewEnumerator1
    Sampled Position Offset=&AddVectorToPosition(
     Position=$Particles.ShapeLocation.ShapePosition
     Vector=&Lerp_Vector(A=<-20 -5 -10>, B=<0 -6 -16>, Alpha=$User.PocketTrail_Location)
    )
    Skeletal Mesh=$System.Skeletal Mesh Reader
    Source Movement Scale:Vector3f=<0.9 0.9 0.9>
   )
   &ScaleSpriteSize(Scale Factor=&Vector2DFromFloat(Value=&FloatFromCurve(FloatCurve=DataInterface() {propertyValues {Curve {keys [
          {interpMode=RCIM_Cubic, tangentMode=RCTM_User, time=0, value=0}
          {interpMode=RCIM_Cubic, leaveTangent=-1.6234, tangentMode=RCTM_Break, time=0.4, value=1}
          {tangentMode=RCTM_User, time=1, value=0}
         ]}}})))
   /SetVariables(Particles.Size002:Vector2f=&Multiply_Vector2D(A=$Particles.SpriteSize, B=&Vector2DFromFloat(Value=1.3))) {isSetParameters=true, module=SetVariables_#5}
   /SetVariables(Particles.Size_Triangle:Vector2f=&Multiply_Vector2D(A=$Particles.SpriteSize, B=&Vector2DFromFloat(Value=0.7))) {isSetParameters=true, module=SetVariables_#6}
   &Drag(Use Linear Drag=true, Use Rotational Drag=true)
   &SolveForcesAndVelocity(Clamp Velocity=false, Limit Acceleration=false)
   &ScaleColor(
    Scale Alpha=&Multiply_Float(
     A=&FNTimeOfDayScalar(Night Scale=1, Morning Scale=0.3, Evening Scale=0.3)
     B=&FloatFromCurve(FloatCurve=DataInterface() {propertyValues {Curve {keys [
          {interpMode=RCIM_Cubic, tangentMode=RCTM_User, time=0.1, value=0}
          {interpMode=RCIM_Cubic, tangentMode=RCTM_User, time=0.3, value=0.8}
          {tangentMode=RCTM_User, time=0.8, value=0}
         ]}}})
    )
    Scale RGB=&MakeVectorFromLinearColorRGB(Color=&SelectLinearColorFromArray(
      Array Sampling Mode=::NewEnumerator1
      Direct Array Index=&MakeIntFromFloat(Float=$User.VariantSwitch)
      Color Selection Array=DataInterface() {propertyValues {ColorData [<rgba 1 1 1 1>, <rgba 0 0.6174 1 1>, <rgba 1 1 1 1>]}}
      Random Seed=?
     ))
    ScaleA=true
    ScaleRGB=true
   )
   /SetVariables(Particles.Color002=&MultiplyLinearColor(Linear Color=<rgba 1.2 1.2 1.2 1>, Scale Factor=&MakeVector4_FromLinearColor(LinearColor=$Particles.Color))) {isSetParameters=true, module=SetVariables_#7}
  ]
  rendererClasses [&NiagaraSpriteRendererProperties]
  renderers [
   {rendererClass=&NiagaraSpriteRendererProperties}
   {rendererClass=&NiagaraSpriteRendererProperties, rendererIndex=1}
   {rendererClass=&NiagaraSpriteRendererProperties, rendererIndex=2}
   {rendererClass=&NiagaraSpriteRendererProperties, rendererIndex=3}
  ]
 }
 Trail_soft001 {
  emitterUpdateScript [
   &EmitterState(
    Inactive Response=::NewEnumerator1
    Life Cycle Mode=::NewEnumerator1
    Loop Behavior=::NewEnumerator1
    Loop Duration Mode=::NewEnumerator1
    Reset Age On Awaken=true
    Scalability Mode=::NewEnumerator1
    UseLoopDelay=false
   )
   &SpawnBurst_Instantaneous(Spawn Count=1, Use Spawn Probability=false)
  ]
  particleSpawnScript [
   &InitializeParticle(
    Color=<rgba 1 1 1 1.5>
    Mass Mode=::NewEnumerator1
    Mesh Scale=<0.5 0.5 0.8>
    Mesh Scale Mode=::NewEnumerator4
    Sprite Size Mode=::NewEnumerator4
    UsePositionOffset=false
    Write Material Random=true
   ) {module=InitializeParticle001}
   &DynamicMaterialParameters(
    Index 0 Param 1=0.5
    Index 0 Param 2=1
    Index 0 Param 3=20
    Module.Param0WriteEnabled=true
    Module.Param1WriteEnabled=true
    Module.Param2WriteEnabled=true
    Module.Param3WriteEnabled=true
   )
   &InitialMeshOrientation(Mesh Orientation Mode=::NewEnumerator2, Rotation=<0 -0.25 0>, Use Rotation Vector=true)
   &InitialMeshOrientation(Mesh Orientation Mode=::NewEnumerator2, Rotation=<0.15 -0.1 0.3>, Use Rotation Vector=true) {module=InitialMeshOrientation001}
  ]
  particleUpdateScript [
   &ParticleState(Kill Particles When Lifetime Has Elapsed=false, Let Infinitely Lived Particles Die When Emitter Deactivates=true)
   &SkeletalMeshLocation(
    Bone / Socket Index=1
    Bone Sampling Mode=::NewEnumerator4
    "Interpolate Mesh Sampling (Update)"=true
    MaskSpawnGroup=false
    Sampled Position Offset=&Lerp_Vector(A=<-22 -3 -8>, B=<-1.5 0 -11>, Alpha=$User.PocketTrail_Location)
    Skeletal Mesh=$System.Skeletal Mesh Reader
   ) {module=SkeletalMeshLocation001}
   &InterpolateOverTime(Rate Of Change=5, Target Vector Value:NiagaraPosition=$Particles.Position, Variable Type=::NewEnumerator2)
   /SetVariables(Particles.Interpolate Dynamic=&MakeFloat4FromVectorAndFloat("Vector (RGB)"=$Output.InterpolateOverTime.Moving Average, "Float (Alpha)"=1)) {isSetParameters=true, module=SetVariables_#17}
   &SolveForcesAndVelocity(Clamp Velocity=false, Limit Acceleration=false)
   &ScaleColor(
    Scale Alpha=&MakeCustomFloatFromBool(
     Bool=$User.bisFrontEndPreview
     True Float=&Multiply_Float(
      A=0.02
      B=&FloatFromCurve(FloatCurve=DataInterface() {propertyValues {Curve {keys [<at -0.0012 0>,<at 0.3 0>,<at 1 1>]}}})
     )
     False Float=0.05
    )
    Scale RGB=&MakeVectorFromLinearColorRGB(Color=&SelectLinearColorFromArray(
      Array Sampling Mode=::NewEnumerator1
      Direct Array Index=&MakeIntFromFloat(Float=$User.VariantSwitch)
      Color Selection Array=DataInterface() {propertyValues {ColorData [
         <rgba 0.0072 0.4531 0.2654 1.5>
         <rgba 0.4372 0.6146 0.0709 8>
         <rgba 0.5521 0.1323 0.2613 8>
         <rgba 0.3481 0.1522 0.6146 8>
         <rgba 0.6146 0.4855 0.032 8>
        ]}}
      Random Seed=?
     ))
    ScaleA=true
    ScaleRGB=true
   )
  ]
  rendererClasses [&NiagaraMeshRendererProperties]
  renderers [{rendererClass=&NiagaraMeshRendererProperties, rendererIndex=0}]
 }
}
summary {
 systemName=NS_Character_LocalTalent_Flare_Idle
 emitters [
  {emitterName=Head_Trail, rendererClasses [&NiagaraMeshRendererProperties]}
  {emitterName=Head_Trail_soft, rendererClasses [&NiagaraMeshRendererProperties]}
  {emitterName=Head_Sprites, rendererClasses [&NiagaraSpriteRendererProperties]}
  {emitterName=Head_Light, rendererClasses [&NiagaraLightRendererProperties]}
  {emitterName=Head_Trail001, rendererClasses [&NiagaraMeshRendererProperties]}
  {emitterName=Trail_soft001, rendererClasses [&NiagaraMeshRendererProperties]}
  {emitterName=Sprites001, rendererClasses [&NiagaraSpriteRendererProperties]}
  {emitterName=Light001, rendererClasses [&NiagaraLightRendererProperties, &NiagaraSpriteRendererProperties]}
 ]
 userVariables [
  User.bisFrontEndPreview {type=NiagaraBool}
  User.bisHeadHidden {type=NiagaraBool}
  User.PocketTrail_Location {type=NiagaraFloat, defaultValue=0, description=""}
  User.VariantSwitch {type=NiagaraFloat, defaultValue=0, description=""}
 ]
}
