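# Niagara burst effect for the Local Talent Flare character cosmetic, flashing stars, rays, a flare and a wave at a pocket.
# Six mesh emitters burst with staggered timing and follow skeletal mesh samples, with User.PocketTrail_Location blending offsets and most orientations between two placements.
# User.VariantSwitch picks colours from five-entry arrays and curve blends in most emitters.
# Flare alpha scales by front end, time of day and platform, and Wave dims in front-end preview.
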NiagaraSystem /BRCosmetics/Characters/Player/Female/Medium/Bodies/F_MED_LocalTalent_Flare/FX/NS_Character_LocalTalent_Flare_PocketBurst
dependencies {
 usedDataInterfaces [
  &NiagaraDataInterfaceCurve
  &NiagaraDataInterfaceArrayFloat3
  &NiagaraDataInterfaceArrayColor
  &NiagaraDataInterfaceColorCurve
  &NiagaraDataInterfacePlatformSet
  &NiagaraDataInterfaceArrayFloat
  &NiagaraDataInterfaceSkeletalMesh
 ]
 usedDynamicInputs []
 usedModules [
  &self:SystemSpawnScript.SystemScriptSource.SystemScriptGraph.NiagaraNodeAssignment_0.SetVariables_#1
  &SystemState
 ]
}
emitters {
 Flare {
  emitterUpdateScript [
   &EmitterState(
    Inactive Response=::NewEnumerator1
    Life Cycle Mode=::NewEnumerator1
    Loop Behavior=::NewEnumerator1
    Loop Delay=0.3
    Loop Duration=1
    UseLoopDelay=true
   )
   &SpawnBurst_Instantaneous(Spawn Count=1, Use Spawn Probability=true)
  ]
  particleSpawnScript [
   &InitializeParticle(
    Color=<rgba 0.4167 0.4167 0.4167 5>
    Lifetime=0.6
    Mesh Scale=<0.08 0.15 0.15>
    Mesh Scale Mode=::NewEnumerator4
    Position Offset=<-1.5 0 0>
    UsePositionOffset=true
    Write Material Random=true
   ) {module=InitializeParticle001}
   &InitialMeshOrientation(
    Mesh Orientation Mode=::NewEnumerator2
    Rotation=&Lerp_Vector(A=<-0.14 0.2 0.2>, B=<0 0.2 0.12>, Alpha=$User.PocketTrail_Location)
    Use Rotation Vector=true
   )
  ]
  particleUpdateScript [
   &ParticleState()
   &SkeletalMeshLocation(
    Bone Sampling Mode=::NewEnumerator4
    MaskSpawnGroup=false
    Sampled Position Offset=&Lerp_Vector(A=<-22 2 -10>, B=<-3 8 -11>, Alpha=$User.PocketTrail_Location)
    Skeletal Mesh=$System.Skeletal Mesh Reader
   )
   &ScaleColor(
    Scale Alpha=&FloatFromCurve(
     FloatCurve=DataInterface() {propertyValues {Curve {keys [
         {arriveTangent=-3.8906, interpMode=RCIM_Cubic, leaveTangent=-3.8906, tangentMode=RCTM_User, time=0, value=1}
         {arriveTangent=-1.0522, interpMode=RCIM_Cubic, leaveTangent=-1.0522, tangentMode=RCTM_User, time=0.2, value=0.4}
         <at 1 0 RCIM_Cubic>
        ]}}}
     Scale Curve=&Multiply_Float(
      A=&FN_FrontEndFloatModifier(Gameplay=1, Front End=0.5)
      B=&Multiply_Float(
       A=&FNTimeOfDayScalar(Night Scale=0.5, Morning Scale=1, Evening Scale=1)
       B=&FN_PickFloatDependOnPlatform(TargetPlatformActive=1, TargetPlatformNoActive=0.5)
      )
     )
    )
    Scale RGB=&Multiply_Vector(
     A=/Niagara/DynamicInputs/Math/LerpMultipleVectors(
      Alpha:NiagaraFloat=$User.VariantSwitch
      V 1:Vector3f=&MakeVectorFromLinearColorRGB(Color=&ColorFromCurve(ColorCurve=DataInterface() {propertyValues {
          AlphaCurve {keys []}
          BlueCurve {keys [<at 0.259 0>,<at 0.401 0.6094>]}
          GreenCurve {keys [<at 0.259 1.2188>,<at 0.401 0.1543>]}
          RedCurve {keys [<at 0.259 0.0323>,<at 0.401 0>]}
         }}))
      V 2:Vector3f=&MakeVectorFromLinearColorRGB(Color=&ColorFromCurve(ColorCurve=DataInterface() {propertyValues {
          AlphaCurve {keys []}
          BlueCurve {keys [<at 0.259 0>,<at 0.401 0.6094>]}
          GreenCurve {keys [<at 0.259 1.2188>,<at 0.401 0.1543>]}
          RedCurve {keys [<at 0.259 0.0323>,<at 0.401 0>]}
         }}))
      V 3:Vector3f=&MakeVectorFromLinearColorRGB(Color=&ColorFromCurve(ColorCurve=DataInterface() {propertyValues {
          AlphaCurve {keys []}
          BlueCurve {keys [<at 0.2577 0.3724 RCIM_Cubic>,<at 0.4674 1 RCIM_Cubic>]}
          GreenCurve {keys [<at 0.2577 0.1771 RCIM_Cubic>,<at 0.4674 0.2514 RCIM_Cubic>]}
          RedCurve {keys [<at 0.2577 1 RCIM_Cubic>,<at 0.4674 0.0379 RCIM_Cubic>]}
         }}))
      V 4:Vector3f=&MakeVectorFromLinearColorRGB(Color=&ColorFromCurve(ColorCurve=DataInterface() {propertyValues {
          AlphaCurve {keys []}
          BlueCurve {keys [<at 0.2105 0.0359 RCIM_Cubic>,<at 0.3334 1 RCIM_Cubic>]}
          GreenCurve {keys [<at 0.2105 0.6771 RCIM_Cubic>,<at 0.3334 0.1095 RCIM_Cubic>]}
          RedCurve {keys [<at 0.2105 0 RCIM_Cubic>,<at 0.3334 0.1904 RCIM_Cubic>]}
         }}))
      V 5:Vector3f=&MakeVectorFromLinearColorRGB(Color=&ColorFromCurve(ColorCurve=DataInterface() {propertyValues {
          AlphaCurve {keys []}
          BlueCurve {keys [<at 0.3397 0.0309 RCIM_Cubic>,<at 0.5589 0 RCIM_Cubic>]}
          GreenCurve {keys [<at 0.3397 0.4582 RCIM_Cubic>,<at 0.5589 0.0751 RCIM_Cubic>]}
          RedCurve {keys [<at 0.3397 0.8906 RCIM_Cubic>,<at 0.5589 1 RCIM_Cubic>]}
         }}))
      Vector Count:ENiagara_MultipleLerpCount=::NewEnumerator2
     )
     B=&VectorFromFloat(Value=&FloatFromCurve(
       FloatCurve=DataInterface() {propertyValues {Curve {keys [{arriveTangent=-1.3603, interpMode=RCIM_Cubic, leaveTangent=-1.3603, tangentMode=RCTM_User, time=0, value=0.8}, <at 1 0 RCIM_Cubic>]}}}
       Scale Curve=0.3
      ))
    )
    ScaleA=true
    ScaleRGB=true
   ) {module=ScaleColor001}
   &ScaleMeshSize(Scale Factor=&VectorFromFloat(Value=&FloatFromCurve(FloatCurve=DataInterface() {propertyValues {Curve {keys [<at 0 0>,<at 0.2 1.375>,<at 1 2>]}}}, Scale Curve=0.85)))
   &DynamicMaterialParameters(
    Index 0 Param 1=&FloatFromCurve(FloatCurve=DataInterface() {propertyValues {Curve {keys [{arriveTangent=-0.466, interpMode=RCIM_Cubic, leaveTangent=-0.466, tangentMode=RCTM_User, time=0, value=0.5}, <at 1 0.2>]}}})
    Index 0 Param 2=&FloatFromCurve(FloatCurve=DataInterface() {propertyValues {Curve {keys [{arriveTangent=-0.1866, interpMode=RCIM_Cubic, leaveTangent=-0.1866, tangentMode=RCTM_User, time=0, value=1}, <at 1 0.9>]}}})
    Index 0 Param 3=&FloatFromCurve(FloatCurve=DataInterface() {propertyValues {Curve {keys [<at 0 0>,<at 0.15 1>,<at 1 1.2>]}}})
    Module.Param0WriteEnabled=true
    Module.Param1WriteEnabled=true
    Module.Param2WriteEnabled=true
    Module.Param3WriteEnabled=false
   )
   /SetVariables(Particles.Dyn002:Vector4f=&MakeVector4(
     X=&FloatFromCurve(
      FloatCurve=DataInterface() {propertyValues {Curve {keys [{arriveTangent=-0.6916, interpMode=RCIM_Cubic, leaveTangent=-0.6916, tangentMode=RCTM_User, time=0, value=0.5}, <at 1 0.1>]}}}
      Scale Curve=0.6
     )
     Y=&FloatFromCurve(
      FloatCurve=DataInterface() {propertyValues {Curve {keys [{arriveTangent=-0.1866, interpMode=RCIM_Cubic, leaveTangent=-0.1866, tangentMode=RCTM_User, time=0, value=1}, <at 1 0.9>]}}}
      Scale Curve=0.4
     )
     Z=&FloatFromCurve(FloatCurve=DataInterface() {propertyValues {Curve {keys [<at 0 0>,<at 0.15 1>,<at 1 1.2>]}}})
     W=0
    )) {isSetParameters=true}
  ]
  rendererClasses [&NiagaraMeshRendererProperties]
  renderers [{rendererClass=&NiagaraMeshRendererProperties, rendererIndex=0}]
 }
 Ray_Line {
  emitterUpdateScript [
   &EmitterState(Inactive Response=::NewEnumerator1, Life Cycle Mode=::NewEnumerator1, Loop Behavior=::NewEnumerator1, Loop Duration=1, UseLoopDelay=false)
   &SpawnBurst_Instantaneous(Spawn Count=2, Use Spawn Probability=true, UseLoopCountLimit=false)
  ]
  particleSpawnScript [
   &PartitionParticles()
   &InitializeParticle(
    Lifetime=1.5
    Mesh Scale=<0.5 -1.2 -1.2>
    Mesh Scale Mode=::NewEnumerator4
    Position Offset=<-3 0 0>
    Sprite Rotation Angle=&RandomRangeFloat(Minimum=0, Maximum=1)
    Sprite Rotation Mode=::NewEnumerator1
    UsePositionOffset=true
    Write Material Random=true
   )
   &InitialMeshOrientation(
    Mesh Orientation Mode=::NewEnumerator2
    Rotation=&SelectVectorFromArray(
     Direct Array Index=$Particles.PartitionParticles.Partition
     Vector Selection Array=DataInterface() {propertyValues {FloatData [<0 0 0>, <0.5 0 0>]}}
    )
    Use Rotation Vector=true
   )
   &InitialMeshOrientation(
    Mesh Orientation Mode=::NewEnumerator2
    Rotation=&Lerp_Vector(A=<-0.14 0.2 0.2>, B=<0 0.2 0.12>, Alpha=$User.PocketTrail_Location)
    Use Rotation Vector=true
   ) {module=InitialMeshOrientation001}
  ]
  particleUpdateScript [
   &ParticleState()
   &SkeletalMeshLocation(
    Bone Sampling Mode=::NewEnumerator4
    MaskSpawnGroup=false
    Sampled Position Offset=&Lerp_Vector(A=<-22 2 -10>, B=<-3 8 -11>, Alpha=$User.PocketTrail_Location)
    Skeletal Mesh=$System.Skeletal Mesh Reader
   )
   &UpdateMeshOrientation()
   &ScaleMeshSize(Scale Factor=&VectorFromFloat(Value=&FloatFromCurve(
      FloatCurve=DataInterface() {propertyValues {Curve {keys [
          {arriveTangent=2.3969, interpMode=RCIM_Cubic, leaveTangent=2.3969, tangentMode=RCTM_User, time=0, value=0}
          {arriveTangent=0.0007, interpMode=RCIM_Cubic, leaveTangent=0.0007, tangentMode=RCTM_User, time=0.2, value=0.35}
          {arriveTangent=-0.4123, interpMode=RCIM_Cubic, leaveTangent=-0.4123, tangentMode=RCTM_User, time=1, value=0.1}
         ]}}}
      Scale Curve=&RandomRangeFloat(Minimum=0.65, Maximum=0.65)
     )))
   &ScaleColor(
    Scale Alpha=&FloatFromCurve(
     FloatCurve=DataInterface() {propertyValues {Curve {keys [
         <at 0 0 RCIM_Cubic>
         {arriveTangent=2.8571, interpMode=RCIM_Cubic, leaveTangent=2.8571, time=0.4, value=1}
         {arriveTangent=0.0225, interpMode=RCIM_Cubic, leaveTangent=0.0225, tangentMode=RCTM_User, time=0.7, value=2}
         <at 1 0 RCIM_Cubic>
        ]}}}
     Scale Curve=10
    )
    Scale RGB=&MakeVectorFromLinearColorRGB(Color=&SelectLinearColorFromArray(
      Array Sampling Mode=::NewEnumerator1
      Direct Array Index=&MakeIntFromFloat(Float=$User.VariantSwitch)
      Color Selection Array=DataInterface() {propertyValues {ColorData [<rgba 0 30 23.913 1>, <rgba 0 30 23.913 1>, <rgba 30 6.7757 17.8271 1>, <rgba 6.9718 1.5 30 1>, <rgba 30 7.608 1.0874 1>]}}
      Random Seed=?
     ))
    ScaleA=true
    ScaleRGB=true
   )
  ]
  rendererClasses [&NiagaraMeshRendererProperties]
  renderers [{rendererClass=&NiagaraMeshRendererProperties, rendererIndex=0}]
 }
 Rays {
  emitterUpdateScript [
   &EmitterState(
    Inactive Response=::NewEnumerator1
    Life Cycle Mode=::NewEnumerator1
    Loop Behavior=::NewEnumerator1
    Loop Delay=0.3
    Loop Duration=1
    UseLoopDelay=true
   )
   &SpawnBurst_Instantaneous(Spawn Count=2, Use Spawn Probability=false, UseLoopCountLimit=false)
  ]
  particleSpawnScript [
   &InitializeParticle(
    Color=<rgba 0 1 0.1685 0.5>
    Lifetime=&RandomRangeFloat(Minimum=0.5, Maximum=0.6)
    Mesh Scale=<0.3 -1 -1>
    Mesh Scale Mode=::NewEnumerator4
    Sprite Rotation Angle=&RandomRangeFloat(Minimum=0, Maximum=1)
    Sprite Rotation Mode=::NewEnumerator1
    UsePositionOffset=true
    Write Material Random=true
   )
   &Color(Color=&MakeLinearColorFromVectorAndFloat("Vector (RGB)"=<0.1 0.002 0.12>, "Float (Alpha)"=&ClampFloat(Float=1)))
   &DynamicMaterialParameters(
    Index 0 Param 1=&RandomRangeFloat(Minimum=0, Maximum=10)
    Index 0 Param 2=&RandomRangeFloat(Minimum=2.5, Maximum=4.5)
    Index 0 Param 3=&RandomRangeFloat(Minimum=0.8, Maximum=1.6)
    Module.Param0WriteEnabled=true
    Module.Param1WriteEnabled=true
    Module.Param2WriteEnabled=true
    Module.Param3WriteEnabled=false
   )
   &CameraOffset(Camera Offset Amount=10)
  ]
  particleUpdateScript [
   &ParticleState()
   &SkeletalMeshLocation(
    Bone Sampling Mode=::NewEnumerator4
    MaskSpawnGroup=false
    Orientation Sampling=::NewEnumerator1
    Sampled Position Offset=&Lerp_Vector(A=<-22 2 -10>, B=<-2 7.5 -10.5>, Alpha=$User.PocketTrail_Location)
    Skeletal Mesh=$System.Skeletal Mesh Reader
   )
   &ScaleMeshSize(Scale Factor=&VectorFromFloat(Value=&FloatFromCurve(
      FloatCurve=DataInterface() {propertyValues {Curve {keys [
          {arriveTangent=15.9108, interpMode=RCIM_Cubic, time=0, value=0.0062}
          {arriveTangent=1.323, interpMode=RCIM_Cubic, leaveTangent=1.323, time=0.2, value=0.6}
          {arriveTangent=-0.0562, interpMode=RCIM_Cubic, leaveTangent=-0.0562, tangentMode=RCTM_User, time=0.6, value=0.8}
          {arriveTangent=-0.9004, interpMode=RCIM_Cubic, leaveTangent=-0.9004, tangentMode=RCTM_User, time=1, value=0.4}
         ]}}}
      Scale Curve=&RandomRangeFloat(Minimum=0.6, Maximum=0.8)
     )))
   &ScaleColor(
    Scale Alpha=&FloatFromCurve(FloatCurve=DataInterface() {propertyValues {Curve {keys [
         <at 0 0 RCIM_Cubic>
         {arriveTangent=-2.0099, interpMode=RCIM_Cubic, leaveTangent=-2.0099, tangentMode=RCTM_User, time=0.3, value=1}
         <at 1 0 RCIM_Cubic>
        ]}}})
    Scale RGB=&Multiply_VectorByFloat(Vector=<1 1 1>, Float=0.8)
    ScaleA=true
    ScaleRGB=true
   )
   &DynamicMaterialParameters(
    Index 0 Param 4=1
    Module.Param0WriteEnabled=false
    Module.Param1WriteEnabled=false
    Module.Param2WriteEnabled=false
    Module.Param3WriteEnabled=true
   ) {module=DynamicMaterialParameters001}
  ]
  rendererClasses [&NiagaraMeshRendererProperties]
  renderers [{rendererClass=&NiagaraMeshRendererProperties, rendererIndex=0}]
 }
 Star_Glowmesh {
  emitterUpdateScript [
   &EmitterState(
    Inactive Response=::NewEnumerator1
    Life Cycle Mode=::NewEnumerator1
    Loop Behavior=::NewEnumerator1
    Loop Delay=0.3
    Loop Duration=1
    UseLoopDelay=true
   )
   &SpawnBurst_Instantaneous(Use Spawn Probability=false, UseLoopCountLimit=false)
  ]
  particleSpawnScript [
   &InitializeParticle(
    Color=&SelectLinearColorFromArray(
     Array Sampling Mode=::NewEnumerator1
     Direct Array Index=&MakeIntFromFloat(Float=$User.VariantSwitch)
     Color Selection Array=DataInterface() {propertyValues {ColorData [<rgba 1 1 1 1>, <rgba 1 1 1 1>, <rgba 1 1 1 1>, <rgba 0.2282 2 0.1659 1>, <rgba 1.5 1.5 1.5 1>]}}
     Random Seed=?
    )
    Lifetime=0.5
    Mesh Scale=<0.5 0.5 1.4>
    Mesh Scale Mode=::NewEnumerator4
    Position Offset=<-2 0 0>
    Sprite Size Mode=::NewEnumerator3
    Uniform Sprite Size=23
    UsePositionOffset=true
    Write Material Random=true
   )
   &InitialMeshOrientation(
    Mesh Orientation Mode=::NewEnumerator2
    Rotation=&Lerp_Vector(A=<0 0.52 0.06>, B=<0 0.46 0.15>, Alpha=$User.PocketTrail_Location)
    Use Rotation Vector=true
   )
  ]
  particleUpdateScript [
   &ParticleState()
   &SkeletalMeshLocation(
    Bone Sampling Mode=::NewEnumerator4
    MaskSpawnGroup=false
    Sampled Position Offset=&Lerp_Vector(A=<-22 2 -10>, B=<-3 7 -12>, Alpha=$User.PocketTrail_Location)
    Skeletal Mesh=$System.Skeletal Mesh Reader
   )
   &ScaleMeshSize(Scale Factor=&MakeVector(
     X=1
     Y=1
     Z=&FloatFromCurve(FloatCurve=DataInterface() {propertyValues {Curve {keys [
          {arriveTangent=1.9289, interpMode=RCIM_Cubic, leaveTangent=1.9289, tangentMode=RCTM_User, time=0, value=0.3}
          {arriveTangent=0.0001, interpMode=RCIM_Cubic, leaveTangent=0.0001, tangentMode=RCTM_User, time=1, value=1.1}
         ]}}})
    ))
   &ScaleColor(
    Scale Alpha=&FloatFromCurve(FloatCurve=DataInterface() {propertyValues {Curve {keys [
         <at 0 0 RCIM_Cubic>
         {arriveTangent=-2.0518, interpMode=RCIM_Cubic, leaveTangent=-2.0518, tangentMode=RCTM_User, time=0.2, value=1}
         <at 1 0 RCIM_Cubic>
        ]}}})
    Scale RGB=&Multiply_Vector(
     A=&LerpMultipleVectors(
      Alpha:NiagaraFloat=$User.VariantSwitch
      V 1:Vector3f=&MakeVectorFromLinearColorRGB(Color=&ColorFromCurve(ColorCurve=DataInterface() {propertyValues {
          AlphaCurve {keys []}
          BlueCurve {keys [{arriveTangent=1.6431, interpMode=RCIM_Cubic, time=0.3693, value=0}, <at 0.5103 0.823>]}
          GreenCurve {keys [{arriveTangent=-0.9364, interpMode=RCIM_Cubic, time=0.3693, value=1}, <at 0.5103 1>]}
          RedCurve {keys [{arriveTangent=-1.6426, interpMode=RCIM_Cubic, time=0.3693, value=0.0389}, <at 0.5103 0>]}
         }}))
      V 2:Vector3f=&MakeVectorFromLinearColorRGB(Color=&ColorFromCurve(ColorCurve=DataInterface() {propertyValues {
          AlphaCurve {keys []}
          BlueCurve {keys [{arriveTangent=1.6431, interpMode=RCIM_Cubic, time=0.3693, value=0.0384}, <at 0.5103 0.0046>]}
          GreenCurve {keys [{arriveTangent=-0.9364, interpMode=RCIM_Cubic, time=0.3693, value=1}, <at 0.5103 1>]}
          RedCurve {keys [{arriveTangent=-1.6426, interpMode=RCIM_Cubic, time=0.3693, value=0.5517}, <at 0.5103 0.0517>]}
         }}))
      V 3:Vector3f=&MakeVectorFromLinearColorRGB(Color=&ColorFromCurve(ColorCurve=DataInterface() {propertyValues {
          AlphaCurve {keys []}
          BlueCurve {keys [<at 0.3817 0.3724 RCIM_Cubic>,<at 0.5706 1 RCIM_Cubic>]}
          GreenCurve {keys [<at 0.3817 0.1771 RCIM_Cubic>,<at 0.5706 0.2514 RCIM_Cubic>]}
          RedCurve {keys [<at 0.3817 1 RCIM_Cubic>,<at 0.5706 0.0379 RCIM_Cubic>]}
         }}))
      V 4:Vector3f=&MakeVectorFromLinearColorRGB(Color=&ColorFromCurve(ColorCurve=DataInterface() {propertyValues {
          AlphaCurve {keys []}
          BlueCurve {keys [<at 0.3436 0 RCIM_Cubic>,<at 0.5066 1 RCIM_Cubic>]}
          GreenCurve {keys [<at 0.3436 0.849 RCIM_Cubic>,<at 0.5066 0.0542 RCIM_Cubic>]}
          RedCurve {keys [<at 0.3436 0.0289 RCIM_Cubic>,<at 0.5066 0.2154 RCIM_Cubic>]}
         }}))
      V 5:Vector3f=&MakeVectorFromLinearColorRGB(Color=&ColorFromCurve(ColorCurve=DataInterface() {propertyValues {
          AlphaCurve {keys []}
          BlueCurve {keys [<at 0.467 0 RCIM_Cubic>,<at 0.79 0 RCIM_Cubic>]}
          GreenCurve {keys [<at 0.467 0.5419 RCIM_Cubic>,<at 0.79 0.0892 RCIM_Cubic>]}
          RedCurve {keys [<at 0.467 1 RCIM_Cubic>,<at 0.79 1 RCIM_Cubic>]}
         }}))
      Vector Count:ENiagara_MultipleLerpCount=::NewEnumerator2
     )
     B=&VectorFromFloat(Value=0.5)
    )
    ScaleA=true
    ScaleRGB=true
   )
   &SolveForcesAndVelocity(Clamp Velocity=false, Limit Acceleration=false)
  ]
  rendererClasses [&NiagaraMeshRendererProperties]
  renderers [{rendererClass=&NiagaraMeshRendererProperties, rendererIndex=0}]
 }
 Star_mesh {
  emitterUpdateScript [
   &EmitterState(Inactive Response=::NewEnumerator1, Life Cycle Mode=::NewEnumerator1, Loop Behavior=::NewEnumerator1, Loop Duration=1, UseLoopDelay=false)
   &SpawnBurst_Instantaneous(Spawn Time=0.3, Use Spawn Probability=false, UseLoopCountLimit=false)
  ]
  particleSpawnScript [
   &InitializeParticle(Color Mode=::NewEnumerator0, Lifetime=0.6, Mesh Scale Mode=::NewEnumerator4, UsePositionOffset=true, Write Material Random=true)
   &InitialMeshOrientation(
    Mesh Orientation Mode=::NewEnumerator2
    Rotation=&Lerp_Vector(A=<-0.14 0.2 0.2>, B=<0 0.2 0.12>, Alpha=$User.PocketTrail_Location)
    Use Rotation Vector=true
   )
  ]
  particleUpdateScript [
   &ParticleState()
   &SkeletalMeshLocation(
    Bone Sampling Mode=::NewEnumerator4
    MaskSpawnGroup=false
    Sampled Position Offset=&Lerp_Vector(A=<-22 2 -10>, B=<-3 8 -11>, Alpha=$User.PocketTrail_Location)
    Skeletal Mesh=$System.Skeletal Mesh Reader
   )
   &ScaleColor(
    Scale Alpha=&FloatFromCurve(FloatCurve=DataInterface() {propertyValues {Curve {keys [
         {arriveTangent=0.9513, interpMode=RCIM_Cubic, leaveTangent=0.9513, tangentMode=RCTM_User, time=0, value=0}
         {arriveTangent=10.1951, interpMode=RCIM_Cubic, leaveTangent=-5.7722, tangentMode=RCTM_Break, time=0.1, value=1.5}
         {arriveTangent=-1.8916, interpMode=RCIM_Cubic, leaveTangent=-1.8916, tangentMode=RCTM_User, time=0.3, value=0.575}
         {arriveTangent=-0.2689, interpMode=RCIM_Cubic, leaveTangent=-0.2689, tangentMode=RCTM_User, time=1, value=0}
        ]}}})
    Scale RGB=&Multiply_VectorByFloat(
     Vector=&MakeVectorFromLinearColorRGB(Color=&SelectLinearColorFromArray(
       Array Sampling Mode=::NewEnumerator1
       Direct Array Index=&MakeIntFromFloat(Float=$User.VariantSwitch)
       Color Selection Array=DataInterface() {propertyValues {ColorData [<rgba 1.2 0.7326 0.1114 1>, <rgba 1.2 0.7326 0.1114 1>, <rgba 0.3091 0.2767 1.2 1>, <rgba 0.4199 0.1977 1.2 1>, <rgba 1.2 0.5388 0.0778 1>]}}
       Random Seed=?
      ))
     Float=1
    )
    ScaleA=true
    ScaleRGB=true
   )
   &ScaleMeshSize(Scale Factor=&Multiply_Vector(
     A=<0.5 0.5 0.5>
     B=&VectorFromFloat(Value=&FloatFromCurve(FloatCurve=DataInterface() {propertyValues {Curve {keys [{arriveTangent=1.8728, interpMode=RCIM_Cubic, leaveTangent=1.8728, tangentMode=RCTM_User, time=0, value=1}, <at 1 2 RCIM_Cubic>]}}}))
    ))
   &DynamicMaterialParameters(
    Index 0 Param 1=&FloatFromCurve(FloatCurve=DataInterface() {propertyValues {Curve {keys [
         {arriveTangent=-4.0585, interpMode=RCIM_Cubic, leaveTangent=-4.0585, tangentMode=RCTM_User, time=0, value=1}
         {arriveTangent=-1.0113, interpMode=RCIM_Cubic, leaveTangent=-1.0113, tangentMode=RCTM_User, time=1, value=-1.5}
        ]}}})
    Module.Index0Param0Write=true
    Module.Index0Param1Write=true
    Module.Index0Param2Write=true
    Module.Index0Param3Write=true
   )
  ]
  rendererClasses [&NiagaraMeshRendererProperties]
  renderers [{rendererClass=&NiagaraMeshRendererProperties, rendererIndex=0}]
 }
 Wave {
  emitterUpdateScript [
   &EmitterState(
    Inactive Response=::NewEnumerator1
    Life Cycle Mode=::NewEnumerator1
    Loop Behavior=::NewEnumerator1
    Loop Delay=0.3
    Loop Duration=1
    UseLoopDelay=true
   )
   &SpawnBurst_Instantaneous(Use Spawn Probability=false, UseLoopCountLimit=false)
  ]
  particleSpawnScript [
   &InitializeParticle(
    Color Mode=::NewEnumerator0
    Lifetime=0.35
    Mesh Scale=<0.8 0.8 0.4>
    Mesh Scale Mode=::NewEnumerator4
    UsePositionOffset=false
    Write Material Random=true
   )
   &InitialMeshOrientation(
    Mesh Orientation Mode=::NewEnumerator2
    Rotation=&Lerp_Vector(A=<0 0.52 0.06>, B=<0 0.46 0.15>, Alpha=$User.PocketTrail_Location)
    Use Rotation Vector=true
   )
  ]
  particleUpdateScript [
   &ParticleState()
   &SkeletalMeshLocation(
    Bone Sampling Mode=::NewEnumerator4
    MaskSpawnGroup=false
    Sampled Position Offset=&Lerp_Vector(A=<-22 2 -14>, B=<-3 10 -14>, Alpha=$User.PocketTrail_Location)
    Skeletal Mesh=$System.Skeletal Mesh Reader
   )
   &ScaleMeshSize(Scale Factor=&VectorFromFloat(Value=&FloatFromCurve(FloatCurve=DataInterface() {propertyValues {Curve {keys [
          {arriveTangent=0.9909, interpMode=RCIM_Cubic, leaveTangent=0.9909, tangentMode=RCTM_User, time=0, value=0.7}
          {arriveTangent=0.0685, interpMode=RCIM_Cubic, leaveTangent=0.0685, tangentMode=RCTM_User, time=1, value=1.3}
         ]}}})))
   &DynamicMaterialParameters(
    Index 0 Param 1=1
    Index 0 Param 3=&FloatFromCurve(
     FloatCurve=DataInterface() {propertyValues {Curve {keys [<at 0 0.55>, {arriveTangent=0.8333, time=0.1, value=1}, {arriveTangent=-0.875, leaveTangent=-0.875, time=0.5, value=1}, <at 1 0.4>]}}}
     Scale Curve=1.5
    )
    Module.Index0Float0Write=true
    Module.Index0Float1Write=true
    Module.Index0Float2Write=true
    Module.Index0Float3Write=true
   )
   &ScaleColor(
    Scale Alpha=&FloatFromCurve(
     FloatCurve=DataInterface() {propertyValues {Curve {keys [<at 0 0>,<at 0.1 1>,<at 0.5 1>,<at 1 0>]}}}
     Scale Curve=&Multiply_Float(
      A=&MakeCustomFloatFromBool(Bool=$User.bIsFrontEndPreview, True Float=0.5, False Float=1)
      B=&SelectFloatFromArray(
       Direct Array Index=&MakeIntFromFloat(Float=$User.VariantSwitch)
       Float Selection Array=DataInterface() {propertyValues {FloatData [1,0.7]}}
      )
     )
    )
    Scale RGB=&MakeVectorFromLinearColorRGB(Color=&SelectLinearColorFromArray(
      Array Sampling Mode=::NewEnumerator1
      Direct Array Index=&MakeIntFromFloat(Float=$User.VariantSwitch)
      Color Selection Array=DataInterface() {propertyValues {ColorData [<rgba 0 0.6094 0.2983 1>, <rgba 0.5315 0.6 0.0752 1>, <rgba 0.0986 0.4616 0.6094 1>, <rgba 0.2686 0.024 0.6 1>, <rgba 1 0.1265 0 1>]}}
      Random Seed=?
     ))
    ScaleA=true
    ScaleRGB=true
   )
   /SetVariables(Particles.Color002=&MakeLinearColorFromVectorAndFloat(
     "Vector (RGB)"=&MakeVectorFromLinearColorRGB(Color=&SelectLinearColorFromArray(
       Array Sampling Mode=::NewEnumerator1
       Direct Array Index=&MakeIntFromFloat(Float=$User.VariantSwitch)
       Color Selection Array=DataInterface() {propertyValues {ColorData [<rgba 2.5 1.7119 0 1>, <rgba 0.1575 2 0.1544 1>, <rgba 0.0464 0.9143 1.7812 1>, <rgba 3 0.3 1.4651 1>, <rgba 10 0.3333 0 1>]}}
       Random Seed=?
      ))
     "Float (Alpha)"=&FloatFromCurve(FloatCurve=DataInterface() {propertyValues {Curve {keys [<at 0 0>,<at 0.1 1>,<at 0.5 1>,<at 1 0>]}}})
    )) {isSetParameters=true}
   /SetVariables(Particles.scale_small:Vector3f=&Multiply_VectorByFloat(Vector=$Particles.Scale, Float=0.75)) {isSetParameters=true}
   &SolveForcesAndVelocity(Clamp Velocity=false, Limit Acceleration=false)
   /SetVariables(Particles.Color_light02:LinearColor=&MakeLinearColorFromVectorAndFloat(
     "Vector (RGB)"=&Multiply_VectorByFloat(
      Vector=&MakeVectorFromLinearColorRGB(Color=&SelectLinearColorFromArray(
        Array Sampling Mode=::NewEnumerator1
        Direct Array Index=&MakeIntFromFloat(Float=$User.VariantSwitch)
        Color Selection Array=DataInterface() {propertyValues {ColorData [<rgba 3 3 3 1>, <rgba 1.9249 2 0.1961 1>, <rgba 2 0.3074 1.3055 1>, <rgba 0.0026 2 0.3072 1>, <rgba 2 0.7585 0.0945 1>]}}
        Random Seed=?
       ))
      Float=1
     )
     "Float (Alpha)"=&FloatFromCurve(
      FloatCurve=DataInterface() {propertyValues {Curve {keys [<at 0.6 1 RCIM_Cubic>,<at 1 0>]}}}
      Scale Curve=&SelectFloatFromArray(
       Direct Array Index=&MakeIntFromFloat(Float=$User.VariantSwitch)
       Float Selection Array=DataInterface() {propertyValues {FloatData [0,1.5]}}
      )
     )
    )) {isSetParameters=true}
  ]
  rendererClasses [&NiagaraMeshRendererProperties]
  renderers [
   {rendererClass=&NiagaraMeshRendererProperties, rendererIndex=0}
   {rendererClass=&NiagaraMeshRendererProperties, rendererIndex=1}
   {rendererClass=&NiagaraMeshRendererProperties, rendererIndex=2}
   {rendererClass=&NiagaraMeshRendererProperties, rendererIndex=3}
  ]
 }
}
summary {
 systemName=NS_Character_LocalTalent_Flare_PocketBurst
 emitters [
  {emitterName=Rays, rendererClasses [&NiagaraMeshRendererProperties]}
  {emitterName=Ray_Line, rendererClasses [&NiagaraMeshRendererProperties]}
  {emitterName=Star_mesh, rendererClasses [&NiagaraMeshRendererProperties]}
  {emitterName=Star_Glowmesh, rendererClasses [&NiagaraMeshRendererProperties]}
  {emitterName=Flare, rendererClasses [&NiagaraMeshRendererProperties]}
  {emitterName=Wave, rendererClasses [&NiagaraMeshRendererProperties]}
 ]
 userVariables [
  User.bIsFrontEnd {type=NiagaraBool}
  User.bIsFrontEndPreview {type=NiagaraBool}
  User.Hide light {type=NiagaraBool}
  User.IsLeft {type=NiagaraBool}
  User.PocketTrail_Location {type=NiagaraFloat, defaultValue=0, description=""}
  User.VariantSwitch {type=NiagaraFloat, defaultValue=0, description=""}
 ]
}
