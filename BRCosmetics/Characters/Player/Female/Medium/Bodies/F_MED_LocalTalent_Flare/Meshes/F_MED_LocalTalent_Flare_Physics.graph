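# Physics asset for the F_MED_LocalTalent_Flare female character body, set up for a dangling necklace rather than a full ragdoll.
# A four-link dyn_necklace capsule chain hangs from a zero-size base_necklace_01 anchor locked to spine_05, with limited swing and very little twist per link.
# Capsules on head, spine_04 and spine_05 form the only other bodies, with the spine joint locked and the head joint free.
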PhysicsAsset /BRCosmetics/Characters/Player/Female/Medium/Bodies/F_MED_LocalTalent_Flare/Meshes/F_MED_LocalTalent_Flare_Physics
bodies {
 base_necklace_01 [{shapeType=Capsule, length=0, radius=0, shapeName=base_necklace_01_capsule}]
 dyn_necklace_01 [{shapeType=Capsule, length=8.3017, radius=1.7374, shapeName=dyn_necklace_01_capsule, center=<1.3355 0 1.7596>, rotation=<rot 0 0 90>}]
 dyn_necklace_02 [{shapeType=Capsule, length=3.9728, radius=1.7374, shapeName=dyn_necklace_02_capsule, center=<1.3355 0 1.8783>, rotation=<rot 0 0 90>}]
 dyn_necklace_03 [{shapeType=Capsule, length=4.1053, radius=2.2329, shapeName=dyn_necklace_03_capsule, center=<1.7493 0 2.1051>, rotation=<rot 0 0 90>}]
 dyn_necklace_04 [{shapeType=Capsule, length=2.4331, radius=6.9703, shapeName=dyn_necklace_04_capsule, center=<0.759 0 6.5855>, rotation=<rot 85.2311 0 0>}]
 head [{shapeType=Capsule, length=4.031, radius=7.4738, shapeName=head_capsule, center=<1.3969 2.1872 0>, rotation=<rot 90 -18.4459 0>}]
 spine_04 [{shapeType=Capsule, length=13.7544, radius=10.6473, shapeName=spine_04_capsule, center=<3.3331 0.2607 0>}]
 spine_05 [{shapeType=Capsule, length=8.3133, radius=11.0709, shapeName=spine_05_capsule, center=<-8.2878 2.5823 0>}]
}
constraints [
 {
  bone1Name=spine_05
  bone2Name=spine_04
  swing1LimitDegrees=45
  swing1Motion=Locked
  swing2LimitDegrees=45
  swing2Motion=Locked
  twistLimitDegrees=45
  twistMotion=Locked
 }
 {
  bone1Name=base_necklace_01
  bone2Name=spine_05
  swing1LimitDegrees=45
  swing1Motion=Locked
  swing2LimitDegrees=45
  swing2Motion=Locked
  twistLimitDegrees=45
  twistMotion=Locked
 }
 {bone1Name=dyn_necklace_01, bone2Name=base_necklace_01, swing1LimitDegrees=15, swing2LimitDegrees=30, twistLimitDegrees=1}
 {bone1Name=dyn_necklace_02, bone2Name=dyn_necklace_01, swing1LimitDegrees=15, swing2LimitDegrees=30, twistLimitDegrees=1}
 {bone1Name=dyn_necklace_03, bone2Name=dyn_necklace_02, swing1LimitDegrees=15, swing2LimitDegrees=30, twistLimitDegrees=1}
 {bone1Name=dyn_necklace_04, bone2Name=dyn_necklace_03, swing1LimitDegrees=15, swing2LimitDegrees=30, twistLimitDegrees=1}
 {
  bone1Name=head
  bone2Name=spine_05
  swing1LimitDegrees=45
  swing1Motion=Free
  swing2LimitDegrees=45
  swing2Motion=Free
  twistLimitDegrees=45
  twistMotion=Free
 }
]
