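# Post-process style AnimBP for the F_MED_LocalTalent_Flare cosmetic body, adding secondary motion over the pose copied from the attached parent mesh.
# A Control Rig receives that pose and the DBNO flag, then feeds chained spring-driven AnimDynamics nodes for kuzi, cloth, calf ribbon and bag bones.
# A final RigidBody node simulates the body's physics asset relative to spine_04.
# The update event calls Set WPO Globals Func after checking PlayerPawn.
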AnimBlueprint /BRCosmetics/Characters/Player/Female/Medium/Bodies/F_MED_LocalTalent_Flare/Meshes/F_MED_LocalTalent_Flare_AnimBP
target_skeleton=&/F_MED_LocalTalent_Flare_Skeleton
graphs [
 AnimGraph {nodes [
   n1 = Control Rig!(Source=Copy Pose From Mesh() {bUseAttachedParent=true}, Alpha, id dbno:bool=Get bIsDBNO()) -> (Pose) {controlRigAssetReference {blueprintRigClass=&/F_MED_LocalTalent_Flare_CtrlRig.F_MED_LocalTalent_Flare_CtrlRig_C}}
   n2 = Anim Dynamics - Bone: dyn_kuzi_f_r_01(ComponentPose=Local To Component(LocalPose=@n1), Alpha=1, ExternalForce="0,0,0") {
    angularDampingOverride=0.9999
    angularSpringConstant=1000
    bAngularSpring=true
    bGravityOverrideInSimSpace=true
    bOverrideAngularDamping=true
    bUseGravityOverride=true
    gravityScale=0.35
    simulationSpace=RootRelative
    alphaBoolBlend {blendOption=Linear}
    boundBone {boneName=dyn_kuzi_f_r_01}
    componentAppliedLinearAccClamp=<1000 1000 1000>
    componentLinearAccScale=<0.3 0.3 0.3>
    componentLinearVelScale=<0.3 0.3 0.3>
    gravityOverride {z=-500}
    physicsBodyDefinitions [{
      collisionType=CustomSphere
      sphereCollisionRadius=2
      boundBone {boneName=dyn_kuzi_f_r_01}
      boxExtents=<15 15 15>
      constraintSetup {angularLimitsMax=<5 5 2>, angularLimitsMin=<-5 -5 -2>, angularTarget=<0 0 0>}
      localJointOffset=<3 0 -0.2>
     }]
    relativeSpaceBone {boneName=calf_l}
    simSpaceSettings {maxAngularAcceleration=10, maxAngularVelocity=2, simSpaceAngularAlpha=1}
   }
   n3 = Anim Dynamics - Bone: dyn_kuzi_f_r_02(ComponentPose=@n2, Alpha=1, ExternalForce="0,0,0") {
    angularDampingOverride=0.9999
    angularSpringConstant=1000
    bAngularSpring=true
    bGravityOverrideInSimSpace=true
    bOverrideAngularDamping=true
    bUseGravityOverride=true
    gravityScale=0.35
    simulationSpace=RootRelative
    alphaBoolBlend {blendOption=Linear}
    boundBone {boneName=dyn_kuzi_f_r_02}
    componentAppliedLinearAccClamp=<1000 1000 1000>
    componentLinearAccScale=<0.3 0.3 0.3>
    componentLinearVelScale=<0.3 0.3 0.3>
    gravityOverride {z=-500}
    physicsBodyDefinitions [{
      collisionType=CustomSphere
      sphereCollisionRadius=2
      boundBone {boneName=dyn_kuzi_f_r_02}
      boxExtents=<15 15 15>
      constraintSetup {angularLimitsMax=<5 5 2>, angularLimitsMin=<-5 -5 -2>, angularTarget=<0 0 0>}
      localJointOffset=<3 0 -0.2>
     }]
    relativeSpaceBone {boneName=calf_l}
    simSpaceSettings {maxAngularAcceleration=10, maxAngularVelocity=2, simSpaceAngularAlpha=1}
   }
   n4 = Anim Dynamics - Bone: dyn_kuzi_f_r_03(ComponentPose=@n3, Alpha=1, ExternalForce="0,0,0") {
    angularDampingOverride=0.9999
    angularSpringConstant=1000
    bAngularSpring=true
    bGravityOverrideInSimSpace=true
    bOverrideAngularDamping=true
    bUseGravityOverride=true
    gravityScale=0.35
    simulationSpace=RootRelative
    alphaBoolBlend {blendOption=Linear}
    boundBone {boneName=dyn_kuzi_f_r_03}
    componentAppliedLinearAccClamp=<1000 1000 1000>
    componentLinearAccScale=<0.3 0.3 0.3>
    componentLinearVelScale=<0.3 0.3 0.3>
    gravityOverride {z=-500}
    physicsBodyDefinitions [{
      collisionType=CustomSphere
      sphereCollisionRadius=2
      boundBone {boneName=dyn_kuzi_f_r_03}
      boxExtents=<15 15 15>
      constraintSetup {angularLimitsMax=<5 5 2>, angularLimitsMin=<-5 -5 -2>, angularTarget=<0 0 0>}
      localJointOffset=<3 0 -0.2>
     }]
    relativeSpaceBone {boneName=calf_l}
    simSpaceSettings {maxAngularAcceleration=10, maxAngularVelocity=2, simSpaceAngularAlpha=1}
   }
   n5 = Anim Dynamics - Bone: dyn_kuzi_out_r_01(ComponentPose=@n4, Alpha=1, ExternalForce="0,0,0") {
    angularDampingOverride=0.9999
    angularSpringConstant=1000
    bAngularSpring=true
    bGravityOverrideInSimSpace=true
    bOverrideAngularDamping=true
    bUseGravityOverride=true
    gravityScale=0.35
    simulationSpace=RootRelative
    alphaBoolBlend {blendOption=Linear}
    boundBone {boneName=dyn_kuzi_out_r_01}
    componentAppliedLinearAccClamp=<1000 1000 1000>
    componentLinearAccScale=<0.3 0.3 0.3>
    componentLinearVelScale=<0.3 0.3 0.3>
    gravityOverride {z=-500}
    physicsBodyDefinitions [{
      collisionType=CustomSphere
      sphereCollisionRadius=2
      boundBone {boneName=dyn_kuzi_out_r_01}
      boxExtents=<15 15 15>
      constraintSetup {angularLimitsMax=<5 5 2>, angularLimitsMin=<-5 -5 -2>, angularTarget=<0 0 0>}
      localJointOffset=<3 0 -0.2>
     }]
    relativeSpaceBone {boneName=calf_l}
    simSpaceSettings {maxAngularAcceleration=10, maxAngularVelocity=2, simSpaceAngularAlpha=1}
   }
   n6 = Anim Dynamics - Bone: dyn_kuzi_out_r_02(ComponentPose=@n5, Alpha=1, ExternalForce="0,0,0") {
    angularDampingOverride=0.9999
    angularSpringConstant=1000
    bAngularSpring=true
    bGravityOverrideInSimSpace=true
    bOverrideAngularDamping=true
    bUseGravityOverride=true
    gravityScale=0.35
    simulationSpace=RootRelative
    alphaBoolBlend {blendOption=Linear}
    boundBone {boneName=dyn_kuzi_out_r_02}
    componentAppliedLinearAccClamp=<1000 1000 1000>
    componentLinearAccScale=<0.3 0.3 0.3>
    componentLinearVelScale=<0.3 0.3 0.3>
    gravityOverride {z=-500}
    physicsBodyDefinitions [{
      collisionType=CustomSphere
      sphereCollisionRadius=2
      boundBone {boneName=dyn_kuzi_out_r_02}
      boxExtents=<15 15 15>
      constraintSetup {angularLimitsMax=<5 5 2>, angularLimitsMin=<-5 -5 -2>, angularTarget=<0 0 0>}
      localJointOffset=<3 0 -0.2>
     }]
    relativeSpaceBone {boneName=calf_l}
    simSpaceSettings {maxAngularAcceleration=10, maxAngularVelocity=2, simSpaceAngularAlpha=1}
   }
   n7 = Anim Dynamics - Bone: dyn_kuzi_out_r_03(ComponentPose=@n6, Alpha=1, ExternalForce="0,0,0") {
    angularDampingOverride=0.9999
    angularSpringConstant=1000
    bAngularSpring=true
    bGravityOverrideInSimSpace=true
    bOverrideAngularDamping=true
    bUseGravityOverride=true
    gravityScale=0.35
    simulationSpace=RootRelative
    alphaBoolBlend {blendOption=Linear}
    boundBone {boneName=dyn_kuzi_out_r_03}
    componentAppliedLinearAccClamp=<1000 1000 1000>
    componentLinearAccScale=<0.3 0.3 0.3>
    componentLinearVelScale=<0.3 0.3 0.3>
    gravityOverride {z=-500}
    physicsBodyDefinitions [{
      collisionType=CustomSphere
      sphereCollisionRadius=2
      boundBone {boneName=dyn_kuzi_out_r_03}
      boxExtents=<15 15 15>
      constraintSetup {angularLimitsMax=<5 5 2>, angularLimitsMin=<-5 -5 -2>, angularTarget=<0 0 0>}
      localJointOffset=<3 0 -0.2>
     }]
    relativeSpaceBone {boneName=calf_l}
    simSpaceSettings {maxAngularAcceleration=10, maxAngularVelocity=2, simSpaceAngularAlpha=1}
   }
   n8 = Anim Dynamics - Bone: dyn_kuzi_b_r_01(ComponentPose=@n7, Alpha=1, ExternalForce="0,0,0") {
    angularDampingOverride=0.9999
    angularSpringConstant=1000
    bAngularSpring=true
    bGravityOverrideInSimSpace=true
    bOverrideAngularDamping=true
    bUseGravityOverride=true
    gravityScale=0.35
    simulationSpace=RootRelative
    alphaBoolBlend {blendOption=Linear}
    boundBone {boneName=dyn_kuzi_b_r_01}
    componentAppliedLinearAccClamp=<1000 1000 1000>
    componentLinearAccScale=<0.3 0.3 0.3>
    componentLinearVelScale=<0.3 0.3 0.3>
    gravityOverride {z=-500}
    physicsBodyDefinitions [{
      collisionType=CustomSphere
      sphereCollisionRadius=2
      boundBone {boneName=dyn_kuzi_b_r_01}
      boxExtents=<15 15 15>
      constraintSetup {angularLimitsMax=<5 5 2>, angularLimitsMin=<-5 -5 -2>, angularTarget=<0 0 0>}
      localJointOffset=<3 0 -0.2>
     }]
    relativeSpaceBone {boneName=calf_l}
    simSpaceSettings {maxAngularAcceleration=10, maxAngularVelocity=2, simSpaceAngularAlpha=1}
   }
   n9 = Anim Dynamics - Bone: dyn_kuzi_b_r_02(ComponentPose=@n8, Alpha=1, ExternalForce="0,0,0") {
    angularDampingOverride=0.9999
    angularSpringConstant=1000
    bAngularSpring=true
    bGravityOverrideInSimSpace=true
    bOverrideAngularDamping=true
    bUseGravityOverride=true
    gravityScale=0.35
    simulationSpace=RootRelative
    alphaBoolBlend {blendOption=Linear}
    boundBone {boneName=dyn_kuzi_b_r_02}
    componentAppliedLinearAccClamp=<1000 1000 1000>
    componentLinearAccScale=<0.3 0.3 0.3>
    componentLinearVelScale=<0.3 0.3 0.3>
    gravityOverride {z=-500}
    physicsBodyDefinitions [{
      collisionType=CustomSphere
      sphereCollisionRadius=2
      boundBone {boneName=dyn_kuzi_b_r_02}
      boxExtents=<15 15 15>
      constraintSetup {angularLimitsMax=<5 5 2>, angularLimitsMin=<-5 -5 -2>, angularTarget=<0 0 0>}
      localJointOffset=<3 0 -0.2>
     }]
    relativeSpaceBone {boneName=calf_l}
    simSpaceSettings {maxAngularAcceleration=10, maxAngularVelocity=2, simSpaceAngularAlpha=1}
   }
   n10 = Anim Dynamics - Bone: dyn_kuzi_b_r_03(ComponentPose=@n9, Alpha=1, ExternalForce="0,0,0") {
    angularDampingOverride=0.9999
    angularSpringConstant=1000
    bAngularSpring=true
    bGravityOverrideInSimSpace=true
    bOverrideAngularDamping=true
    bUseGravityOverride=true
    gravityScale=0.35
    simulationSpace=RootRelative
    alphaBoolBlend {blendOption=Linear}
    boundBone {boneName=dyn_kuzi_b_r_03}
    componentAppliedLinearAccClamp=<1000 1000 1000>
    componentLinearAccScale=<0.3 0.3 0.3>
    componentLinearVelScale=<0.3 0.3 0.3>
    gravityOverride {z=-500}
    physicsBodyDefinitions [{
      collisionType=CustomSphere
      sphereCollisionRadius=2
      boundBone {boneName=dyn_kuzi_b_r_03}
      boxExtents=<15 15 15>
      constraintSetup {angularLimitsMax=<5 5 2>, angularLimitsMin=<-5 -5 -2>, angularTarget=<0 0 0>}
      localJointOffset=<3 0 -0.2>
     }]
    relativeSpaceBone {boneName=calf_l}
    simSpaceSettings {maxAngularAcceleration=10, maxAngularVelocity=2, simSpaceAngularAlpha=1}
   }
   n11 = Anim Dynamics - Bone: dyn_kuzi_in_r_01(ComponentPose=@n10, Alpha=1, ExternalForce="0,0,0") {
    angularDampingOverride=0.9999
    angularSpringConstant=1000
    bAngularSpring=true
    bGravityOverrideInSimSpace=true
    bOverrideAngularDamping=true
    bUseGravityOverride=true
    gravityScale=0.35
    simulationSpace=RootRelative
    alphaBoolBlend {blendOption=Linear}
    boundBone {boneName=dyn_kuzi_in_r_01}
    componentAppliedLinearAccClamp=<1000 1000 1000>
    componentLinearAccScale=<0.3 0.3 0.3>
    componentLinearVelScale=<0.3 0.3 0.3>
    gravityOverride {z=-500}
    physicsBodyDefinitions [{
      collisionType=CustomSphere
      sphereCollisionRadius=2
      boundBone {boneName=dyn_kuzi_in_r_01}
      boxExtents=<15 15 15>
      constraintSetup {angularLimitsMax=<5 5 2>, angularLimitsMin=<-5 -5 -2>, angularTarget=<0 0 0>}
      localJointOffset=<3 0 -0.2>
     }]
    relativeSpaceBone {boneName=calf_l}
    simSpaceSettings {maxAngularAcceleration=10, maxAngularVelocity=2, simSpaceAngularAlpha=1}
   }
   n12 = Anim Dynamics - Bone: dyn_kuzi_in_r_02(ComponentPose=@n11, Alpha=1, ExternalForce="0,0,0") {
    angularDampingOverride=0.9999
    angularSpringConstant=1000
    bAngularSpring=true
    bGravityOverrideInSimSpace=true
    bOverrideAngularDamping=true
    bUseGravityOverride=true
    gravityScale=0.35
    simulationSpace=RootRelative
    alphaBoolBlend {blendOption=Linear}
    boundBone {boneName=dyn_kuzi_in_r_02}
    componentAppliedLinearAccClamp=<1000 1000 1000>
    componentLinearAccScale=<0.3 0.3 0.3>
    componentLinearVelScale=<0.3 0.3 0.3>
    gravityOverride {z=-500}
    physicsBodyDefinitions [{
      collisionType=CustomSphere
      sphereCollisionRadius=2
      boundBone {boneName=dyn_kuzi_in_r_02}
      boxExtents=<15 15 15>
      constraintSetup {angularLimitsMax=<5 5 2>, angularLimitsMin=<-5 -5 -2>, angularTarget=<0 0 0>}
      localJointOffset=<3 0 -0.2>
     }]
    relativeSpaceBone {boneName=calf_l}
    simSpaceSettings {maxAngularAcceleration=10, maxAngularVelocity=2, simSpaceAngularAlpha=1}
   }
   n13 = Anim Dynamics - Bone: dyn_kuzi_in_r_03(ComponentPose=@n12, Alpha=1, ExternalForce="0,0,0") {
    angularDampingOverride=0.9999
    angularSpringConstant=1000
    bAngularSpring=true
    bGravityOverrideInSimSpace=true
    bOverrideAngularDamping=true
    bUseGravityOverride=true
    gravityScale=0.35
    simulationSpace=RootRelative
    alphaBoolBlend {blendOption=Linear}
    boundBone {boneName=dyn_kuzi_in_r_03}
    componentAppliedLinearAccClamp=<1000 1000 1000>
    componentLinearAccScale=<0.3 0.3 0.3>
    componentLinearVelScale=<0.3 0.3 0.3>
    gravityOverride {z=-500}
    physicsBodyDefinitions [{
      collisionType=CustomSphere
      sphereCollisionRadius=2
      boundBone {boneName=dyn_kuzi_in_r_03}
      boxExtents=<15 15 15>
      constraintSetup {angularLimitsMax=<5 5 2>, angularLimitsMin=<-5 -5 -2>, angularTarget=<0 0 0>}
      localJointOffset=<3 0 -0.2>
     }]
    relativeSpaceBone {boneName=calf_l}
    simSpaceSettings {maxAngularAcceleration=10, maxAngularVelocity=2, simSpaceAngularAlpha=1}
   }
   n14 = Anim Dynamics - Bone: dyn_kuzi_f_l_01(ComponentPose=@n13, Alpha=1, ExternalForce="0,0,0") {
    angularDampingOverride=0.9999
    angularSpringConstant=1000
    bAngularSpring=true
    bGravityOverrideInSimSpace=true
    bOverrideAngularDamping=true
    bUseGravityOverride=true
    gravityScale=0.35
    simulationSpace=RootRelative
    alphaBoolBlend {blendOption=Linear}
    boundBone {boneName=dyn_kuzi_f_l_01}
    componentAppliedLinearAccClamp=<1000 1000 1000>
    componentLinearAccScale=<0.3 0.3 0.3>
    componentLinearVelScale=<0.3 0.3 0.3>
    gravityOverride {z=-500}
    physicsBodyDefinitions [{
      collisionType=CustomSphere
      sphereCollisionRadius=2
      boundBone {boneName=dyn_kuzi_f_l_01}
      boxExtents=<15 15 15>
      constraintSetup {angularLimitsMax=<5 5 2>, angularLimitsMin=<-5 -5 -2>, angularTarget=<0 0 0>}
      localJointOffset=<3 0 -0.2>
     }]
    relativeSpaceBone {boneName=calf_r}
    simSpaceSettings {maxAngularAcceleration=10, maxAngularVelocity=2, simSpaceAngularAlpha=1}
   }
   n15 = Anim Dynamics - Bone: dyn_kuzi_f_l_02(ComponentPose=@n14, Alpha=1, ExternalForce="0,0,0") {
    angularDampingOverride=0.9999
    angularSpringConstant=1000
    bAngularSpring=true
    bGravityOverrideInSimSpace=true
    bOverrideAngularDamping=true
    bUseGravityOverride=true
    gravityScale=0.35
    simulationSpace=RootRelative
    alphaBoolBlend {blendOption=Linear}
    boundBone {boneName=dyn_kuzi_f_l_02}
    componentAppliedLinearAccClamp=<1000 1000 1000>
    componentLinearAccScale=<0.3 0.3 0.3>
    componentLinearVelScale=<0.3 0.3 0.3>
    gravityOverride {z=-500}
    physicsBodyDefinitions [{
      collisionType=CustomSphere
      sphereCollisionRadius=2
      boundBone {boneName=dyn_kuzi_f_l_02}
      boxExtents=<15 15 15>
      constraintSetup {angularLimitsMax=<5 5 2>, angularLimitsMin=<-5 -5 -2>, angularTarget=<0 0 0>}
      localJointOffset=<3 0 -0.2>
     }]
    relativeSpaceBone {boneName=calf_r}
    simSpaceSettings {maxAngularAcceleration=10, maxAngularVelocity=2, simSpaceAngularAlpha=1}
   }
   n16 = Anim Dynamics - Bone: dyn_kuzi_f_l_03(ComponentPose=@n15, Alpha=1, ExternalForce="0,0,0") {
    angularDampingOverride=0.9999
    angularSpringConstant=1000
    bAngularSpring=true
    bGravityOverrideInSimSpace=true
    bOverrideAngularDamping=true
    bUseGravityOverride=true
    gravityScale=0.35
    simulationSpace=RootRelative
    alphaBoolBlend {blendOption=Linear}
    boundBone {boneName=dyn_kuzi_f_l_03}
    componentAppliedLinearAccClamp=<1000 1000 1000>
    componentLinearAccScale=<0.3 0.3 0.3>
    componentLinearVelScale=<0.3 0.3 0.3>
    gravityOverride {z=-500}
    physicsBodyDefinitions [{
      collisionType=CustomSphere
      sphereCollisionRadius=2
      boundBone {boneName=dyn_kuzi_f_l_03}
      boxExtents=<15 15 15>
      constraintSetup {angularLimitsMax=<5 5 2>, angularLimitsMin=<-5 -5 -2>, angularTarget=<0 0 0>}
      localJointOffset=<3 0 -0.2>
     }]
    relativeSpaceBone {boneName=calf_r}
    simSpaceSettings {maxAngularAcceleration=10, maxAngularVelocity=2, simSpaceAngularAlpha=1}
   }
   n17 = Anim Dynamics - Bone: dyn_kuzi_in_l_01(ComponentPose=@n16, Alpha=1, ExternalForce="0,0,0") {
    angularDampingOverride=0.9999
    angularSpringConstant=1000
    bAngularSpring=true
    bGravityOverrideInSimSpace=true
    bOverrideAngularDamping=true
    bUseGravityOverride=true
    gravityScale=0.35
    simulationSpace=RootRelative
    alphaBoolBlend {blendOption=Linear}
    boundBone {boneName=dyn_kuzi_in_l_01}
    componentAppliedLinearAccClamp=<1000 1000 1000>
    componentLinearAccScale=<0.3 0.3 0.3>
    componentLinearVelScale=<0.3 0.3 0.3>
    gravityOverride {z=-500}
    physicsBodyDefinitions [{
      collisionType=CustomSphere
      sphereCollisionRadius=2
      boundBone {boneName=dyn_kuzi_in_l_01}
      boxExtents=<15 15 15>
      constraintSetup {angularLimitsMax=<5 5 2>, angularLimitsMin=<-5 -5 -2>, angularTarget=<0 0 0>}
      localJointOffset=<3 0 -0.2>
     }]
    relativeSpaceBone {boneName=calf_r}
    simSpaceSettings {maxAngularAcceleration=10, maxAngularVelocity=2, simSpaceAngularAlpha=1}
   }
   n18 = Anim Dynamics - Bone: dyn_kuzi_in_l_02(ComponentPose=@n17, Alpha=1, ExternalForce="0,0,0") {
    angularDampingOverride=0.9999
    angularSpringConstant=1000
    bAngularSpring=true
    bGravityOverrideInSimSpace=true
    bOverrideAngularDamping=true
    bUseGravityOverride=true
    gravityScale=0.35
    simulationSpace=RootRelative
    alphaBoolBlend {blendOption=Linear}
    boundBone {boneName=dyn_kuzi_in_l_02}
    componentAppliedLinearAccClamp=<1000 1000 1000>
    componentLinearAccScale=<0.3 0.3 0.3>
    componentLinearVelScale=<0.3 0.3 0.3>
    gravityOverride {z=-500}
    physicsBodyDefinitions [{
      collisionType=CustomSphere
      sphereCollisionRadius=2
      boundBone {boneName=dyn_kuzi_in_l_02}
      boxExtents=<15 15 15>
      constraintSetup {angularLimitsMax=<5 5 2>, angularLimitsMin=<-5 -5 -2>, angularTarget=<0 0 0>}
      localJointOffset=<3 0 -0.2>
     }]
    relativeSpaceBone {boneName=calf_r}
    simSpaceSettings {maxAngularAcceleration=10, maxAngularVelocity=2, simSpaceAngularAlpha=1}
   }
   n19 = Anim Dynamics - Bone: dyn_kuzi_in_l_03(ComponentPose=@n18, Alpha=1, ExternalForce="0,0,0") {
    angularDampingOverride=0.9999
    angularSpringConstant=1000
    bAngularSpring=true
    bGravityOverrideInSimSpace=true
    bOverrideAngularDamping=true
    bUseGravityOverride=true
    gravityScale=0.35
    simulationSpace=RootRelative
    alphaBoolBlend {blendOption=Linear}
    boundBone {boneName=dyn_kuzi_in_l_03}
    componentAppliedLinearAccClamp=<1000 1000 1000>
    componentLinearAccScale=<0.3 0.3 0.3>
    componentLinearVelScale=<0.3 0.3 0.3>
    gravityOverride {z=-500}
    physicsBodyDefinitions [{
      collisionType=CustomSphere
      sphereCollisionRadius=2
      boundBone {boneName=dyn_kuzi_in_l_03}
      boxExtents=<15 15 15>
      constraintSetup {angularLimitsMax=<5 5 2>, angularLimitsMin=<-5 -5 -2>, angularTarget=<0 0 0>}
      localJointOffset=<3 0 -0.2>
     }]
    relativeSpaceBone {boneName=calf_r}
    simSpaceSettings {maxAngularAcceleration=10, maxAngularVelocity=2, simSpaceAngularAlpha=1}
   }
   n20 = Anim Dynamics - Bone: dyn_kuzi_b_l_01(ComponentPose=@n19, Alpha=1, ExternalForce="0,0,0") {
    angularDampingOverride=0.9999
    angularSpringConstant=1000
    bAngularSpring=true
    bGravityOverrideInSimSpace=true
    bOverrideAngularDamping=true
    bUseGravityOverride=true
    gravityScale=0.35
    simulationSpace=RootRelative
    alphaBoolBlend {blendOption=Linear}
    boundBone {boneName=dyn_kuzi_b_l_01}
    componentAppliedLinearAccClamp=<1000 1000 1000>
    componentLinearAccScale=<0.3 0.3 0.3>
    componentLinearVelScale=<0.3 0.3 0.3>
    gravityOverride {z=-500}
    physicsBodyDefinitions [{
      collisionType=CustomSphere
      sphereCollisionRadius=2
      boundBone {boneName=dyn_kuzi_b_l_01}
      boxExtents=<15 15 15>
      constraintSetup {angularLimitsMax=<5 5 2>, angularLimitsMin=<-5 -5 -2>, angularTarget=<0 0 0>}
      localJointOffset=<3 0 -0.2>
     }]
    relativeSpaceBone {boneName=calf_r}
    simSpaceSettings {maxAngularAcceleration=10, maxAngularVelocity=2, simSpaceAngularAlpha=1}
   }
   n21 = Anim Dynamics - Bone: dyn_kuzi_b_l_02(ComponentPose=@n20, Alpha=1, ExternalForce="0,0,0") {
    angularDampingOverride=0.9999
    angularSpringConstant=1000
    bAngularSpring=true
    bGravityOverrideInSimSpace=true
    bOverrideAngularDamping=true
    bUseGravityOverride=true
    gravityScale=0.35
    simulationSpace=RootRelative
    alphaBoolBlend {blendOption=Linear}
    boundBone {boneName=dyn_kuzi_b_l_02}
    componentAppliedLinearAccClamp=<1000 1000 1000>
    componentLinearAccScale=<0.3 0.3 0.3>
    componentLinearVelScale=<0.3 0.3 0.3>
    gravityOverride {z=-500}
    physicsBodyDefinitions [{
      collisionType=CustomSphere
      sphereCollisionRadius=2
      boundBone {boneName=dyn_kuzi_b_l_02}
      boxExtents=<15 15 15>
      constraintSetup {angularLimitsMax=<5 5 2>, angularLimitsMin=<-5 -5 -2>, angularTarget=<0 0 0>}
      localJointOffset=<3 0 -0.2>
     }]
    relativeSpaceBone {boneName=calf_r}
    simSpaceSettings {maxAngularAcceleration=10, maxAngularVelocity=2, simSpaceAngularAlpha=1}
   }
   n22 = Anim Dynamics - Bone: dyn_kuzi_b_l_03(ComponentPose=@n21, Alpha=1, ExternalForce="0,0,0") {
    angularDampingOverride=0.9999
    angularSpringConstant=1000
    bAngularSpring=true
    bGravityOverrideInSimSpace=true
    bOverrideAngularDamping=true
    bUseGravityOverride=true
    gravityScale=0.35
    simulationSpace=RootRelative
    alphaBoolBlend {blendOption=Linear}
    boundBone {boneName=dyn_kuzi_b_l_03}
    componentAppliedLinearAccClamp=<1000 1000 1000>
    componentLinearAccScale=<0.3 0.3 0.3>
    componentLinearVelScale=<0.3 0.3 0.3>
    gravityOverride {z=-500}
    physicsBodyDefinitions [{
      collisionType=CustomSphere
      sphereCollisionRadius=2
      boundBone {boneName=dyn_kuzi_b_l_03}
      boxExtents=<15 15 15>
      constraintSetup {angularLimitsMax=<5 5 2>, angularLimitsMin=<-5 -5 -2>, angularTarget=<0 0 0>}
      localJointOffset=<3 0 -0.2>
     }]
    relativeSpaceBone {boneName=calf_r}
    simSpaceSettings {maxAngularAcceleration=10, maxAngularVelocity=2, simSpaceAngularAlpha=1}
   }
   n23 = Anim Dynamics - Bone: dyn_kuzi_out_l_01(ComponentPose=@n22, Alpha=1, ExternalForce="0,0,0") {
    angularDampingOverride=0.9999
    angularSpringConstant=1000
    bAngularSpring=true
    bGravityOverrideInSimSpace=true
    bOverrideAngularDamping=true
    bUseGravityOverride=true
    gravityScale=0.35
    simulationSpace=RootRelative
    alphaBoolBlend {blendOption=Linear}
    boundBone {boneName=dyn_kuzi_out_l_01}
    componentAppliedLinearAccClamp=<1000 1000 1000>
    componentLinearAccScale=<0.3 0.3 0.3>
    componentLinearVelScale=<0.3 0.3 0.3>
    gravityOverride {z=-500}
    physicsBodyDefinitions [{
      collisionType=CustomSphere
      sphereCollisionRadius=2
      boundBone {boneName=dyn_kuzi_out_l_01}
      boxExtents=<15 15 15>
      constraintSetup {angularLimitsMax=<5 5 2>, angularLimitsMin=<-5 -5 -2>, angularTarget=<0 0 0>}
      localJointOffset=<3 0 -0.2>
     }]
    relativeSpaceBone {boneName=calf_r}
    simSpaceSettings {maxAngularAcceleration=10, maxAngularVelocity=2, simSpaceAngularAlpha=1}
   }
   n24 = Anim Dynamics - Bone: dyn_kuzi_out_l_02(ComponentPose=@n23, Alpha=1, ExternalForce="0,0,0") {
    angularDampingOverride=0.9999
    angularSpringConstant=1000
    bAngularSpring=true
    bGravityOverrideInSimSpace=true
    bOverrideAngularDamping=true
    bUseGravityOverride=true
    gravityScale=0.35
    simulationSpace=RootRelative
    alphaBoolBlend {blendOption=Linear}
    boundBone {boneName=dyn_kuzi_out_l_02}
    componentAppliedLinearAccClamp=<1000 1000 1000>
    componentLinearAccScale=<0.3 0.3 0.3>
    componentLinearVelScale=<0.3 0.3 0.3>
    gravityOverride {z=-500}
    physicsBodyDefinitions [{
      collisionType=CustomSphere
      sphereCollisionRadius=2
      boundBone {boneName=dyn_kuzi_out_l_02}
      boxExtents=<15 15 15>
      constraintSetup {angularLimitsMax=<5 5 2>, angularLimitsMin=<-5 -5 -2>, angularTarget=<0 0 0>}
      localJointOffset=<3 0 -0.2>
     }]
    relativeSpaceBone {boneName=calf_r}
    simSpaceSettings {maxAngularAcceleration=10, maxAngularVelocity=2, simSpaceAngularAlpha=1}
   }
   n25 = Anim Dynamics - Bone: dyn_kuzi_out_l_03(ComponentPose=@n24, Alpha=1, ExternalForce="0,0,0") {
    angularDampingOverride=0.9999
    angularSpringConstant=1000
    bAngularSpring=true
    bGravityOverrideInSimSpace=true
    bOverrideAngularDamping=true
    bUseGravityOverride=true
    gravityScale=0.35
    simulationSpace=RootRelative
    alphaBoolBlend {blendOption=Linear}
    boundBone {boneName=dyn_kuzi_out_l_03}
    componentAppliedLinearAccClamp=<1000 1000 1000>
    componentLinearAccScale=<0.3 0.3 0.3>
    componentLinearVelScale=<0.3 0.3 0.3>
    gravityOverride {z=-500}
    physicsBodyDefinitions [{
      collisionType=CustomSphere
      sphereCollisionRadius=2
      boundBone {boneName=dyn_kuzi_out_l_03}
      boxExtents=<15 15 15>
      constraintSetup {angularLimitsMax=<5 5 2>, angularLimitsMin=<-5 -5 -2>, angularTarget=<0 0 0>}
      localJointOffset=<3 0 -0.2>
     }]
    relativeSpaceBone {boneName=calf_r}
    simSpaceSettings {maxAngularAcceleration=10, maxAngularVelocity=2, simSpaceAngularAlpha=1}
   }
   n26 = Anim Dynamics - Bone: dyn_cloth_f_01(ComponentPose=@n25, Alpha=1, ExternalForce="0,0,0") {
    angularDampingOverride=0.9999
    angularSpringConstant=500
    bAngularSpring=true
    bGravityOverrideInSimSpace=true
    bOverrideAngularDamping=true
    bUseGravityOverride=true
    gravityScale=0.35
    simulationSpace=RootRelative
    alphaBoolBlend {blendOption=Linear}
    boundBone {boneName=dyn_cloth_f_01}
    componentAppliedLinearAccClamp=<2000 2000 2000>
    componentLinearAccScale=<0.3 0.3 0.3>
    componentLinearVelScale=<0.3 0.3 0.3>
    gravityOverride {z=-980}
    physicsBodyDefinitions [{
      collisionType=CustomSphere
      sphereCollisionRadius=2
      boundBone {boneName=dyn_cloth_f_01}
      boxExtents=<10 15 15>
      constraintSetup {angularLimitsMax=<5 20 30>, angularLimitsMin=<-5 -50 -30>, angularTarget=<0 0 0>}
      localJointOffset=<4 0 0>
     }]
    relativeSpaceBone {boneName=spine_02}
    simSpaceSettings {maxAngularAcceleration=10, maxAngularVelocity=2, simSpaceAngularAlpha=1}
   }
   n27 = Anim Dynamics - Bone: dyn_cloth_r_01(ComponentPose=@n26, Alpha=1, ExternalForce="0,0,0") {
    angularDampingOverride=0.9999
    angularSpringConstant=500
    bAngularSpring=true
    bGravityOverrideInSimSpace=true
    bOverrideAngularDamping=true
    bUseGravityOverride=true
    gravityScale=0.35
    simulationSpace=RootRelative
    alphaBoolBlend {blendOption=Linear}
    boundBone {boneName=dyn_cloth_r_01}
    componentAppliedLinearAccClamp=<2000 2000 2000>
    componentLinearAccScale=<0.3 0.3 0.3>
    componentLinearVelScale=<0.3 0.3 0.3>
    gravityOverride {z=-980}
    physicsBodyDefinitions [{
      collisionType=CustomSphere
      sphereCollisionRadius=2
      boundBone {boneName=dyn_cloth_r_01}
      boxExtents=<10 15 15>
      constraintSetup {angularLimitsMax=<5 20 30>, angularLimitsMin=<-5 -50 -30>, angularTarget=<0 0 0>}
      localJointOffset=<4 0 0>
     }]
    relativeSpaceBone {boneName=spine_02}
    simSpaceSettings {maxAngularAcceleration=10, maxAngularVelocity=2, simSpaceAngularAlpha=1}
   }
   n28 = Anim Dynamics - Bone: dyn_cloth_f_r_01(ComponentPose=@n27, Alpha=1, ExternalForce="0,0,0") {
    angularDampingOverride=0.9999
    angularSpringConstant=500
    bAngularSpring=true
    bGravityOverrideInSimSpace=true
    bOverrideAngularDamping=true
    bUseGravityOverride=true
    gravityScale=0.35
    simulationSpace=RootRelative
    alphaBoolBlend {blendOption=Linear}
    boundBone {boneName=dyn_cloth_f_r_01}
    componentAppliedLinearAccClamp=<2000 2000 2000>
    componentLinearAccScale=<0.3 0.3 0.3>
    componentLinearVelScale=<0.3 0.3 0.3>
    gravityOverride {z=-980}
    physicsBodyDefinitions [{
      collisionType=CustomSphere
      sphereCollisionRadius=2
      boundBone {boneName=dyn_cloth_f_r_01}
      boxExtents=<10 15 15>
      constraintSetup {angularLimitsMax=<5 20 30>, angularLimitsMin=<-5 -50 -30>, angularTarget=<0 0 0>}
      localJointOffset=<4 0 0>
     }]
    relativeSpaceBone {boneName=spine_02}
    simSpaceSettings {maxAngularAcceleration=10, maxAngularVelocity=2, simSpaceAngularAlpha=1}
   }
   n29 = Anim Dynamics - Bone: dyn_cloth_b_01(ComponentPose=@n28, Alpha=1, ExternalForce="0,0,0") {
    angularDampingOverride=0.9999
    angularSpringConstant=500
    bAngularSpring=true
    bGravityOverrideInSimSpace=true
    bOverrideAngularDamping=true
    bUseGravityOverride=true
    gravityScale=0.35
    simulationSpace=RootRelative
    alphaBoolBlend {blendOption=Linear}
    boundBone {boneName=dyn_cloth_b_01}
    componentAppliedLinearAccClamp=<2000 2000 2000>
    componentLinearAccScale=<0.3 0.3 0.3>
    componentLinearVelScale=<0.3 0.3 0.3>
    gravityOverride {z=-980}
    physicsBodyDefinitions [{
      collisionType=CustomSphere
      sphereCollisionRadius=2
      boundBone {boneName=dyn_cloth_b_01}
      boxExtents=<10 15 15>
      constraintSetup {angularLimitsMax=<5 20 30>, angularLimitsMin=<-5 -50 -30>, angularTarget=<0 0 0>}
      localJointOffset=<4 0 0>
     }]
    relativeSpaceBone {boneName=spine_02}
    simSpaceSettings {maxAngularAcceleration=10, maxAngularVelocity=2, simSpaceAngularAlpha=1}
   }
   n30 = Anim Dynamics - Bone: dyn_cloth_b_r_01(ComponentPose=@n29, Alpha=1, ExternalForce="0,0,0") {
    angularDampingOverride=0.9999
    angularSpringConstant=500
    bAngularSpring=true
    bGravityOverrideInSimSpace=true
    bOverrideAngularDamping=true
    bUseGravityOverride=true
    gravityScale=0.35
    simulationSpace=RootRelative
    alphaBoolBlend {blendOption=Linear}
    boundBone {boneName=dyn_cloth_b_r_01}
    componentAppliedLinearAccClamp=<2000 2000 2000>
    componentLinearAccScale=<0.3 0.3 0.3>
    componentLinearVelScale=<0.3 0.3 0.3>
    gravityOverride {z=-980}
    physicsBodyDefinitions [{
      collisionType=CustomSphere
      sphereCollisionRadius=2
      boundBone {boneName=dyn_cloth_b_r_01}
      boxExtents=<10 15 15>
      constraintSetup {angularLimitsMax=<5 20 30>, angularLimitsMin=<-5 -50 -30>, angularTarget=<0 0 0>}
      localJointOffset=<4 0 0>
     }]
    relativeSpaceBone {boneName=spine_02}
    simSpaceSettings {maxAngularAcceleration=10, maxAngularVelocity=2, simSpaceAngularAlpha=1}
   }
   n31 = Anim Dynamics - Bone: dyn_cloth_f_l_01(ComponentPose=@n30, Alpha=1, ExternalForce="0,0,0") {
    angularDampingOverride=0.9999
    angularSpringConstant=500
    bAngularSpring=true
    bGravityOverrideInSimSpace=true
    bOverrideAngularDamping=true
    bUseGravityOverride=true
    gravityScale=0.35
    simulationSpace=RootRelative
    alphaBoolBlend {blendOption=Linear}
    boundBone {boneName=dyn_cloth_f_l_01}
    componentAppliedLinearAccClamp=<2000 2000 2000>
    componentLinearAccScale=<0.3 0.3 0.3>
    componentLinearVelScale=<0.3 0.3 0.3>
    gravityOverride {z=-980}
    physicsBodyDefinitions [{
      collisionType=CustomSphere
      sphereCollisionRadius=2
      boundBone {boneName=dyn_cloth_f_l_01}
      boxExtents=<10 15 15>
      constraintSetup {angularLimitsMax=<5 20 30>, angularLimitsMin=<-5 -50 -30>, angularTarget=<0 0 0>}
      localJointOffset=<4 0 0>
     }]
    relativeSpaceBone {boneName=spine_02}
    simSpaceSettings {maxAngularAcceleration=10, maxAngularVelocity=2, simSpaceAngularAlpha=1}
   }
   n32 = Anim Dynamics - Bone: dyn_cloth_b_l_01(ComponentPose=@n31, Alpha=1, ExternalForce="0,0,0") {
    angularDampingOverride=0.9999
    angularSpringConstant=500
    bAngularSpring=true
    bGravityOverrideInSimSpace=true
    bOverrideAngularDamping=true
    bUseGravityOverride=true
    gravityScale=0.35
    simulationSpace=RootRelative
    alphaBoolBlend {blendOption=Linear}
    boundBone {boneName=dyn_cloth_b_l_01}
    componentAppliedLinearAccClamp=<2000 2000 2000>
    componentLinearAccScale=<0.3 0.3 0.3>
    componentLinearVelScale=<0.3 0.3 0.3>
    gravityOverride {z=-980}
    physicsBodyDefinitions [{
      collisionType=CustomSphere
      sphereCollisionRadius=2
      boundBone {boneName=dyn_cloth_b_l_01}
      boxExtents=<10 15 15>
      constraintSetup {angularLimitsMax=<5 20 30>, angularLimitsMin=<-5 -50 -30>, angularTarget=<0 0 0>}
      localJointOffset=<4 0 0>
     }]
    relativeSpaceBone {boneName=spine_02}
    simSpaceSettings {maxAngularAcceleration=10, maxAngularVelocity=2, simSpaceAngularAlpha=1}
   }
   n33 = Anim Dynamics - Bone: dyn_cloth_l_01(ComponentPose=@n32, Alpha=1, ExternalForce="0,0,0") {
    angularDampingOverride=0.9999
    angularSpringConstant=500
    bAngularSpring=true
    bGravityOverrideInSimSpace=true
    bOverrideAngularDamping=true
    bUseGravityOverride=true
    gravityScale=0.35
    simulationSpace=RootRelative
    alphaBoolBlend {blendOption=Linear}
    boundBone {boneName=dyn_cloth_l_01}
    componentAppliedLinearAccClamp=<2000 2000 2000>
    componentLinearAccScale=<0.3 0.3 0.3>
    componentLinearVelScale=<0.3 0.3 0.3>
    gravityOverride {z=-980}
    physicsBodyDefinitions [{
      collisionType=CustomSphere
      sphereCollisionRadius=2
      boundBone {boneName=dyn_cloth_l_01}
      boxExtents=<10 15 15>
      constraintSetup {angularLimitsMax=<5 20 30>, angularLimitsMin=<-5 -50 -30>, angularTarget=<0 0 0>}
      localJointOffset=<4 0 0>
     }]
    relativeSpaceBone {boneName=spine_02}
    simSpaceSettings {maxAngularAcceleration=10, maxAngularVelocity=2, simSpaceAngularAlpha=1}
   }
   n34 = Anim Dynamics - Bone: dyn_calf_ribbon_r_01(ComponentPose=@n33, Alpha=1, ExternalForce="0,0,0") {
    angularDampingOverride=0.9999
    angularSpringConstant=500
    bAngularSpring=true
    bGravityOverrideInSimSpace=true
    bOverrideAngularDamping=true
    bUseGravityOverride=true
    gravityScale=0.35
    simulationSpace=RootRelative
    alphaBoolBlend {blendOption=Linear}
    boundBone {boneName=dyn_calf_ribbon_r_01}
    componentAppliedLinearAccClamp=<2000 2000 2000>
    componentLinearAccScale=<0.3 0.3 0.3>
    componentLinearVelScale=<0.3 0.3 0.3>
    gravityOverride {z=-980}
    physicsBodyDefinitions [{
      collisionType=CustomSphere
      sphereCollisionRadius=2
      boundBone {boneName=dyn_calf_ribbon_r_01}
      boxExtents=<10 15 15>
      constraintSetup {angularLimitsMax=<5 20 10>, angularLimitsMin=<-5 -2 -10>, angularTarget=<0 0 0>}
      localJointOffset=<3 0 0>
     }]
    relativeSpaceBone {boneName=spine_02}
    simSpaceSettings {maxAngularAcceleration=10, maxAngularVelocity=2, simSpaceAngularAlpha=1}
   }
   n35 = Anim Dynamics - Bone: dyn_calf_ribbon_r_02(ComponentPose=@n34, Alpha=1, ExternalForce="0,0,0") {
    angularDampingOverride=0.9999
    angularSpringConstant=500
    bAngularSpring=true
    bGravityOverrideInSimSpace=true
    bOverrideAngularDamping=true
    bUseGravityOverride=true
    gravityScale=0.35
    simulationSpace=RootRelative
    alphaBoolBlend {blendOption=Linear}
    boundBone {boneName=dyn_calf_ribbon_r_02}
    componentAppliedLinearAccClamp=<2000 2000 2000>
    componentLinearAccScale=<0.3 0.3 0.3>
    componentLinearVelScale=<0.3 0.3 0.3>
    gravityOverride {z=-980}
    physicsBodyDefinitions [{
      collisionType=CustomSphere
      sphereCollisionRadius=2
      boundBone {boneName=dyn_calf_ribbon_r_02}
      boxExtents=<10 15 15>
      constraintSetup {angularLimitsMax=<5 20 10>, angularLimitsMin=<-5 -2 -10>, angularTarget=<0 0 0>}
      localJointOffset=<3 0 0>
     }]
    relativeSpaceBone {boneName=spine_02}
    simSpaceSettings {maxAngularAcceleration=10, maxAngularVelocity=2, simSpaceAngularAlpha=1}
   }
   n36 = Anim Dynamics - Bone: dyn_calf_libbon_l_01(ComponentPose=@n35, Alpha=1, ExternalForce="0,0,0") {
    angularDampingOverride=0.9999
    angularSpringConstant=500
    bAngularSpring=true
    bGravityOverrideInSimSpace=true
    bOverrideAngularDamping=true
    bUseGravityOverride=true
    gravityScale=0.35
    simulationSpace=RootRelative
    alphaBoolBlend {blendOption=Linear}
    boundBone {boneName=dyn_calf_libbon_l_01}
    componentAppliedLinearAccClamp=<2000 2000 2000>
    componentLinearAccScale=<0.3 0.3 0.3>
    componentLinearVelScale=<0.3 0.3 0.3>
    gravityOverride {z=-980}
    physicsBodyDefinitions [{
      collisionType=CustomSphere
      sphereCollisionRadius=2
      boundBone {boneName=dyn_calf_libbon_l_01}
      boxExtents=<10 15 15>
      constraintSetup {angularLimitsMax=<5 20 10>, angularLimitsMin=<-5 -2 -10>, angularTarget=<0 0 0>}
      localJointOffset=<3 0 0>
     }]
    relativeSpaceBone {boneName=spine_02}
    simSpaceSettings {maxAngularAcceleration=10, maxAngularVelocity=2, simSpaceAngularAlpha=1}
   }
   n37 = Anim Dynamics - Bone: dyn_calf_libbon_l_02(ComponentPose=@n36, Alpha=1, ExternalForce="0,0,0") {
    angularDampingOverride=0.9999
    angularSpringConstant=500
    bAngularSpring=true
    bGravityOverrideInSimSpace=true
    bOverrideAngularDamping=true
    bUseGravityOverride=true
    gravityScale=0.35
    simulationSpace=RootRelative
    alphaBoolBlend {blendOption=Linear}
    boundBone {boneName=dyn_calf_libbon_l_02}
    componentAppliedLinearAccClamp=<2000 2000 2000>
    componentLinearAccScale=<0.3 0.3 0.3>
    componentLinearVelScale=<0.3 0.3 0.3>
    gravityOverride {z=-980}
    physicsBodyDefinitions [{
      collisionType=CustomSphere
      sphereCollisionRadius=2
      boundBone {boneName=dyn_calf_libbon_l_02}
      boxExtents=<10 15 15>
      constraintSetup {angularLimitsMax=<5 20 10>, angularLimitsMin=<-5 -2 -10>, angularTarget=<0 0 0>}
      localJointOffset=<3 0 0>
     }]
    relativeSpaceBone {boneName=spine_02}
    simSpaceSettings {maxAngularAcceleration=10, maxAngularVelocity=2, simSpaceAngularAlpha=1}
   }
   n38 = Anim Dynamics - Bone: follow_bag_01(ComponentPose=@n37, Alpha=1, ExternalForce="0,0,0") {
    angularDampingOverride=0.9999
    angularSpringConstant=500
    bAngularSpring=true
    bGravityOverrideInSimSpace=true
    bOverrideAngularDamping=true
    bOverrideLinearDamping=true
    bUseGravityOverride=true
    gravityScale=0.35
    linearDampingOverride=0.9999
    linearSpringConstant=500
    simulationSpace=RootRelative
    alphaBoolBlend {blendOption=Linear}
    boundBone {boneName=follow_bag_01}
    componentAppliedLinearAccClamp=<2000 2000 2000>
    componentLinearAccScale=<0.3 0.3 0.3>
    componentLinearVelScale=<0.3 0.3 0.3>
    gravityOverride {z=-980}
    physicsBodyDefinitions [{
      collisionType=CustomSphere
      sphereCollisionRadius=2
      boundBone {boneName=follow_bag_01}
      boxExtents=<15 15 15>
      constraintSetup {angularLimitsMax=<5 5 10>, angularLimitsMin=<-5 -2 -10>, angularTarget=<0 0 0>}
      localJointOffset=<-2 0 0>
     }]
    relativeSpaceBone {boneName=spine_02}
    simSpaceSettings {maxAngularAcceleration=10, maxAngularVelocity=2, simSpaceAngularAlpha=1}
   }
   n39 = RigidBody!(
    ComponentPose=Local To Component(LocalPose=Component To Local(ComponentPose=@n38))
    Alpha
    OverrideWorldGravity:Vector="0,0,0"
    ExternalForce
   ) -> (Pose) {
    bClampLinearTranslationLimitToRefPose=true
    bOverrideWorldGravity=true
    overridePhysicsAsset=&/F_MED_LocalTalent_Flare_Physics
    simulationSpace=BaseBoneSpace
    baseBoneRef {boneName=spine_04}
    overrideWorldGravity {z=-980}
    simSpaceSettings {
     dampingAlpha=0.5
     maxAngularAcceleration=100
     maxAngularVelocity=5
     maxLinearAcceleration=500
     maxLinearVelocity=100
     velocityScaleZ=0
     worldAlpha=1
     externalLinearDragV {y=10}
    }
   }
   Output Pose(Result=Component To Local(ComponentPose=@n39)) {layerGroup=None}
  ]}
 EventGraph {nodes [
   n1 = Event BlueprintUpdateAnimation()
   > n2 = Sequence()
   Is Valid(exec=@n2.then_1, InputObject=Get PlayerPawn() -> (PlayerPawn:Fort Player Pawn Athena Object Reference))
   > Set WPO Globals Function(
    Material=Get Mat Instance Dynamic()
    Overlay Material=Get Overlay Mat Instance Dynamic()
    Skeletal Mesh=Get SkelMesh()
    Local Noise Offset Multiplier=0.5
    "Local Noise Transform  Location"=Get Local Noise Transform Location() -> (Local Noise Transform Location:Vector)
    Local Noise Offset=Get Local Noise Offset()
   )
   > Set WPO Regional Function!(
    execute
    Material:Material Instance Dynamic Object Reference=Get Mat Instance Dynamic()
    Overlay Material:Material Instance Dynamic Object Reference=Get Overlay Mat Instance Dynamic()
    Skeletal Mesh:Skeletal Mesh Component Object Reference=Get SkelMesh()
    Velocity Transform:Name=foot_r
    Velocity Interp Speed:double=Get Velocity Interp Speed()
    Collision Transform:Name=calf_r
    Collision Transform Axis:EAxis Enum
    Invert Collision Transform Axis:bool=true
    Delta Time X:double=@n1.DeltaTimeX
    Previous Velocity Transform Location:"Vector (by ref)"=Get R1 Velocity Transform Location()
    Previous Velocity Transform Velocity:"Vector (by ref)"=Get R1 Velocity Transform Velocity()
    Region:int=1
   ) -> (then)
   > Set WPO Regional Function(
    Material=Get Mat Instance Dynamic()
    Overlay Material=Get Overlay Mat Instance Dynamic()
    Skeletal Mesh=Get SkelMesh()
    Velocity Transform=foot_l
    Velocity Interp Speed=Get Velocity Interp Speed()
    Collision Transform=calf_l
    Invert Collision Transform Axis=false
    Delta Time X=@n1.DeltaTimeX
    Previous Velocity Transform Location=Get R2 Velocity Transform Location()
    Previous Velocity Transform Velocity=Get R2 Velocity Transform Velocity()
    Region=2
   )
   > n3 = Branch(Condition=Get bIsSkyDiving())
   > n4 = Set WPO Pho Wind Intensity(WPO Pho Wind Intensity=1)
   SetScalarParameterValue(
    execute=[@n5.then, @n4.then]
    self=[Get Overlay Mat Instance Dynamic(), Get Mat Instance Dynamic()]
    ParameterName=ClothSim_WindStrength
    Value=Get WPO Pho Wind Intensity()
   )
   > Set Crouching(Crouching:double="To Float (Boolean)"(InBool=Get bIsCrouching())) -> (Output_Get:double)
   n5 = Set WPO Pho Wind Intensity(execute=@n3.else, WPO Pho Wind Intensity=0)
   n6 = "Gravity Override State Function (Parameter Asset) (Nativized)"(
    execute=@n2.then_0
    CharacterPartAnimInstance=Self-Reference()
    GameTimeInSeconds=GetGameTimeInSeconds()
    Parameters=&/F_MED_LocalTalent_Parameters
   )
   > Set GravityOverride(GravityOverride=@n6)
   Event BlueprintInitializeAnimation()
   > Set SkelMesh(SkelMesh=GetOwnerSkeletalMeshComponent())
   n7 = TryGetPawnOwner()
   > Is Valid(InputObject=@n7)
   > n8 = Cast To FortPlayerPawnAthena(Object=@n7)
   > Set PlayerPawn(PlayerPawn:Fort Player Pawn Athena Object Reference=@n8) -> (Output_Get:Fort Player Pawn Athena Object Reference)
   Event OnCharacterCustomizationLoaded()
   > n9 = Set SkelMesh(SkelMesh=GetPartSkeletalMeshComponent())
   > n10 = Get Dynamic Material Instance(SkeletalMeshComponent=@n9)
   > Set Mat Instance Dynamic(Mat Instance Dynamic=@n10)
   > n11 = GetOrCreateDynamicOverlayMaterial(MeshComponent=@n9, MaterialName=None)
   > Set Overlay Mat Instance Dynamic(Overlay Mat Instance Dynamic=@n11)
  ]}
]
variables [
 Crouching {type=number}
 GravityOverride {type=Vector}
 Local Noise Offset {type=Vector}
 Local Noise Transform Location {type=Vector}
 Mat Instance Dynamic {type=MaterialInstanceDynamic}
 Overlay Mat Instance Dynamic {type=MaterialInstanceDynamic}
 PlayerPawn {type=FortPlayerPawnAthena}
 R1 Velocity Transform Location {type=Vector}
 R1 Velocity Transform Velocity {type=Vector}
 R2 Velocity Transform Location {type=Vector}
 R2 Velocity Transform Velocity {type=Vector}
 SkelMesh {type=SkeletalMeshComponent}
 Velocity Interp Speed {type=number, default=30}
 WPO Pho Wind Intensity {type=number}
]
