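# Control Rig for the LocalTalent Flare face-accessory mesh that adjusts the three back hair strands.
# Parent constraints tie their base bones to spine_05 with filtered translation, then a spherical pose reader on head relative to neck_02 weights positional offsets.
# The backpack bool squashes the strands' Z scale, and the targeting bool slightly rotates the right one.
# A debug-only function, disabled here, copies neck and head controls onto bones by stripping "_ctrl".
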ControlRigBlueprint /BRCosmetics/Characters/Player/Female/Medium/Bodies/F_MED_LocalTalent_Flare/Meshes/Parts/F_MED_LocalTalent_Flare_FaceAcc_CtrlRig
graphs [
 New Function_ContainedGraph {nodes [
   n1 = RigVMFunctionEntryNode()
   RigVMFunction_ControlFlowBranch = &RigVMFunction_ControlFlowBranch(Condition=@n1.Debug, ExecuteContext=@n1.ExecuteContext)
   RigVMFunctionReturnNode(ExecuteContext=@RigVMFunction_ControlFlowBranch.Completed)
   DISPATCH_RigVMDispatch_ArrayIterator = RigVMDispatchNode(
    Array=/Script/ControlRig.RigUnit_CollectionGetAll(TypeToSearch=Control) {is=struct RigVMUnitNode}.Items
    ExecuteContext=@RigVMFunction_ControlFlowBranch.True
   ) {template="DISPATCH_RigVMDispatch_ArrayIterator(in Array,out Element,out Index,out Count,out Ratio)"}
   Replace = /Script/RigVM.RigVMFunction_NameReplace(Old=_ctrl, Name=@DISPATCH_RigVMDispatch_ArrayIterator.Element.Name) {is=struct RigVMUnitNode, template="Replace::Execute(in Name,in Old,in New,out Result)"}
   > &RigUnit_SetTransform(
    Item=(Type=Bone,Name="spine_01")
    Weight=1
    bInitial=False
    Item.Name=@Replace.Result
    Value=&RigUnit_GetTransform(Item=(Type=Control,Name="spine_01_ctrl"), bInitial=False, Item.Name=@DISPATCH_RigVMDispatch_ArrayIterator.Element.Name).Transform
   )
  ]}
 RigVMFunctionLibrary {nodes [ctrlRig = RigVMCollapseNode(Debug=False, ExecuteContext=())]}
 RigVMModel {nodes [
   RigUnit_BeginExecution = &RigUnit_BeginExecution()
   ctrlRig = RigVMFunctionReferenceNode(Debug=false, ExecuteContext=@RigUnit_BeginExecution.ExecutePin)
   > &RigUnit_ParentConstraint(
    AdvancedSettings=(RotationOrderForFilter=XZY)
    Child=(Type=Bone,Name="base_hair_back_M_01")
    Filter=(TranslationFilter=(bX=True,bY=False,bZ=False),RotationFilter=(bX=True,bY=True,bZ=True),ScaleFilter=(bX=True,bY=True,bZ=True))
    Parents=((Item=(Type=Bone,Name="spine_05"),Weight=1),(Item=(Type=Bone,Name="head"),Weight=0))
    Weight=1
   )
   > &RigUnit_ParentConstraint(
    AdvancedSettings=(RotationOrderForFilter=XZY)
    Child=(Type=Bone,Name="base_hair_back_L_01")
    Filter=(TranslationFilter=(bX=True,bY=False,bZ=False),RotationFilter=(bX=True,bY=True,bZ=True),ScaleFilter=(bX=True,bY=True,bZ=True))
    Parents=((Item=(Type=Bone,Name="spine_05"),Weight=1),(Item=(Type=Bone,Name="head"),Weight=0))
    Weight=1
   )
   > &RigUnit_ParentConstraint(
    AdvancedSettings=(RotationOrderForFilter=XZY)
    Child=(Type=Bone,Name="base_hair_back_R_01")
    Filter=(TranslationFilter=(bX=True,bY=False,bZ=False),RotationFilter=(bX=True,bY=True,bZ=True),ScaleFilter=(bX=True,bY=True,bZ=True))
    Parents=((Item=(Type=Bone,Name="spine_05"),Weight=1),(Item=(Type=Bone,Name="head"),Weight=0))
    Weight=1
   )
   > n1 = &RigUnit_SphericalPoseReader(
    ActiveRegionScaleFactors=(PositiveWidth=2,NegativeWidth=3,PositiveHeight=3,NegativeHeight=5)
    ActiveRegionSize=0.1
    Debug=(bDrawDebug=false,bDraw2D=false,bDrawLocalAxes=false,DebugScale=15,DebugSegments=20,DebugThickness=0.25)
    DriverAxis=(X=1,Y=0,Z=0)
    DriverItem=(Type=Bone,Name="head")
    FalloffRegionScaleFactors=(PositiveWidth=0.8,NegativeWidth=1,PositiveHeight=1,NegativeHeight=1)
    FalloffSize=0.5
    FlipHeightScaling=false
    FlipWidthScaling=false
    OptionalParentItem=(Type=Bone,Name="neck_02")
    RotationOffset=(X=-90,Y=0,Z=10)
   )
   > &RigUnit_ModifyTransforms(
    ItemToModify=((Item=(Type=Bone,Name="base_hair_back_R_01"),Transform=(Rotation=(X=0,Y=0,Z=0,W=1),Translation=(X=1,Y=0,Z=3),Scale3D=(X=1,Y=1,Z=1))))
    WeightMinimum=0
    Weight=@n1.OutputParam
   )
   > &RigUnit_ModifyTransforms(
    ItemToModify=((Item=(Type=Bone,Name="base_hair_back_M_01"),Transform=(Rotation=(X=0,Y=0,Z=0,W=1),Translation=(X=1,Y=0,Z=2),Scale3D=(X=1,Y=1,Z=1))))
    WeightMinimum=0
    Weight=@n1.OutputParam
   )
   > &RigUnit_ModifyTransforms(
    ItemToModify=((Item=(Type=Bone,Name="base_hair_back_L_01"),Transform=(Rotation=(X=0,Y=0,Z=0,W=1),Translation=(X=1,Y=0,Z=2),Scale3D=(X=1,Y=1,Z=1))))
    WeightMinimum=0
    Weight=@n1.OutputParam
   )
   n2 = &RigVMFunction_MathBoolToFloat(Value=RigVMVariableNode() {variable=is backpack}.Value)
   > &RigUnit_ModifyTransforms(
    ItemToModify=((Item=(Type=Bone,Name="base_hair_back_M_01"),Transform=(Rotation=(X=0,Y=0,Z=0,W=1),Translation=(X=0,Y=0,Z=0),Scale3D=(X=1,Y=1,Z=0.3))))
    WeightMinimum=0
    Weight=@n2.Result
   )
   > &RigUnit_ModifyTransforms(
    ItemToModify=((Item=(Type=Bone,Name="base_hair_back_R_01"),Transform=(Rotation=(X=0,Y=0,Z=0,W=1),Translation=(X=0,Y=0,Z=0),Scale3D=(X=1,Y=1,Z=0.3))))
    WeightMinimum=0
    Weight=@n2.Result
   )
   > &RigUnit_ModifyTransforms(
    ItemToModify=((Item=(Type=Bone,Name="base_hair_back_L_01"),Transform=(Rotation=(X=0,Y=0,Z=0,W=1),Translation=(X=0,Y=0,Z=0),Scale3D=(X=1,Y=1,Z=0.3))))
    WeightMinimum=0
    Weight=@n2.Result
   )
   RigVMFunction_MathBoolToFloat = &RigVMFunction_MathBoolToFloat(Value=RigVMVariableNode() {variable=is Targeting}.Value)
   > &RigUnit_ModifyTransforms(
    ItemToModify=((Item=(Type=Bone,Name="base_hair_back_R_01"),Transform=(Rotation=(X=0,Y=-0.0175,Z=0,W=0.9998),Translation=(X=0,Y=0,Z=0),Scale3D=(X=1,Y=1,Z=1))))
    WeightMinimum=0
    Weight=@RigVMFunction_MathBoolToFloat.Result
   )
  ]}
]
hierarchy [
 root {loc=<0 0 0>, rot=<0 0 0 1>}
 pelvis {parent=root, loc=<0.0006 -0.5342 98.0529>, rot=<0.0127 0.707 -0.0127 -0.707>}
 spine_01 {parent=pelvis, loc=<6.1208 -0.9328 0>, rot=<0 0 0.034 -0.9994>}
 spine_02 {parent=spine_01, loc=<6.5997 -0.2095 0>, rot=<0 0 0.0238 -0.9997>}
 spine_03 {parent=spine_02, loc=<6.4914 -0.1237 0>, rot=<0 0 0.0149 -0.9999>}
 spine_04 {parent=spine_03, loc=<6.9598 0.0794 0>, rot=<0 0 -0.0132 -0.9999>}
 spine_05 {parent=spine_04, loc=<6.9987 0 0>, rot=<0 0 -0.0015 -1>}
 neck_01 {parent=spine_05, loc=<11.3138 0.1529 0>, rot=<0 0 -0.2003 -0.9797>}
 neck_02 {parent=neck_01, loc=<4.5627 0 0>, rot=<0 0 0.0504 -0.9987>}
 head {parent=neck_02, loc=<5.3914 0 0>, rot=<0 0 0.115 -0.9934>}
 base_hair_back_M_01 {parent=head, loc=<4.0613 -8.3419 -0.1698>, rot=<0.0928 -0.701 0.7028 -0.0775>}
 dyn_hair_back_M_01 {parent=base_hair_back_M_01, loc=<0 0 0>, rot=<0 0 0 1>}
 dyn_hair_back_M_02 {parent=dyn_hair_back_M_01, loc=<8.331 0 0>, rot=<0.0005 -0.0224 0.0001 -0.9997>}
 dyn_hair_back_M_03 {parent=dyn_hair_back_M_02, loc=<8.3208 0 0>, rot=<0.0007 -0.0312 0.0001 -0.9995>}
 dyn_hair_back_M_04 {parent=dyn_hair_back_M_03, loc=<8.2927 0 0>, rot=<0.0028 -0.1284 0 -0.9917>}
 dyn_hair_back_M_05 {parent=dyn_hair_back_M_04, loc=<8.2219 0 0>, rot=<0 0 0 1>}
 base_hair_back_R_01 {parent=head, loc=<3.9399 -6.0587 5.5742>, rot=<0.1633 -0.688 0.7044 -0.0617>}
 dyn_hair_back_R_01 {parent=base_hair_back_R_01, loc=<0 0 0>, rot=<0 0 0 1>}
 dyn_hair_back_R_02 {parent=dyn_hair_back_R_01, loc=<8.2687 0 0>, rot=<0.0009 -0.007 -0.0112 -0.9999>}
 dyn_hair_back_R_03 {parent=dyn_hair_back_R_02, loc=<8.3154 0 0>, rot=<0.007 -0.1143 -0.0614 -0.9915>}
 dyn_hair_back_R_04 {parent=dyn_hair_back_R_03, loc=<8.163 0 0>, rot=<0 0 0 1>}
 base_hair_back_L_01 {parent=head, loc=<3.9402 -6.0587 -5.5729>, rot=<0.0617 -0.7044 0.688 -0.1633>}
 dyn_hair_back_L_01 {parent=base_hair_back_L_01, loc=<0 0 0>, rot=<0 0 0 1>}
 dyn_hair_back_L_02 {parent=dyn_hair_back_L_01, loc=<8.2691 0 0>, rot=<-0.0009 -0.007 0.0112 -0.9999>}
 dyn_hair_back_L_03 {parent=dyn_hair_back_L_02, loc=<8.315 0 0>, rot=<-0.007 -0.1143 0.0614 -0.9915>}
 dyn_hair_back_L_04 {parent=dyn_hair_back_L_03, loc=<8.1632 0 0>, rot=<0 0 0 1>}
 base_hair_front_R_01 {parent=head, loc=<-2.5514 0.6569 5.7971>, rot=<0.0038 0.0875 0.9945 -0.0569>}
 dyn_hair_front_R_01 {parent=base_hair_front_R_01, loc=<0 0 0>, rot=<0 0 0 1>}
 dyn_hair_front_R_02 {parent=dyn_hair_front_R_01, loc=<3.2808 0 0>, rot=<0.0038 0.041 0.0795 -0.996>}
 dyn_hair_front_R_03 {parent=dyn_hair_front_R_02, loc=<3.2142 0 0>, rot=<0 0 0 1>}
 dyn_hair_front_R_05 {parent=dyn_hair_front_R_02, loc=<-0.2673 1.3576 1.2376>, rot=<0.0803 0.187 -0.0672 -0.9768>}
 dyn_hair_front_R_07 {parent=dyn_hair_front_R_02, loc=<0.0051 -1.3339 1.1728>, rot=<0.0607 0.222 0.0372 -0.9724>}
 dyn_hair_front_R_08 {parent=dyn_hair_front_R_07, loc=<2.4263 0 -0.148>, rot=<0 0 0 1>}
 dyn_hair_front_R_09 {parent=dyn_hair_front_R_02, loc=<0.2066 -3.183 0.2296>, rot=<0.0301 -0.266 0.1015 -0.9582>}
 base_hair_front_L_01 {parent=head, loc=<-2.6562 1.5281 -5.7473>, rot=<-0.0701 -0.9922 -0.083 -0.0606>}
 dyn_hair_front_L_01 {parent=base_hair_front_L_01, loc=<0 0 0>, rot=<0 0 0 1>}
 dyn_hair_front_L_02 {parent=dyn_hair_front_L_01, loc=<3.8099 0 0>, rot=<-0.0094 0.0852 -0.1467 -0.9855>}
 dyn_hair_front_L_03 {parent=dyn_hair_front_L_02, loc=<3.809 0 0>, rot=<-0.0151 0.1382 -0.1036 -0.9849>}
 dyn_hair_frontB_L_01 {parent=dyn_hair_front_L_01, loc=<1.9851 -3.0955 -0.5815>, rot=<-0.0091 0.2323 -0.0364 -0.9719>}
 dyn_hair_frontB_L_02 {parent=dyn_hair_frontB_L_01, loc=<4.5952 0 0>, rot=<-0.0058 0.1104 -0.0378 -0.9932>}
 dyn_hair_frontA_L_01 {parent=dyn_hair_front_L_01, loc=<1.5478 2.2084 0.0045>, rot=<0.0143 -0.3027 -0.1321 -0.9438>}
 dyn_hair_frontA_L_02 {parent=dyn_hair_frontA_L_01, loc=<2.7984 0 0>, rot=<0.0194 -0.0495 -0.226 -0.9727>}
 base_hair_frontA_R_01 {parent=head, loc=<-1.9052 5.1958 6.0653>, rot=<-0.1664 0.1116 0.9664 0.1609>}
 dyn_hair_frontA_R_01 {parent=base_hair_frontA_R_01, loc=<0 0 0>, rot=<0 0 0 1>}
 base_hair_frontC_L_01 {parent=head, loc=<-0.161 6.7099 -5.5123>, rot=<-0.0669 -0.8507 0.4167 0.3133>}
 dyn_hair_frontC_L_01 {parent=base_hair_frontC_L_01, loc=<0 0 0>, rot=<0 0 0 1>}
 base_hair_BangsA_01 {parent=head, loc=<5.6525 8.3822 5.1431>, rot=<-0.0993 0.3585 0.9044 0.209>}
 dyn_hair_BangsA_01 {parent=base_hair_BangsA_01, loc=<0 0 0>, rot=<0 0 0 1>}
 dyn_hair_BangsA_02 {parent=dyn_hair_BangsA_01, loc=<2.0937 0 0>, rot=<0.0027 -0.1196 0.0937 -0.9884>}
 base_hair_BangsB_01 {parent=head, loc=<6.2589 10.0119 2.5967>, rot=<-0.0455 0.6453 0.7315 0.2153>}
 dyn_hair_BangsB_01 {parent=base_hair_BangsB_01, loc=<0 0 0>, rot=<0 0 0 1>}
 dyn_hair_BangsB_02 {parent=dyn_hair_BangsB_01, loc=<2.7025 0 0>, rot=<0.0499 -0.1301 -0.0053 -0.9902>}
 base_hair_BangsC_01 {parent=head, loc=<6.8377 9.1559 -1.2557>, rot=<-0.0157 -0.7655 -0.6372 -0.0877>}
 dyn_hair_BangsC_01 {parent=base_hair_BangsC_01, loc=<0 0 0>, rot=<0 0 0 1>}
 dyn_hair_BangsC_02 {parent=dyn_hair_BangsC_01, loc=<3.5224 0 0>, rot=<0.0113 -0.1447 -0.0091 -0.9894>}
 clavicle_l {parent=spine_05, loc=<5.4831 -0.1053 -3.2946>, rot=<0.0465 -0.7289 -0.009 -0.6829>}
 upperarm_l {parent=clavicle_l, loc=<9.8658 -1.2136 -0.4105>, rot=<0.0164 -0.3245 -0.0269 -0.9454>}
 lowerarm_l {parent=upperarm_l, loc=<26.465 -0.8345 0.259>, rot=<0.0036 0.0161 -0.2298 -0.9731>}
 hand_l {parent=lowerarm_l, loc=<22.7355 -0.3203 -0.6105>, rot=<-0.6784 -0.0223 0.0735 -0.7307>}
 thumb_01_l {parent=hand_l, loc=<1.3118 -0.3191 -2.459>, rot=<0.6546 -0.4042 0.0664 -0.6354>}
 thumb_02_l {parent=thumb_01_l, loc=<5.9336 -0.0448 0.0401>, rot=<-0.0037 0.0213 0.298 -0.9543>}
 thumb_03_l {parent=thumb_02_l, loc=<2.2223 0.3682 -0.0907>, rot=<0.0118 -0.013 0.0066 -0.9998>}
 index_metacarpal_l {parent=hand_l, loc=<1.6535 0.0183 -2.96>, rot=<0.0158 -0.0124 -0.0173 -0.9996>}
 index_01_l {parent=index_metacarpal_l, loc=<8.3008 0.0357 -0.0524>, rot=<-0.0044 0.0485 0.0924 -0.9945>}
 index_02_l {parent=index_01_l, loc=<3.3587 -0.3004 0.0452>, rot=<-0.002 0.0027 0.3225 -0.9466>}
 index_03_l {parent=index_02_l, loc=<2.5578 -0.123 -0.0252>, rot=<-0.002 0.0067 0.2558 -0.9667>}
 middle_metacarpal_l {parent=hand_l, loc=<2.1856 -0.1725 -0.5984>, rot=<0.006 0.0168 -0.0329 -0.9993>}
 middle_01_l {parent=middle_metacarpal_l, loc=<7.3275 -0.2447 0.0129>, rot=<-0.0208 0.0042 0.0858 -0.9961>}
 middle_02_l {parent=middle_01_l, loc=<4.0123 0.1401 0.1175>, rot=<-0.0218 0.0258 0.3512 -0.9357>}
 middle_03_l {parent=middle_02_l, loc=<2.6077 0.0534 0.1373>, rot=<0.0082 0.0188 0.1708 -0.9851>}
 ring_metacarpal_l {parent=hand_l, loc=<2.2074 -0.3625 1.3638>, rot=<-0.0284 0.0722 -0.0458 -0.9959>}
 ring_01_l {parent=ring_metacarpal_l, loc=<7.0295 -0.6197 -0.1446>, rot=<-0.0489 -0.0252 0.1892 -0.9804>}
 ring_02_l {parent=ring_01_l, loc=<3.6174 0.4647 -0.1027>, rot=<0 0 0.2806 -0.9598>}
 ring_03_l {parent=ring_02_l, loc=<2.4152 0.114 0.0199>, rot=<0 0 0.1189 -0.9929>}
 pinky_metacarpal_l {parent=hand_l, loc=<1.7616 -0.9753 3.3814>, rot=<-0.0397 0.0814 -0.069 -0.9935>}
 pinky_01_l {parent=pinky_metacarpal_l, loc=<7.1709 -0.4831 -0.1094>, rot=<-0.0654 -0.0035 0.2242 -0.9723>}
 pinky_02_l {parent=pinky_01_l, loc=<2.8274 0.0893 -0.1043>, rot=<0.0037 0.0011 0.2854 -0.9584>}
 pinky_03_l {parent=pinky_02_l, loc=<2.2046 0.0307 -0.0078>, rot=<-0.0002 0.0098 0.1445 -0.9895>}
 deform_wrist_dn_base_l {parent=hand_l, loc=<0 -1.5326 0>, rot=<0 0 0 -1>}
 deform_wrist_dn_l {parent=deform_wrist_dn_base_l, loc=<0 0 0>, rot=<0 0 0 1>}
 deform_wrist_up_base_l {parent=hand_l, loc=<0 2.3009 0>, rot=<0 0 0 -1>}
 deform_wrist_up_l {parent=deform_wrist_up_base_l, loc=<0 0 0>, rot=<0 0 0 1>}
 deform_wrist_l_main {parent=hand_l, loc=<0 0 0>, rot=<0 0 0 1>}
 deform_wrist_f_base_l {parent=deform_wrist_l_main, loc=<0 0 -3.869>, rot=<0 0 0 1>}
 deform_wrist_f_l {parent=deform_wrist_f_base_l, loc=<0 0 0>, rot=<0 0 0 1>}
 deform_wrist_b_base_l {parent=deform_wrist_l_main, loc=<0 0 3.649>, rot=<0 0 0 1>}
 deform_wrist_b_l {parent=deform_wrist_b_base_l, loc=<0 0 0>, rot=<0 0 0 1>}
 lowerarm_twist_02_l {parent=lowerarm_l, loc=<10.6707 0 0>, rot=<0 0 0 1>}
 lowerarm_twist_01_l {parent=lowerarm_l, loc=<18.316 0 0>, rot=<0 0 0 1>}
 deform_elbow_in_base_l {parent=lowerarm_l, loc=<0.1834 3.6062 -0.2572>, rot=<0 0 0 -1>}
 deform_elbow_in_l {parent=deform_elbow_in_base_l, loc=<0 0 0>, rot=<0 0 0 1>}
 deform_elbow_out_base_l {parent=lowerarm_l, loc=<0 -4 0>, rot=<0 0 0 1>}
 deform_elbow_out_l {parent=deform_elbow_out_base_l, loc=<0 0 0>, rot=<0 0 0 1>}
 upperarm_twist_01_l {parent=upperarm_l, loc=<7.2135 -0.1134 0>, rot=<0 0 0 1>}
 deform_upperarm_bk_base_l {parent=upperarm_twist_01_l, loc=<-6.4053 -4.3746 0.314>, rot=<0 0 0 -1>}
 deform_upperarm_bk_l {parent=deform_upperarm_bk_base_l, loc=<0 0 0>, rot=<0 0 0 1>}
 deform_upperarm_fr_base_l {parent=upperarm_twist_01_l, loc=<-5.6312 5.092 0.3354>, rot=<0 0 0 -1>}
 deform_upperarm_fr_l {parent=deform_upperarm_fr_base_l, loc=<0 0 0>, rot=<0 0 0 1>}
 deform_upperarm_up_base_l {parent=upperarm_twist_01_l, loc=<-8.891 0.4825 3.9517>, rot=<0 0 0 -1>}
 deform_upperarm_up_l {parent=deform_upperarm_up_base_l, loc=<0 0 0>, rot=<0 0 0 1>}
 upperarm_twist_02_l {parent=upperarm_l, loc=<19.7998 -0.1134 0>, rot=<0 0 0 1>}
 clavicle_r {parent=spine_05, loc=<5.4831 -0.1053 3.2946>, rot=<0.7289 0.0465 -0.6829 0.009>}
 upperarm_r {parent=clavicle_r, loc=<-9.8658 1.2136 0.4105>, rot=<0.0164 -0.3245 -0.0269 -0.9454>}
 lowerarm_r {parent=upperarm_r, loc=<-26.465 0.8345 -0.259>, rot=<0.0036 0.0161 -0.2298 -0.9731>}
 hand_r {parent=lowerarm_r, loc=<-22.7355 0.3203 0.6106>, rot=<-0.6784 -0.0223 0.0735 -0.7307>}
 thumb_01_r {parent=hand_r, loc=<-1.3118 0.3191 2.459>, rot=<0.6546 -0.4042 0.0664 -0.6354>}
 thumb_02_r {parent=thumb_01_r, loc=<-5.9336 0.0448 -0.0401>, rot=<-0.0037 0.0213 0.298 -0.9543>}
 thumb_03_r {parent=thumb_02_r, loc=<-2.2223 -0.3682 0.0907>, rot=<0.0118 -0.013 0.0066 -0.9998>}
 index_metacarpal_r {parent=hand_r, loc=<-1.6535 -0.0183 2.96>, rot=<0.0158 -0.0124 -0.0173 -0.9996>}
 index_01_r {parent=index_metacarpal_r, loc=<-8.3008 -0.0357 0.0524>, rot=<-0.0044 0.0485 0.0924 -0.9945>}
 index_02_r {parent=index_01_r, loc=<-3.3587 0.3004 -0.0452>, rot=<-0.002 0.0027 0.3225 -0.9466>}
 index_03_r {parent=index_02_r, loc=<-2.5578 0.123 0.0252>, rot=<-0.002 0.0067 0.2558 -0.9667>}
 middle_metacarpal_r {parent=hand_r, loc=<-2.1856 0.1725 0.5984>, rot=<0.006 0.0168 -0.0329 -0.9993>}
 middle_01_r {parent=middle_metacarpal_r, loc=<-7.3275 0.2447 -0.0129>, rot=<-0.0208 0.0042 0.0858 -0.9961>}
 middle_02_r {parent=middle_01_r, loc=<-4.0123 -0.1401 -0.1175>, rot=<-0.0218 0.0258 0.3512 -0.9357>}
 middle_03_r {parent=middle_02_r, loc=<-2.6077 -0.0534 -0.1373>, rot=<0.0082 0.0188 0.1708 -0.9851>}
 ring_metacarpal_r {parent=hand_r, loc=<-2.2074 0.3625 -1.3638>, rot=<-0.0284 0.0722 -0.0458 -0.9959>}
 ring_01_r {parent=ring_metacarpal_r, loc=<-7.0295 0.6197 0.1446>, rot=<-0.0489 -0.0252 0.1892 -0.9804>}
 ring_02_r {parent=ring_01_r, loc=<-3.6174 -0.4647 0.1027>, rot=<0 0 0.2806 -0.9598>}
 ring_03_r {parent=ring_02_r, loc=<-2.4152 -0.114 -0.0199>, rot=<0 0 0.1189 -0.9929>}
 pinky_metacarpal_r {parent=hand_r, loc=<-1.7616 0.9753 -3.3814>, rot=<-0.0397 0.0814 -0.069 -0.9935>}
 pinky_01_r {parent=pinky_metacarpal_r, loc=<-7.1709 0.4831 0.1094>, rot=<-0.0654 -0.0035 0.2242 -0.9723>}
 pinky_02_r {parent=pinky_01_r, loc=<-2.8274 -0.0893 0.1043>, rot=<0.0037 0.0011 0.2854 -0.9584>}
 pinky_03_r {parent=pinky_02_r, loc=<-2.2046 -0.0307 0.0078>, rot=<-0.0002 0.0098 0.1445 -0.9895>}
 deform_wrist_dn_base_r {parent=hand_r, loc=<-0.0001 1.5328 0>, rot=<0 0 0 -1>}
 deform_wrist_dn_r {parent=deform_wrist_dn_base_r, loc=<0 0 0>, rot=<0 0 0 1>}
 deform_wrist_up_base_r {parent=hand_r, loc=<-0.0002 -2.3007 -0.0001>, rot=<0 0 0 -1>}
 deform_wrist_up_r {parent=deform_wrist_up_base_r, loc=<0 0 0>, rot=<0 0 0 1>}
 deform_wrist_r_main {parent=hand_r, loc=<0 0 0>, rot=<0 0 0 1>}
 deform_wrist_f_base_r {parent=deform_wrist_r_main, loc=<0 0 3.8687>, rot=<0 0 0 1>}
 deform_wrist_f_r {parent=deform_wrist_f_base_r, loc=<0 0 0>, rot=<0 0 0 1>}
 deform_wrist_b_base_r {parent=deform_wrist_r_main, loc=<0 0 -3.649>, rot=<0 0 0 1>}
 deform_wrist_b_r {parent=deform_wrist_b_base_r, loc=<0 0 0>, rot=<0 0 0 1>}
 lowerarm_twist_02_r {parent=lowerarm_r, loc=<-10.6707 0 0>, rot=<0 0 0 -1>}
 lowerarm_twist_01_r {parent=lowerarm_r, loc=<-18.316 0 0>, rot=<0 0 0 -1>}
 deform_elbow_in_base_r {parent=lowerarm_r, loc=<-0.1834 -3.6062 0.2572>, rot=<0 0 0 -1>}
 deform_elbow_in_r {parent=deform_elbow_in_base_r, loc=<0 0 0>, rot=<0 0 0 1>}
 deform_elbow_out_base_r {parent=lowerarm_r, loc=<0 4 -0.0001>, rot=<0 0 0 -1>}
 deform_elbow_out_r {parent=deform_elbow_out_base_r, loc=<0 0 0>, rot=<0 0 0 1>}
 upperarm_twist_01_r {parent=upperarm_r, loc=<-7.2135 0.1134 0>, rot=<0 0 0 -1>}
 deform_upperarm_bk_base_r {parent=upperarm_twist_01_r, loc=<6.4053 4.3746 -0.314>, rot=<0 0 0 -1>}
 deform_upperarm_bk_r {parent=deform_upperarm_bk_base_r, loc=<0 0 0>, rot=<0 0 0 1>}
 deform_upperarm_fr_base_r {parent=upperarm_twist_01_r, loc=<5.6312 -5.092 -0.3354>, rot=<0 0 0 -1>}
 deform_upperarm_fr_r {parent=deform_upperarm_fr_base_r, loc=<0 0 0>, rot=<0 0 0 1>}
 deform_upperarm_up_base_r {parent=upperarm_twist_01_r, loc=<8.891 -0.4825 -3.9517>, rot=<0 0 0 -1>}
 deform_upperarm_up_r {parent=deform_upperarm_up_base_r, loc=<0 0 0>, rot=<0 0 0 1>}
 upperarm_twist_02_r {parent=upperarm_r, loc=<-19.7998 0.1134 0>, rot=<0 0 0 -1>}
 attach_backpack {parent=spine_05, loc=<-0.123 -13.473 0>, rot=<-0.0093 -0.707 0.0093 -0.707>}
 attach_cape {parent=spine_05, loc=<9.017 -11.714 0>, rot=<0.0616 -0.7044 -0.0616 -0.7044>}
 base_necklace_01 {parent=spine_05, loc=<6.6183 5.3438 -0.0108>, rot=<0.2537 0.66 0.6604 0.2528>}
 dyn_necklace_01 {parent=base_necklace_01, loc=<0 0 0>, rot=<0 0 0 1>}
 dyn_necklace_02 {parent=dyn_necklace_01, loc=<3.475 0 0>, rot=<0 -0.0062 0.0007 -1>}
 dyn_necklace_03 {parent=dyn_necklace_02, loc=<3.4829 0 0>, rot=<-0.0001 -0.0699 -0.0011 -0.9976>}
 dyn_necklace_04 {parent=dyn_necklace_03, loc=<3.4724 0 0>, rot=<0 0 0 1>}
 deform_pec_aim_base_l {parent=spine_04, loc=<3.2456 9.3309 -1.2237>, rot=<0.0603 -0.4033 0.3318 -0.8507>}
 deform_pec_aim_l {parent=deform_pec_aim_base_l, loc=<0 0 0>, rot=<0 0 0 1>}
 deform_pec_base_l {parent=deform_pec_aim_l, loc=<9.3665 0 0>, rot=<0 0 0 1>}
 deform_pec_l {parent=deform_pec_base_l, loc=<0 0 0>, rot=<0 0 0 1>}
 deform_pec_aim_base_r {parent=spine_04, loc=<3.2452 9.3309 1.2237>, rot=<-0.4033 -0.0603 0.8507 0.3318>}
 deform_pec_aim_r {parent=deform_pec_aim_base_r, loc=<0 0 0>, rot=<0 0 0 1>}
 deform_pec_base_r {parent=deform_pec_aim_r, loc=<-9.3667 -0.0003 -0.0004>, rot=<0 0 0 1>}
 deform_pec_r {parent=deform_pec_base_r, loc=<0 0 0>, rot=<0 0 0 1>}
 base_cloth_f_01 {parent=spine_03, loc=<2.2186 12.6084 0>, rot=<0.0423 -0.7058 -0.7058 0.0423>}
 dyn_cloth_f_01 {parent=base_cloth_f_01, loc=<0 0 0>, rot=<0 0 0 1>}
 base_cloth_r_01 {parent=spine_03, loc=<3.4327 1.08 11.7923>, rot=<0 0 0.9982 -0.0599>}
 dyn_cloth_r_01 {parent=base_cloth_r_01, loc=<0 0 0>, rot=<0 0 0 1>}
 base_cloth_f_r_01 {parent=spine_03, loc=<2.5147 9.949 8.6693>, rot=<-0.0229 0.382 0.9222 -0.0553>}
 dyn_cloth_f_r_01 {parent=base_cloth_f_r_01, loc=<0 0 0>, rot=<0 0 0 1>}
 base_cloth_b_01 {parent=spine_03, loc=<4.1301 -7.2902 0>, rot=<0.0423 -0.7058 0.7058 -0.0423>}
 dyn_cloth_b_01 {parent=base_cloth_b_01, loc=<0 0 0>, rot=<0 0 0 1>}
 base_cloth_b_r_01 {parent=spine_03, loc=<3.8863 -5.8676 7.6406>, rot=<0.0229 -0.382 0.9222 -0.0553>}
 dyn_cloth_b_r_01 {parent=base_cloth_b_r_01, loc=<0 0 0>, rot=<0 0 0 1>}
 base_cloth_f_l_01 {parent=spine_03, loc=<2.5143 9.949 -8.6693>, rot=<0.0553 -0.9222 -0.382 0.0229>}
 dyn_cloth_f_l_01 {parent=base_cloth_f_l_01, loc=<0 0 0>, rot=<0 0 0 1>}
 base_cloth_b_l_01 {parent=spine_03, loc=<3.8864 -5.8675 -7.6406>, rot=<0.0553 -0.9222 0.382 -0.0229>}
 dyn_cloth_b_l_01 {parent=base_cloth_b_l_01, loc=<0 0 0>, rot=<0 0 0 1>}
 base_cloth_l_01 {parent=spine_03, loc=<3.4322 1.08 -11.7923>, rot=<0.0599 -0.9982 0 0>}
 dyn_cloth_l_01 {parent=base_cloth_l_01, loc=<0 0 0>, rot=<0 0 0 1>}
 thigh_l {parent=pelvis, loc=<-1.0855 -0.2109 -8.7053>, rot=<0.0009 0.0355 0.0155 -0.9993>}
 calf_l {parent=thigh_l, loc=<-42.5284 -0.0019 -0.7297>, rot=<0.0006 -0.0161 -0.0391 -0.9991>}
 foot_l {parent=calf_l, loc=<-46.7401 -2.5014 -0.592>, rot=<0.0427 -0.0238 0.0353 -0.9982>}
 ball_l {parent=foot_l, loc=<-3.5658 10.8066 -0.1544>, rot=<0.0001 -0.0001 -0.7063 -0.7079>}
 deform_foot_main_l {parent=foot_l, loc=<0 0 0>, rot=<0 0 0 1>}
 deform_foot_back_base_l {parent=deform_foot_main_l, loc=<4.1251 -4.8612 -0.4181>, rot=<0 0 0 1>}
 deform_foot_back_l {parent=deform_foot_back_base_l, loc=<0 0 0>, rot=<0 0 0 1>}
 deform_foot_front_base_l {parent=deform_foot_main_l, loc=<5.7821 7.3427 0.7002>, rot=<0 0 0 1>}
 deform_foot_front_l {parent=deform_foot_front_base_l, loc=<0 0 0>, rot=<0 0 0 1>}
 calf_twist_02_l {parent=calf_l, loc=<-19.1546 -1.7943 -0.2638>, rot=<0.0237 0.0095 -0.0035 -0.9997>}
 calf_twist_01_l {parent=calf_l, loc=<-35.8483 -1.9197 -0.5794>, rot=<0.0237 0.0095 -0.0035 -0.9997>}
 deform_knee_in_base_l {parent=calf_l, loc=<0.8031 -5.0231 0.2456>, rot=<0 0 0 -1>}
 deform_knee_in_l {parent=deform_knee_in_base_l, loc=<0 0 0>, rot=<0 0 0 1>}
 deform_knee_out_base_l {parent=calf_l, loc=<0 4 0>, rot=<0 0 0 -1>}
 deform_knee_out_l {parent=deform_knee_out_base_l, loc=<0 0 0>, rot=<0 0 0 1>}
 base_kuzi_f_l_01 {parent=calf_l, loc=<-4.0306 7.8173 -0.2924>, rot=<-0.048 -0.7054 -0.7043 -0.0632>}
 dyn_kuzi_f_l_01 {parent=base_kuzi_f_l_01, loc=<0 0 0>, rot=<0 0 0 1>}
 dyn_kuzi_f_l_02 {parent=dyn_kuzi_f_l_01, loc=<8.3224 0 0>, rot=<0 0 0 1>}
 dyn_kuzi_f_l_03 {parent=dyn_kuzi_f_l_02, loc=<8.3224 0 0>, rot=<0 0 0 1>}
 base_kuzi_out_l_01 {parent=calf_l, loc=<-4.4692 -1.0545 -7.2533>, rot=<0.0021 -0.9953 -0.0034 -0.0972>}
 dyn_kuzi_out_l_01 {parent=base_kuzi_out_l_01, loc=<0 0 0>, rot=<0 0 0 1>}
 dyn_kuzi_out_l_02 {parent=dyn_kuzi_out_l_01, loc=<8.3224 0 0>, rot=<0 0 0 -1>}
 dyn_kuzi_out_l_03 {parent=dyn_kuzi_out_l_02, loc=<8.3224 0 0>, rot=<0 0 0 1>}
 base_kuzi_b_l_01 {parent=calf_l, loc=<-4.0275 -9.7003 0.2819>, rot=<0.0428 -0.7099 0.7006 -0.0577>}
 dyn_kuzi_b_l_01 {parent=base_kuzi_b_l_01, loc=<0 0 0>, rot=<0 0 0 1>}
 dyn_kuzi_b_l_02 {parent=dyn_kuzi_b_l_01, loc=<8.3223 0 0>, rot=<0 0 0 -1>}
 dyn_kuzi_b_l_03 {parent=dyn_kuzi_b_l_02, loc=<8.3223 0 0>, rot=<0 0 0 1>}
 base_kuzi_in_l_01 {parent=calf_l, loc=<-4.5598 -0.7821 9.1726>, rot=<0.0499 -0.0032 0.9987 0.0048>}
 dyn_kuzi_in_l_01 {parent=base_kuzi_in_l_01, loc=<0 0 0>, rot=<0 0 0 1>}
 dyn_kuzi_in_l_02 {parent=dyn_kuzi_in_l_01, loc=<7.8439 0 0>, rot=<0 0 0 1>}
 dyn_kuzi_in_l_03 {parent=dyn_kuzi_in_l_02, loc=<7.8438 0 0>, rot=<0 0 0 1>}
 base_calf_libbon_l_01 {parent=calf_l, loc=<-36.146 -6.4679 -0.1438>, rot=<0.1896 -0.6704 -0.6587 0.284>}
 dyn_calf_libbon_l_01 {parent=base_calf_libbon_l_01, loc=<-0.0057 0.0012 0>, rot=<0 0 0 1>}
 dyn_calf_libbon_l_02 {parent=dyn_calf_libbon_l_01, loc=<2.0872 0 0>, rot=<0 0 0 1>}
 thigh_twist_01_l {parent=thigh_l, loc=<-27.952 0 0>, rot=<0.0013 0 0 -1>}
 deform_groin_base_l {parent=thigh_l, loc=<0.4801 5.6776 -0.2486>, rot=<0 0 0 -1>}
 deform_groin_l {parent=deform_groin_base_l, loc=<0 0 0>, rot=<0 0 0 1>}
 deform_groin_out_base_l {parent=thigh_l, loc=<-3.1912 0.0122 -7.631>, rot=<0 0 0 -1>}
 deform_groin_out_l {parent=deform_groin_out_base_l, loc=<0 0 0>, rot=<0 0 0 1>}
 deform_groin_f_base_l {parent=thigh_l, loc=<-1.3805 8.0703 -0.2486>, rot=<0 0 0 -1>}
 deform_groin_f_l {parent=deform_groin_f_base_l, loc=<0 0 0>, rot=<0 0 0 -1>}
 base_bag_01 {parent=thigh_l, loc=<-2.9327 4.1679 -6.9143>, rot=<-0.2662 0.0908 0.0591 -0.9578>}
 follow_bag_01 {parent=base_bag_01, loc=<0 0 0>, rot=<0 0 0 1>}
 thigh_r {parent=pelvis, loc=<-1.0855 -0.2109 8.7053>, rot=<0.0355 -0.0009 0.9993 0.0155>}
 calf_r {parent=thigh_r, loc=<42.5284 0.0019 0.7297>, rot=<0.0006 -0.0161 -0.0391 -0.9991>}
 foot_r {parent=calf_r, loc=<46.7401 2.5014 0.592>, rot=<0.0427 -0.0238 0.0353 -0.9982>}
 ball_r {parent=foot_r, loc=<3.5658 -10.8066 0.1544>, rot=<0.0001 -0.0001 -0.7063 -0.7079>}
 deform_foot_main_r {parent=foot_r, loc=<0 0 0>, rot=<0 0 0 1>}
 deform_foot_back_base_r {parent=deform_foot_main_r, loc=<-4.1251 4.8612 0.4181>, rot=<0 0 0 1>}
 deform_foot_back_r {parent=deform_foot_back_base_r, loc=<0 0 0>, rot=<0 0 0 1>}
 deform_foot_front_base_r {parent=deform_foot_main_r, loc=<-5.7821 -7.3427 -0.7002>, rot=<0 0 0 1>}
 deform_foot_front_r {parent=deform_foot_front_base_r, loc=<0 0 0>, rot=<0 0 0 1>}
 calf_twist_02_r {parent=calf_r, loc=<19.1546 1.7943 0.2638>, rot=<0.0237 0.0095 -0.0035 -0.9997>}
 calf_twist_01_r {parent=calf_r, loc=<35.8483 1.9197 0.5794>, rot=<0.0237 0.0095 -0.0035 -0.9997>}
 deform_knee_in_base_r {parent=calf_r, loc=<-0.8031 5.0231 -0.2457>, rot=<0 0 0 -1>}
 deform_knee_in_r {parent=deform_knee_in_base_r, loc=<0 0 0>, rot=<0 0 0 1>}
 deform_knee_out_base_r {parent=calf_r, loc=<0 -4 0>, rot=<0 0 0 -1>}
 deform_knee_out_r {parent=deform_knee_out_base_r, loc=<0 0 0>, rot=<0 0 0 1>}
 base_kuzi_f_r_01 {parent=calf_r, loc=<4.0307 -7.8173 0.2924>, rot=<-0.7043 0.0632 0.048 -0.7054>}
 dyn_kuzi_f_r_01 {parent=base_kuzi_f_r_01, loc=<0 0 0>, rot=<0 0 0 1>}
 dyn_kuzi_f_r_02 {parent=dyn_kuzi_f_r_01, loc=<8.3224 0 0>, rot=<0 0 0 1>}
 dyn_kuzi_f_r_03 {parent=dyn_kuzi_f_r_02, loc=<8.3224 0 0>, rot=<0 0 0 1>}
 base_kuzi_out_r_01 {parent=calf_r, loc=<4.4693 1.0545 7.2533>, rot=<-0.0034 0.0972 -0.0021 -0.9953>}
 dyn_kuzi_out_r_01 {parent=base_kuzi_out_r_01, loc=<0 0 0>, rot=<0 0 0 1>}
 dyn_kuzi_out_r_02 {parent=dyn_kuzi_out_r_01, loc=<8.3224 0 0>, rot=<0 0 0 1>}
 dyn_kuzi_out_r_03 {parent=dyn_kuzi_out_r_02, loc=<8.3224 0 0>, rot=<0 0 0 1>}
 base_kuzi_b_r_01 {parent=calf_r, loc=<4.0275 9.7003 -0.2819>, rot=<0.7006 0.0577 -0.0428 -0.7099>}
 dyn_kuzi_b_r_01 {parent=base_kuzi_b_r_01, loc=<0 0 0>, rot=<0 0 0 1>}
 dyn_kuzi_b_r_02 {parent=dyn_kuzi_b_r_01, loc=<8.3224 0 0>, rot=<0 0 0 1>}
 dyn_kuzi_b_r_03 {parent=dyn_kuzi_b_r_02, loc=<8.3224 0 0>, rot=<0 0 0 1>}
 base_kuzi_in_r_01 {parent=calf_r, loc=<4.5598 0.7821 -9.1726>, rot=<0.9987 -0.0048 -0.0499 -0.0032>}
 dyn_kuzi_in_r_01 {parent=base_kuzi_in_r_01, loc=<0 0 0>, rot=<0 0 0 1>}
 dyn_kuzi_in_r_02 {parent=dyn_kuzi_in_r_01, loc=<7.8438 0 0>, rot=<0 0 0 1>}
 dyn_kuzi_in_r_03 {parent=dyn_kuzi_in_r_02, loc=<7.8438 0 0>, rot=<0 0 0 1>}
 base_calf_ribbon_r_01 {parent=calf_r, loc=<36.1416 6.4642 0.1446>, rot=<-0.6587 -0.284 -0.1896 -0.6704>}
 dyn_calf_ribbon_r_01 {parent=base_calf_ribbon_r_01, loc=<0 0 0>, rot=<0 0 0 1>}
 dyn_calf_ribbon_r_02 {parent=dyn_calf_ribbon_r_01, loc=<2.0873 0 0>, rot=<0 0 0 1>}
 thigh_twist_01_r {parent=thigh_r, loc=<27.952 0 0>, rot=<0.0013 0 0 -1>}
 deform_groin_base_r {parent=thigh_r, loc=<-0.4801 -5.6776 0.2486>, rot=<0 0 0 -1>}
 deform_groin_r {parent=deform_groin_base_r, loc=<0 0 0>, rot=<0 0 0 1>}
 deform_groin_out_base_r {parent=thigh_r, loc=<-2.7635 0.0112 6.1927>, rot=<0 0 0 -1>}
 deform_groin_out_r {parent=deform_groin_out_base_r, loc=<0 0 0>, rot=<0 0 0 1>}
 deform_groin_f_base_r {parent=thigh_r, loc=<-1.8264 -7.2544 0.3432>, rot=<0 0 0 -1>}
 deform_groin_f_r {parent=deform_groin_f_base_r, loc=<0 0 0>, rot=<0 0 0 1>}
 deform_glute_base_l {parent=pelvis, loc=<-7.3466 -9.0213 -9.1245>, rot=<0.0172 0.031 -0.4477 -0.8935>}
 deform_glute_l {parent=deform_glute_base_l, loc=<0 0 0>, rot=<0 0 0 1>}
 deform_Belt_out_base_l {parent=deform_glute_l, loc=<0 0 0>, rot=<0 0 0 1>}
 deform_Belt_out_l {parent=deform_Belt_out_base_l, loc=<0 0 0>, rot=<0 0 0 1>}
 deform_glute_base_r {parent=pelvis, loc=<-7.3464 -9.0213 9.1245>, rot=<0.031 -0.0172 0.8935 -0.4477>}
 deform_glute_r {parent=deform_glute_base_r, loc=<0 0 0>, rot=<0 0 0 1>}
 neck_01_ctrl {controlType=EULER_TRANSFORM, type=Control, loc=<0 0 0>, rot=<0 0 0 1>}
 neck_02_ctrl {controlType=EULER_TRANSFORM, parent=neck_01_ctrl, type=Control, loc=<0 0 0>, rot=<0 0 0 1>}
 head_ctrl {controlType=EULER_TRANSFORM, parent=neck_02_ctrl, type=Control, loc=<0 0 0>, rot=<0 0 0 1>}
]
variables [
 is Targeting {type=bool, default=True}
 is backpack {type=bool, default=True}
]
