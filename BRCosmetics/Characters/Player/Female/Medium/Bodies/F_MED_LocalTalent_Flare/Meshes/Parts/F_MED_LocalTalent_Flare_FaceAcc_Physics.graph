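# Physics asset for the FaceAcc mesh part of the female medium LocalTalent Flare body, setting up physics for bangs and front hair strands.
# Capsules on upper spine, clavicles, upper arms, neck and head form a fully locked body collision frame.
# Zero-size base hair bodies lock to the head and anchor the dynamic capsule chains, some branching.
# Dynamic joints allow narrow swing and very little twist, limiting how far strands deform.
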PhysicsAsset /BRCosmetics/Characters/Player/Female/Medium/Bodies/F_MED_LocalTalent_Flare/Meshes/Parts/F_MED_LocalTalent_Flare_FaceAcc_Physics
bodies {
 base_hair_BangsA_01 [{shapeType=Capsule, length=0, radius=0, shapeName=base_hair_BangsA_01_capsule}]
 base_hair_BangsB_01 [{shapeType=Capsule, length=0, radius=0, shapeName=base_hair_BangsB_01_capsule}]
 base_hair_BangsC_01 [{shapeType=Capsule, length=0, radius=0, shapeName=base_hair_BangsC_01_capsule}]
 base_hair_frontA_R_01 [{shapeType=Capsule, length=0, radius=0, shapeName=base_hair_frontA_R_01_capsule}]
 base_hair_frontC_L_01 [{shapeType=Capsule, length=0, radius=0, shapeName=base_hair_frontC_L_01_capsule}]
 base_hair_front_L_01 [{shapeType=Capsule, length=0, radius=0, shapeName=base_hair_front_L_01_capsule}]
 base_hair_front_R_01 [{shapeType=Capsule, length=0, radius=0, shapeName=base_hair_front_R_01_capsule}]
 clavicle_l [{
   shapeType=Capsule
   length=6.5724
   radius=5.59
   shapeName=clavicle_l_capsule
   center=<6.4304 -1.1091 -1.5103>
   rotation=<rot -81.2292 179.9993 -179.9992>
  }]
 clavicle_r [{
   shapeType=Capsule
   length=6.5724
   radius=5.59
   shapeName=clavicle_r_capsule
   center=<-6.4304 1.1091 1.5103>
   rotation=<rot 81.2292 -0.0007 -0.0008>
  }]
 dyn_hair_BangsA_01 [{shapeType=Capsule, length=0.4197, radius=1.2826, shapeName=dyn_hair_BangsA_01_capsule, center=<0.9563 0 0.3275>, rotation=<rot 90 0 0>}]
 dyn_hair_BangsA_02 [{shapeType=Capsule, length=0.4197, radius=1.2826, shapeName=dyn_hair_BangsA_02_capsule, center=<0.9563 0 0.672>, rotation=<rot 90 0 0>}]
 dyn_hair_BangsB_01 [{shapeType=Capsule, length=0.8242, radius=1.2826, shapeName=dyn_hair_BangsB_01_capsule, center=<1.1898 0 0>, rotation=<rot 90 0 0>}]
 dyn_hair_BangsB_02 [{shapeType=Capsule, length=0.8242, radius=1.2826, shapeName=dyn_hair_BangsB_02_capsule, center=<1.2611 0 0.4095>, rotation=<rot 90 0 0>}]
 dyn_hair_BangsC_01 [{shapeType=Capsule, length=2.312, radius=1.2826, shapeName=dyn_hair_BangsC_01_capsule, center=<1.7366 0 0.6376>, rotation=<rot 90 0 0>}]
 dyn_hair_BangsC_02 [{
   shapeType=Capsule
   length=2.312
   radius=1.2826
   shapeName=dyn_hair_BangsC_02_capsule
   center=<1.6807 0 0.8328>
   rotation=<rot 78.2145 -179.9999 -179.9999>
  }]
 dyn_hair_frontA_L_01 [{shapeType=Capsule, length=2.312, radius=1.2826, shapeName=dyn_hair_frontA_L_01_capsule, center=<1.7366 0 0.3927>, rotation=<rot 90 0 0>}]
 dyn_hair_frontA_L_02 [{shapeType=Capsule, length=2.312, radius=1.2826, shapeName=dyn_hair_frontA_L_02_capsule, center=<1.7366 0 0>, rotation=<rot 90 0 0>}]
 dyn_hair_frontA_R_01 [{shapeType=Capsule, length=2.312, radius=1.2826, shapeName=dyn_hair_frontA_R_01_capsule, center=<1.7366 0 0.649>, rotation=<rot 90 0 0>}]
 dyn_hair_frontB_L_01 [{shapeType=Capsule, length=2.2189, radius=1.8293, shapeName=dyn_hair_frontB_L_01_capsule, center=<1.7366 -0.3219 0>, rotation=<rot 90 0 0>}]
 dyn_hair_frontB_L_02 [{shapeType=Capsule, length=2.2189, radius=1.8293, shapeName=dyn_hair_frontB_L_02_capsule, center=<1.7366 -0.3219 0>, rotation=<rot 90 0 0>}]
 dyn_hair_frontC_L_01 [{
   shapeType=Capsule
   length=2.312
   radius=1.2826
   shapeName=dyn_hair_frontC_L_01_capsule
   center=<1.3078 0.6898 -0.1528>
   rotation=<rot 53.9689 -7.4555 0.0001>
  }]
 dyn_hair_front_L_01 [{shapeType=Capsule, length=1.5655, radius=1.2826, shapeName=dyn_hair_front_L_01_capsule, center=<1.7366 0 0.7662>, rotation=<rot 90 0 0>}]
 dyn_hair_front_L_02 [{shapeType=Capsule, length=2.312, radius=1.2826, shapeName=dyn_hair_front_L_02_capsule, center=<1.7366 0 0>, rotation=<rot 90 0 0>}]
 dyn_hair_front_L_03 [{shapeType=Capsule, length=2.312, radius=1.2826, shapeName=dyn_hair_front_L_03_capsule, center=<1.7366 0 0>, rotation=<rot 90 0 0>}]
 dyn_hair_front_R_01 [{shapeType=Capsule, length=2.312, radius=1.2826, shapeName=dyn_hair_front_R_01_capsule, center=<1.7366 0 0.583>, rotation=<rot 90 0 0>}]
 dyn_hair_front_R_02 [{
   shapeType=Capsule
   length=0.6349
   radius=1.7718
   shapeName=dyn_hair_front_R_02_capsule
   center=<1.6925 0.6421 0.3464>
   rotation=<rot 86.5959 179.9997 179.9999>
  }]
 dyn_hair_front_R_03 [{
   shapeType=Capsule
   length=1.3089
   radius=1.2826
   shapeName=dyn_hair_front_R_03_capsule
   center=<1.4461 -0.4663 0.3297>
   rotation=<rot 75.4063 179.9998 -157.2763>
  }]
 dyn_hair_front_R_05 [{shapeType=Capsule, length=2.312, radius=1.2826, shapeName=dyn_hair_front_R_05_capsule, center=<2.2488 0 0.5627>, rotation=<rot 90 0 0>}]
 dyn_hair_front_R_07 [{shapeType=Capsule, length=1.0904, radius=1.2826, shapeName=dyn_hair_front_R_07_capsule, center=<1.211 0 0>, rotation=<rot 90 0 0>}]
 dyn_hair_front_R_08 [{shapeType=Capsule, length=1.3994, radius=1.2826, shapeName=dyn_hair_front_R_08_capsule, center=<1.7913 0 0>, rotation=<rot 90 0 0>}]
 dyn_hair_front_R_09 [{
   shapeType=Capsule
   length=1.6924
   radius=1.2826
   shapeName=dyn_hair_front_R_09_capsule
   center=<1.6644 -0.2792 0.5461>
   rotation=<rot 85.7397 164.9032 -179.9998>
  }]
 head [{shapeType=Capsule, length=7.1252, radius=6.2312, shapeName=head_capsule, center=<3.7601 2.4414 0>, rotation=<rot 90 -5.0919 0>}]
 neck_01 [{shapeType=Capsule, length=5.6736, radius=4.2896, shapeName=neck_01_capsule, center=<1.6896 0.3051 0>, rotation=<rot 90 -1.5992 0>}]
 spine_05 [{shapeType=Capsule, length=9.4688, radius=9.6309, shapeName=spine_05_capsule, center=<-2.5794 0.9135 0>}]
 upperarm_l [{
   shapeType=Capsule
   length=22.1065
   radius=4.9692
   shapeName=upperarm_l_capsule
   center=<13.3838 -0.7385 -0.2283>
   rotation=<rot -87.5661 0.0037 -0.0036>
  }]
 upperarm_r [{
   shapeType=Capsule
   length=22.1065
   radius=4.9692
   shapeName=upperarm_r_capsule
   center=<-13.3838 0.7385 0.2283>
   rotation=<rot 87.5661 -179.9963 -179.9964>
  }]
}
constraints [
 {
  bone1Name=clavicle_l
  bone2Name=spine_05
  swing1LimitDegrees=45
  swing1Motion=Locked
  swing2LimitDegrees=45
  swing2Motion=Locked
  twistLimitDegrees=45
  twistMotion=Locked
 }
 {
  bone1Name=clavicle_r
  bone2Name=spine_05
  swing1LimitDegrees=45
  swing1Motion=Locked
  swing2LimitDegrees=45
  swing2Motion=Locked
  twistLimitDegrees=45
  twistMotion=Locked
 }
 {
  bone1Name=upperarm_r
  bone2Name=clavicle_r
  swing1LimitDegrees=45
  swing1Motion=Locked
  swing2LimitDegrees=45
  swing2Motion=Locked
  twistLimitDegrees=45
  twistMotion=Locked
 }
 {
  bone1Name=upperarm_l
  bone2Name=clavicle_l
  swing1LimitDegrees=45
  swing1Motion=Locked
  swing2LimitDegrees=45
  swing2Motion=Locked
  twistLimitDegrees=45
  twistMotion=Locked
 }
 {
  bone1Name=neck_01
  bone2Name=spine_05
  swing1LimitDegrees=45
  swing1Motion=Locked
  swing2LimitDegrees=45
  swing2Motion=Locked
  twistLimitDegrees=45
  twistMotion=Locked
 }
 {
  bone1Name=head
  bone2Name=neck_01
  swing1LimitDegrees=45
  swing1Motion=Locked
  swing2LimitDegrees=45
  swing2Motion=Locked
  twistLimitDegrees=45
  twistMotion=Locked
 }
 {
  bone1Name=base_hair_BangsA_01
  bone2Name=head
  swing1LimitDegrees=45
  swing1Motion=Locked
  swing2LimitDegrees=45
  swing2Motion=Locked
  twistLimitDegrees=45
  twistMotion=Locked
 }
 {bone1Name=dyn_hair_BangsA_01, bone2Name=base_hair_BangsA_01, swing1LimitDegrees=15, swing2LimitDegrees=15, twistLimitDegrees=2}
 {bone1Name=dyn_hair_BangsA_02, bone2Name=dyn_hair_BangsA_01, swing1LimitDegrees=15, swing2LimitDegrees=15, twistLimitDegrees=2}
 {
  bone1Name=base_hair_BangsB_01
  bone2Name=head
  swing1LimitDegrees=45
  swing1Motion=Locked
  swing2LimitDegrees=45
  swing2Motion=Locked
  twistLimitDegrees=45
  twistMotion=Locked
 }
 {bone1Name=dyn_hair_BangsB_01, bone2Name=base_hair_BangsB_01, swing1LimitDegrees=15, swing2LimitDegrees=15, twistLimitDegrees=2}
 {bone1Name=dyn_hair_BangsB_02, bone2Name=dyn_hair_BangsB_01, swing1LimitDegrees=15, swing2LimitDegrees=15, twistLimitDegrees=2}
 {
  bone1Name=base_hair_BangsC_01
  bone2Name=head
  swing1LimitDegrees=45
  swing1Motion=Locked
  swing2LimitDegrees=45
  swing2Motion=Locked
  twistLimitDegrees=45
  twistMotion=Locked
 }
 {bone1Name=dyn_hair_BangsC_01, bone2Name=base_hair_BangsC_01, swing1LimitDegrees=15, swing2LimitDegrees=15, twistLimitDegrees=2}
 {bone1Name=dyn_hair_BangsC_02, bone2Name=dyn_hair_BangsC_01, swing1LimitDegrees=15, swing2LimitDegrees=15, twistLimitDegrees=2}
 {
  bone1Name=base_hair_frontA_R_01
  bone2Name=head
  swing1LimitDegrees=45
  swing1Motion=Locked
  swing2LimitDegrees=45
  swing2Motion=Locked
  twistLimitDegrees=45
  twistMotion=Locked
 }
 {bone1Name=dyn_hair_frontA_R_01, bone2Name=base_hair_frontA_R_01, swing1LimitDegrees=15, swing2LimitDegrees=15, twistLimitDegrees=2}
 {
  bone1Name=base_hair_frontC_L_01
  bone2Name=head
  swing1LimitDegrees=45
  swing1Motion=Locked
  swing2LimitDegrees=45
  swing2Motion=Locked
  twistLimitDegrees=45
  twistMotion=Locked
 }
 {bone1Name=dyn_hair_frontC_L_01, bone2Name=base_hair_frontC_L_01, swing1LimitDegrees=15, swing2LimitDegrees=15, twistLimitDegrees=2}
 {
  bone1Name=base_hair_front_L_01
  bone2Name=head
  swing1LimitDegrees=45
  swing1Motion=Locked
  swing2LimitDegrees=45
  swing2Motion=Locked
  twistLimitDegrees=45
  twistMotion=Locked
 }
 {bone1Name=dyn_hair_front_L_01, bone2Name=base_hair_front_L_01, swing1LimitDegrees=15, swing2LimitDegrees=15, twistLimitDegrees=2}
 {bone1Name=dyn_hair_frontA_L_01, bone2Name=dyn_hair_front_L_01, swing1LimitDegrees=15, swing2LimitDegrees=15, twistLimitDegrees=2}
 {bone1Name=dyn_hair_frontA_L_02, bone2Name=dyn_hair_frontA_L_01, swing1LimitDegrees=15, swing2LimitDegrees=15, twistLimitDegrees=2}
 {bone1Name=dyn_hair_frontB_L_01, bone2Name=dyn_hair_front_L_01, swing1LimitDegrees=15, swing2LimitDegrees=15, twistLimitDegrees=2}
 {bone1Name=dyn_hair_frontB_L_02, bone2Name=dyn_hair_frontB_L_01, swing1LimitDegrees=15, swing2LimitDegrees=15, twistLimitDegrees=2}
 {bone1Name=dyn_hair_front_L_02, bone2Name=dyn_hair_front_L_01, swing1LimitDegrees=15, swing2LimitDegrees=15, twistLimitDegrees=2}
 {bone1Name=dyn_hair_front_L_03, bone2Name=dyn_hair_front_L_02, swing1LimitDegrees=15, swing2LimitDegrees=15, twistLimitDegrees=2}
 {
  bone1Name=base_hair_front_R_01
  bone2Name=head
  swing1LimitDegrees=45
  swing1Motion=Locked
  swing2LimitDegrees=45
  swing2Motion=Locked
  twistLimitDegrees=45
  twistMotion=Locked
 }
 {bone1Name=dyn_hair_front_R_01, bone2Name=base_hair_front_R_01, swing1LimitDegrees=15, swing2LimitDegrees=15, twistLimitDegrees=2}
 {bone1Name=dyn_hair_front_R_02, bone2Name=dyn_hair_front_R_01, swing1LimitDegrees=15, swing2LimitDegrees=15, twistLimitDegrees=2}
 {bone1Name=dyn_hair_front_R_03, bone2Name=dyn_hair_front_R_02, swing1LimitDegrees=15, swing2LimitDegrees=15, twistLimitDegrees=2}
 {bone1Name=dyn_hair_front_R_05, bone2Name=dyn_hair_front_R_02, swing1LimitDegrees=15, swing2LimitDegrees=15, twistLimitDegrees=2}
 {bone1Name=dyn_hair_front_R_07, bone2Name=dyn_hair_front_R_02, swing1LimitDegrees=15, swing2LimitDegrees=15, twistLimitDegrees=2}
 {bone1Name=dyn_hair_front_R_08, bone2Name=dyn_hair_front_R_07, swing1LimitDegrees=15, swing2LimitDegrees=15, twistLimitDegrees=2}
 {bone1Name=dyn_hair_front_R_09, bone2Name=dyn_hair_front_R_02, swing1LimitDegrees=15, swing2LimitDegrees=15, twistLimitDegrees=2}
]
