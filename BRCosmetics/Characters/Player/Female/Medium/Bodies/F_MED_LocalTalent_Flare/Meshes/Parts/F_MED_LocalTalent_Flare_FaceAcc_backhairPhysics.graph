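# Physics asset for the back hair of the female medium Flare character cosmetic, with capsule bodies for the hair and upper body.
# Spine, neck, head, clavicles and upper arms form an angularly locked body frame, with zero-size base hair capsules locked to the head.
# Left, middle and right hair chains hang from those bases, with tight twist and one swing limit widening toward the tips.
# Locked cross-constraints tie the middle chain's second segment to both side chains.
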PhysicsAsset /BRCosmetics/Characters/Player/Female/Medium/Bodies/F_MED_LocalTalent_Flare/Meshes/Parts/F_MED_LocalTalent_Flare_FaceAcc_backhairPhysics
bodies {
 base_hair_back_L_01 [{shapeType=Capsule, length=0, radius=0, shapeName=base_hair_back_L_01_capsule}]
 base_hair_back_M_01 [{shapeType=Capsule, length=0, radius=0, shapeName=base_hair_back_M_01_capsule}]
 base_hair_back_R_01 [{shapeType=Capsule, length=0, radius=0, shapeName=base_hair_back_R_01_capsule}]
 clavicle_l [{
   shapeType=Capsule
   length=6.5724
   radius=5.59
   shapeName=clavicle_l_capsule
   center=<6.4304 -1.1091 -1.5103>
   rotation=<rot -81.2292 179.9993 -179.9992>
  }]
 clavicle_r [{
   shapeType=Capsule
   length=6.5724
   radius=5.59
   shapeName=clavicle_r_capsule
   center=<-6.4304 1.1091 1.5103>
   rotation=<rot 81.2292 -0.0007 -0.0008>
  }]
 dyn_hair_back_L_01 [{shapeType=Capsule, length=1.1438, radius=3.7192, shapeName=dyn_hair_back_L_01_capsule, center=<4.0668 0 2.7806>, rotation=<rot 0 0 90>}]
 dyn_hair_back_L_02 [{
   shapeType=Capsule
   length=1.7149
   radius=5.1904
   shapeName=dyn_hair_back_L_02_capsule
   center=<3.6657 0 1.3509>
   rotation=<rot -0.0001 -8.5688 90>
  }]
 dyn_hair_back_L_03 [{shapeType=Capsule, length=3.1831, radius=4.7877, shapeName=dyn_hair_back_L_03_capsule, center=<4.5491 0 2.9265>, rotation=<rot 0 0 90>}]
 dyn_hair_back_L_04 [{
   shapeType=Capsule
   length=0
   radius=3.9146
   shapeName=dyn_hair_back_L_04_capsule
   center=<5.6467 -0.0161 1.7671>
   rotation=<rot -0.0001 0 95.1655>
  }]
 dyn_hair_back_M_01 [{shapeType=Capsule, length=3.6713, radius=3.7192, shapeName=dyn_hair_back_M_01_capsule, center=<4.0668 0 1.4772>, rotation=<rot 0 0 90>}]
 dyn_hair_back_M_02 [{shapeType=Capsule, length=3.5439, radius=4.7196, shapeName=dyn_hair_back_M_02_capsule, center=<3.6573 0 1.1635>, rotation=<rot 0 0 90>}]
 dyn_hair_back_M_03 [{shapeType=Capsule, length=5.3449, radius=4.7877, shapeName=dyn_hair_back_M_03_capsule, center=<4.0668 0 1.7345>, rotation=<rot 0 0 90>}]
 dyn_hair_back_M_04 [{shapeType=Capsule, length=4.4671, radius=3.7192, shapeName=dyn_hair_back_M_04_capsule, center=<4.0668 0 1.7455>, rotation=<rot 0 0 90>}]
 dyn_hair_back_M_05 [{shapeType=Capsule, length=5.3276, radius=3.7192, shapeName=dyn_hair_back_M_05_capsule, center=<2.8543 0 2.1886>, rotation=<rot 0 90 97.9779>}]
 dyn_hair_back_R_01 [{shapeType=Capsule, length=1.1438, radius=3.7192, shapeName=dyn_hair_back_R_01_capsule, center=<4.0668 0 2.7125>, rotation=<rot 0 0 90>}]
 dyn_hair_back_R_02 [{
   shapeType=Capsule
   length=1.8185
   radius=5.0019
   shapeName=dyn_hair_back_R_02_capsule
   center=<3.6657 0.2126 0.4068>
   rotation=<rot 0.0835 8.2536 89.4226>
  }]
 dyn_hair_back_R_03 [{
   shapeType=Capsule
   length=6.3436
   radius=4.7877
   shapeName=dyn_hair_back_R_03_capsule
   center=<4.5491 0 2.9265>
   rotation=<rot -0.0001 0 82.5062>
  }]
 dyn_hair_back_R_04 [{
   shapeType=Capsule
   length=0.0096
   radius=3.9146
   shapeName=dyn_hair_back_R_04_capsule
   center=<5.6467 1.3965 1.8687>
   rotation=<rot -0.0001 0 79.8191>
  }]
 head [{shapeType=Capsule, length=5.5323, radius=9.0121, shapeName=head_capsule, center=<-0.3623 2.2306 0>, rotation=<rot 90 14.2625 0>}]
 neck_01 [{shapeType=Capsule, length=4.3132, radius=4.4555, shapeName=neck_01_capsule, rotation=<rot 90 0 0>}]
 spine_03 [{shapeType=Capsule, length=5.1326, radius=10.3764, shapeName=spine_03_capsule, center=<3.0852 2.8594 0>}]
 spine_05 [{shapeType=Capsule, length=9.4688, radius=9.6309, shapeName=spine_05_capsule, center=<-2.5794 1.0289 0>}]
 upperarm_l [{
   shapeType=Capsule
   length=22.1065
   radius=4.9692
   shapeName=upperarm_l_capsule
   center=<13.3838 -0.7385 -0.2283>
   rotation=<rot -87.5661 0.0037 -0.0036>
  }]
 upperarm_r [{
   shapeType=Capsule
   length=22.1065
   radius=4.9692
   shapeName=upperarm_r_capsule
   center=<-13.3838 0.7385 0.2283>
   rotation=<rot 87.5661 -179.9963 -179.9964>
  }]
}
constraints [
 {
  bone1Name=spine_05
  bone2Name=spine_03
  swing1LimitDegrees=45
  swing1Motion=Locked
  swing2LimitDegrees=45
  swing2Motion=Locked
  twistLimitDegrees=45
  twistMotion=Locked
 }
 {
  bone1Name=clavicle_l
  bone2Name=spine_05
  swing1LimitDegrees=45
  swing1Motion=Locked
  swing2LimitDegrees=45
  swing2Motion=Locked
  twistLimitDegrees=45
  twistMotion=Locked
 }
 {
  bone1Name=clavicle_r
  bone2Name=spine_05
  swing1LimitDegrees=45
  swing1Motion=Locked
  swing2LimitDegrees=45
  swing2Motion=Locked
  twistLimitDegrees=45
  twistMotion=Locked
 }
 {
  bone1Name=upperarm_r
  bone2Name=clavicle_r
  swing1LimitDegrees=45
  swing1Motion=Locked
  swing2LimitDegrees=45
  swing2Motion=Locked
  twistLimitDegrees=45
  twistMotion=Locked
 }
 {
  bone1Name=upperarm_l
  bone2Name=clavicle_l
  swing1LimitDegrees=45
  swing1Motion=Locked
  swing2LimitDegrees=45
  swing2Motion=Locked
  twistLimitDegrees=45
  twistMotion=Locked
 }
 {
  bone1Name=neck_01
  bone2Name=spine_05
  swing1LimitDegrees=45
  swing1Motion=Locked
  swing2LimitDegrees=45
  swing2Motion=Locked
  twistLimitDegrees=45
  twistMotion=Locked
 }
 {
  bone1Name=head
  bone2Name=neck_01
  swing1LimitDegrees=45
  swing1Motion=Locked
  swing2LimitDegrees=45
  swing2Motion=Locked
  twistLimitDegrees=45
  twistMotion=Locked
 }
 {
  bone1Name=base_hair_back_L_01
  bone2Name=head
  swing1LimitDegrees=45
  swing1Motion=Locked
  swing2LimitDegrees=45
  swing2Motion=Locked
  twistLimitDegrees=45
  twistMotion=Locked
 }
 {bone1Name=dyn_hair_back_L_01, bone2Name=base_hair_back_L_01, swing1LimitDegrees=10, swing2LimitDegrees=5, twistLimitDegrees=2}
 {bone1Name=dyn_hair_back_L_02, bone2Name=dyn_hair_back_L_01, swing1LimitDegrees=10, swing2LimitDegrees=10, twistLimitDegrees=2}
 {bone1Name=dyn_hair_back_L_03, bone2Name=dyn_hair_back_L_02, swing1LimitDegrees=10, swing2LimitDegrees=30, twistLimitDegrees=2}
 {bone1Name=dyn_hair_back_L_04, bone2Name=dyn_hair_back_L_03, swing1LimitDegrees=10, swing2LimitDegrees=40, twistLimitDegrees=2}
 {
  bone1Name=base_hair_back_M_01
  bone2Name=head
  swing1LimitDegrees=45
  swing1Motion=Locked
  swing2LimitDegrees=45
  swing2Motion=Locked
  twistLimitDegrees=45
  twistMotion=Locked
 }
 {bone1Name=dyn_hair_back_M_01, bone2Name=base_hair_back_M_01, swing1LimitDegrees=10, swing2LimitDegrees=5, twistLimitDegrees=2}
 {bone1Name=dyn_hair_back_M_02, bone2Name=dyn_hair_back_M_01, swing1LimitDegrees=10, swing2LimitDegrees=10, twistLimitDegrees=2}
 {bone1Name=dyn_hair_back_M_03, bone2Name=dyn_hair_back_M_02, swing1LimitDegrees=10, swing2LimitDegrees=30, twistLimitDegrees=2}
 {bone1Name=dyn_hair_back_M_04, bone2Name=dyn_hair_back_M_03, swing1LimitDegrees=10, swing2LimitDegrees=40, twistLimitDegrees=2}
 {bone1Name=dyn_hair_back_M_05, bone2Name=dyn_hair_back_M_04, swing1LimitDegrees=10, swing2LimitDegrees=50, twistLimitDegrees=2}
 {
  bone1Name=base_hair_back_R_01
  bone2Name=head
  swing1LimitDegrees=45
  swing1Motion=Locked
  swing2LimitDegrees=45
  swing2Motion=Locked
  twistLimitDegrees=45
  twistMotion=Locked
 }
 {bone1Name=dyn_hair_back_R_01, bone2Name=base_hair_back_R_01, swing1LimitDegrees=10, swing2LimitDegrees=5, twistLimitDegrees=2}
 {bone1Name=dyn_hair_back_R_02, bone2Name=dyn_hair_back_R_01, swing1LimitDegrees=10, swing2LimitDegrees=10, twistLimitDegrees=2}
 {bone1Name=dyn_hair_back_R_03, bone2Name=dyn_hair_back_R_02, swing1LimitDegrees=10, swing2LimitDegrees=30, twistLimitDegrees=2}
 {bone1Name=dyn_hair_back_R_04, bone2Name=dyn_hair_back_R_03, swing1LimitDegrees=10, swing2LimitDegrees=40, twistLimitDegrees=2}
 {
  bone1Name=dyn_hair_back_M_02
  bone2Name=dyn_hair_back_R_02
  swing1LimitDegrees=45
  swing1Motion=Locked
  swing2LimitDegrees=45
  swing2Motion=Locked
  twistLimitDegrees=45
  twistMotion=Locked
 }
 {
  bone1Name=dyn_hair_back_M_02
  bone2Name=dyn_hair_back_L_02
  swing1LimitDegrees=45
  swing1Motion=Locked
  swing2LimitDegrees=45
  swing2Motion=Locked
  twistLimitDegrees=45
  twistMotion=Locked
 }
]
