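# Animation Blueprint for the Flare female character's face-accessory part, simulating back hair that adapts to worn back bling.
# The parent mesh's pose is copied through a Control Rig, cached, then fed to two neck-relative RigidBody sims (backhair, backpack-hair).
# A bool blend selects between them before a final RigidBody pass.
# Events derive that bool from back bling, a visible one-handed pickaxe and the Flare backpack tag.
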AnimBlueprint /BRCosmetics/Characters/Player/Female/Medium/Bodies/F_MED_LocalTalent_Flare/Meshes/Parts/F_MED_LocalTalent_Flare_FaceAcc_AnimBP
target_skeleton=&/F_MED_LocalTalent_Flare_FaceAcc_Skeleton
graphs [
 AnimGraph {nodes [
   n1 = RigidBody(ComponentPose=Local To Component(LocalPose=Use cached pose 'CRCache_Input_1'())) {
    bClampLinearTranslationLimitToRefPose=true
    bOverrideWorldGravity=true
    overridePhysicsAsset=&/F_MED_LocalTalent_Flare_FaceAcc_backpackhairPhy
    simulationSpace=BaseBoneSpace
    alphaBoolBlend {blendOption=Linear}
    baseBoneRef {boneName=neck_01}
    componentAppliedLinearAccClamp=<0 0 0>
    overrideWorldGravity {z=-980}
    simSpaceSettings {
     dampingAlpha=0
     maxAngularAcceleration=32
     maxAngularVelocity=8
     maxLinearAcceleration=1000
     maxLinearVelocity=700
     velocityScaleZ=0.5
     worldAlpha=1
     externalLinearDragV=<0.3 0.3 0.3>
    }
   }
   n2 = Component To Local(ComponentPose=@n1)
   n3 = RigidBody!(ComponentPose=Local To Component(LocalPose=Use cached pose 'CRCache_Input_1'()), Alpha, OverrideWorldGravity:Vector="0,0,0", ExternalForce) -> (Pose) {
    bClampLinearTranslationLimitToRefPose=true
    bOverrideWorldGravity=true
    overridePhysicsAsset=&/F_MED_LocalTalent_Flare_FaceAcc_backhairPhysics
    simulationSpace=BaseBoneSpace
    alphaBoolBlend {blendOption=Linear}
    baseBoneRef {boneName=neck_01}
    componentAppliedLinearAccClamp=<0 0 0>
    overrideWorldGravity {z=-980}
    simSpaceSettings {
     dampingAlpha=0
     maxAngularAcceleration=32
     maxAngularVelocity=8
     maxLinearAcceleration=1000
     maxLinearVelocity=700
     velocityScaleZ=0.5
     worldAlpha=1
     externalLinearDragV=<0.3 0.3 0.3>
    }
   }
   n4 = Component To Local(ComponentPose=@n3)
   n5 = Get isdynamics() -> (isdynamics:bool)
   n6 = RigidBody(
    ComponentPose=Local To Component(LocalPose=Blend Poses by bool(BlendPose_0=@n4, BlendPose_1=@n2, bActiveValue=@n5))
    OverrideWorldGravity="0,0,0"
   ) {
    bClampLinearTranslationLimitToRefPose=true
    bOverrideWorldGravity=true
    overridePhysicsAsset=&/F_MED_LocalTalent_Flare_FaceAcc_Physics
    simulationSpace=BaseBoneSpace
    alphaBoolBlend {blendOption=Linear}
    baseBoneRef {boneName=neck_01}
    componentAppliedLinearAccClamp=<0 0 0>
    overrideWorldGravity {z=-980}
    simSpaceSettings {
     dampingAlpha=0
     maxAngularAcceleration=32
     maxAngularVelocity=8
     maxLinearAcceleration=1000
     maxLinearVelocity=700
     velocityScaleZ=0.5
     worldAlpha=1
     externalLinearDragV=<0.3 0.3 0.3>
    }
   }
   Output Pose(Result=Component To Local(ComponentPose=@n6)) {layerGroup=None}
   n7 = Control Rig!(
    Source=Copy Pose From Mesh() {bUseAttachedParent=true}
    Alpha
    is Targeting:bool=Get bIsTargeting()
    is backpack:bool=Get isWearingBlackbling() -> (isWearingBlackbling:bool)
   ) -> (Pose) {controlRigAssetReference {blueprintRigClass=&/F_MED_LocalTalent_Flare_FaceAcc_CtrlRig.F_MED_LocalTalent_Flare_FaceAcc_CtrlRig_C}}
   Save cached pose 'CRCache_Input_1'(Pose=@n7) {cacheName=CRCache_Input_1}
  ]}
 EventGraph {nodes [
   Event OnWeaponChanged()
   > Set isBackpack_hidden(isBackpack_hidden:bool=BackpackItemDefinitionHasAnyMetaTag!(Tags:Array of Gameplay Tag Structures=Make Array("[0]"=(TagName="Cosmetics.MetaTags.LocalTalent_Flare")), bRequiresEquipped:bool=false) -> (ReturnValue:bool)) -> (Output_Get:bool)
   > Set ispickaxe_hidden(ispickaxe_hidden:bool=PickaxeItemDefinitionHasAnyMetaTag(Tags=Make Array("[0]"=(TagName="Cosmetics.MetaTags.LocalTalent_Flare")), bRequiresEquipped=true)) -> (Output_Get:bool)
   Event OnCharacterCustomizationLoaded()
   n1 = CheckForBackbling()
   > Set isWearingBlackbling(isWearingBlackbling:bool=OR Boolean!(A=@n1, B=@n1.bIsBackblingACape, C:bool=@n1.bIsBackblingLong) -> (ReturnValue)) -> (Output_Get:bool)
   Event BlueprintUpdateAnimation()
   > n2 = Sequence()
   n3 = Get isWearingBlackbling()
   n4 = Get isBackpack_hidden() -> (isBackpack_hidden:bool)
   n5 = Get Specific Weapon Visible() -> (Specific Weapon Visible:bool)
   > Set isdynamics(isdynamics:bool=OR Boolean!(A=NOT Boolean(A=@n3), B=AND Boolean(A=@n5, B=@n4)) -> (ReturnValue)) -> (Output_Get:bool)
   n6 = "Gravity Override State Function (Parameter Asset) (Nativized)"(
    execute=@n2.then_1
    CharacterPartAnimInstance=Self-Reference()
    GameTimeInSeconds=GetGameTimeInSeconds()
    Parameters=&/F_MED_LocalTalent_Flare_HAIR_Parameters
   )
   > Set GravityOverride(GravityOverride=@n6)
   > Set Fort Pawn(Fort Pawn:Fort Pawn Object Reference=TryGetFortPawn()) -> (Output_Get:Fort Pawn Object Reference)
   n7 = Get Fort Pawn() -> (Fort Pawn:Fort Pawn Object Reference)
   > n8 = Is Valid(InputObject=@n7)
   n9 = Set weapon vis(execute=@n8.Is Not Valid, weapon vis:bool=false) -> (Output_Get:bool)
   n10 = Get weapon vis() -> (weapon vis:bool)
   n11 = Branch(execute=[@n9.then, @n15.then], Condition=@n10)
   n12 = Get CurrentWeapon!(self:Fort Pawn Object Reference=Get Fort Pawn()) -> (CurrentWeapon:Fort Weapon Object Reference)
   > Set Specific Weapon Visible(Specific Weapon Visible:bool=AND Boolean(
     A=@n10
     B=WeaponDataHasMatchingGameplayTag!(self:Fort Weapon Object Reference=@n12, TagToCheck:Gameplay Tag Structure=(TagName="Cosmetics.Pickaxe.OneHanded")) -> (ReturnValue:bool)
    )) -> (Output_Get:bool)
   Set Specific Weapon Visible(execute=@n11.else, Specific Weapon Visible=false)
   n13 = Get CurrentWeapon(self=@n7)
   n14 = IsWeaponHolstered!(self:Fort Pawn Object Reference=@n7) -> (ReturnValue:bool)
   n15 = Set weapon vis(execute=@n8.Is Valid, weapon vis=AND Boolean(A=IsValid(Object=@n13), B=NOT Boolean(A=@n14)))
  ]}
]
variables [
 Fort Pawn {type=FortPawn, default=null}
 GravityOverride {type=Vector}
 isBackpack_hidden {type=boolean}
 isdynamics {type=boolean}
 ispickaxe_hidden {type=boolean}
 isWearingBlackbling {type=boolean}
 NewVar {type=number}
 Specific Weapon Visible {type=boolean}
 weapon vis {type=boolean}
 X {type=number}
]
